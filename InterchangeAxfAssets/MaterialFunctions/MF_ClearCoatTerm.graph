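# MF_ClearCoatTerm builds a Substrate slab for the clear coat top layer in the Interchange AxF material assets.
# It has zero diffuse albedo, derives F0 from a ClearCoatIOR parameter, and takes roughness from a parameter and the normal from a function input.
# A transmission color parameter converts to SSS mean free path to tint transmitted light, and the slab is output as Top Layer Slab.
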MaterialFunction /InterchangeAxfAssets/MaterialFunctions/MF_ClearCoatTerm
graph {nodes [
  n1 = ScalarParameter() {defaultValue=1.5, group=ClearCoat, parameterName=ClearCoatIOR, sortPriority=3}
  n2 = VectorParameter() {group=ClearCoat, parameterName=ClearCoatTransmissionColor, sortPriority=5, defaultValue=<rgba 1 1 1 1>}
  FunctionOutput(None=SubstrateSlabBSDF!(
    Diffuse Albedo=Constant()
    F0=MaterialFunctionCall!(IOR=@n1) {materialFunction=/Engine/Functions/Substrate/Substrate-IOR-To-F0}.F0
    F90
    Roughness=ScalarParameter() {defaultValue=0.01, group=ClearCoat, parameterName=ClearCoatRoughness, sortPriority=4}
    Anisotropy
    Normal=FunctionInput() {inputName=ClearCoatNormalV}
    Tangent
    SSS MFP=SubstrateTransmittanceToMFP!(TransmittanceColor=@n2.RGB, Thickness).MFP
    SSS MFP Scale
    SSS Phase Anisotropy
    Emissive Color
    Second Roughness
    Second Roughness Weight
    Fuzz Roughness
    Fuzz Amount
    Fuzz Color
    "Glint Density (Disabled)"
    "Glint UVs (Disabled)"
   )) {outputName=Top Layer Slab}
 ]}
