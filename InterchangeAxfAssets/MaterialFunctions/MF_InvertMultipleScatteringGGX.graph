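# Material function for AxF materials that outputs OneOverMSTerm, the reciprocal of the GGX multiple-scattering energy compensation term.
# It samples single-scattering energy Ess from the GGX_ReflectionEnergy lookup texture, indexed by isotropic roughness and the pixel normal dotted with the negated camera direction.
# It combines Ess with the F0 input as 1 / (1 + F0 × (1 − Ess) / Ess), allowing that compensation to be cancelled.
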MaterialFunction /InterchangeAxfAssets/MaterialFunctions/MF_InvertMultipleScatteringGGX
graph {nodes [
  n1 = NamedRerouteUsage()
  n2 = Multiply(A=FunctionInput() {inputName=F0}, B=Divide(A=Subtract(B=@n1), B=NamedRerouteUsage()))
  FunctionOutput(None=Divide(B=Add(A=@n2))) {outputName=OneOverMSTerm}
  n3 = MaterialFunctionCall() {materialFunction=&CameraDirectionVector}
  n4 = DotProduct(A=Subtract(B=Normalize(VectorInput=@n3.Result)) {constA=0}, B=PixelNormalWS())
  n5 = FunctionInput() {inputName=IsotropicRoughness, inputType=FunctionInput_Scalar}
  NamedRerouteDeclaration(None=TextureSample(UVs=MaterialFunctionCall(X=@n4, Y=@n5) {materialFunction=&MakeFloat2}.Result) {samplerType=SAMPLERTYPE_LinearColor, texture=/InterchangeAxfAssets/Textures/GGX_ReflectionEnergy}.R) {name=Ess, nodeColor=<rgba 0 1 0.9062 1>}
 ]}
