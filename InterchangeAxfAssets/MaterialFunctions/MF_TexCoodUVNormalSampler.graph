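# Material function in the Interchange AxF assets that samples a normal map and rotates the resulting vector within the tangent plane.
# The UV and texture object inputs feed a normal sampler, with a flat normal as the texture preview default.
# A custom HLSL node clamps the scalar rotation input to 0–1, maps it to one full turn and rotates the sample clockwise around Z, leaving Z unchanged.
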MaterialFunction /InterchangeAxfAssets/MaterialFunctions/MF_TexCoodUVNormalSampler
graph {nodes [
  MaterialExpressionFunctionInput_1 = FunctionInput() {inputName=Rotation_TC, inputType=FunctionInput_Scalar, sortPriority=6}
  MaterialExpressionTextureSample_0 = TextureSample(
   UVs=FunctionInput(Preview=TextureCoordinate()) {inputName=UV_TC, inputType=FunctionInput_Vector2, sortPriority=3}
   Tex=FunctionInput(Preview=TextureObject() {samplerType=SAMPLERTYPE_Normal, texture=/Bridge/MSPresets/MSTextures/FlatNormal}) {inputName=NormalTexOBJ, inputType=FunctionInput_Texture2D}
  ) {samplerType=SAMPLERTYPE_Normal}
  n1 = Custom!(RotationAmount=@MaterialExpressionFunctionInput_1, Vector=@MaterialExpressionTextureSample_0.RGB) {
   code="const float theta = saturate(RotationAmount) * PI * 2.;\r\nconst float3x3 rot3M = { cos(theta) , sin(theta), 0.f,\r\n                         -sin(theta), cos(theta), 0.f,\r\n                                 0.f,        0.f, 1.f\r\n                       };\r\nreturn mul(rot3M, Vector);"
   description=RotCW3D_Plane
   showCode=true
   inputs [
    {
     inputName=RotationAmount
     input {expression=&self:MaterialExpressionFunctionInput_1}
    }
    {
     inputName=Vector
     input {
      expression=&self:MaterialExpressionTextureSample_0
      mask=1
      maskB=1
      maskG=1
      maskR=1
     }
    }
   ]
  }
  FunctionOutput(None=@n1)
 ]}
