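# CPA2 is a Substrate base material for AxF assets, with texture-or-constant switches for diffuse, normal, roughness, F0 and F90.
# Millimeter size and transform parameters drive either mesh UVs or projected TP coordinates with explicit derivatives and reoriented normals, chosen by a static bool.
# Roughness is square-rooted, and F0/F90 are scaled by an inverse GGX multiple-scattering term.
# These feed a slab BSDF, optionally layered under a clear coat with its own normal.
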Material /InterchangeAxfAssets/BaseMaterials/CPA2
graph {
 expressions {
  MaterialExpressionAbs_1 {class=MaterialExpressionAbs, inputs [None {from=MaterialExpressionDotProduct_1, output=""}], properties {}}
  MaterialExpressionAdd_0 {
   class=MaterialExpressionAdd
   inputs [
    A {from=MaterialExpressionMaterialFunctionCall_8, output=R}
    B {from=MaterialExpressionMaterialFunctionCall_8, output=R}
   ]
  }
  MaterialExpressionCameraVectorWS_1 {class=MaterialExpressionCameraVectorWS, inputs [], properties {}}
  MaterialExpressionComponentMask_0 {
   class=MaterialExpressionComponentMask
   inputs [None {from=MaterialExpressionVectorParameter_1, output=RGB}]
   properties {b=true, g=true, r=true}
  }
  MaterialExpressionComponentMask_2 {
   class=MaterialExpressionComponentMask
   inputs [None {from=MaterialExpressionVectorParameter_5, output=RGB}]
   properties {b=true, g=true, r=true}
  }
  MaterialExpressionConstant_3 {class=MaterialExpressionConstant, properties {desc="1 = local", r=1}}
  MaterialExpressionConstant_5 {class=MaterialExpressionConstant, properties {desc="0 = world"}}
  MaterialExpressionConvert_0 {
   class=MaterialExpressionConvert
   inputs [None {from=MaterialExpressionMaterialFunctionCall_16, output=DXY}]
   properties {
    convertInputs [{
      type=Vector4
      defaultValue=<rgba 0 0 0 1>
      expressionInput {
       expression=&self:MaterialExpressionMaterialFunctionCall_16
       inputName=None
       mask=0
       maskA=0
       maskB=0
       maskG=0
       maskR=0
       outputIndex=4
      }
     }]
    convertOutputs [{type=Vector2}, {type=Vector2}]
   }
  }
  MaterialExpressionConvert_1 {
   class=MaterialExpressionConvert
   inputs [
    None {from=MaterialExpressionNamedRerouteUsage_68}
    None {from=MaterialExpressionNamedRerouteUsage_69}
    None {from=MaterialExpressionNamedRerouteUsage_74}
    None {from=MaterialExpressionNamedRerouteUsage_67}
   ]
   properties {
    convertInputs [
     {
      type=Scalar
      expressionInput {
       expression=&self:MaterialExpressionNamedRerouteUsage_68
       inputName=None
       mask=0
       maskA=0
       maskB=0
       maskG=0
       maskR=0
       outputIndex=0
      }
     }
     {
      type=Scalar
      expressionInput {
       expression=&self:MaterialExpressionNamedRerouteUsage_69
       inputName=None
       mask=0
       maskA=0
       maskB=0
       maskG=0
       maskR=0
       outputIndex=0
      }
     }
     {
      type=Scalar
      expressionInput {
       expression=&self:MaterialExpressionNamedRerouteUsage_74
       inputName=None
       mask=0
       maskA=0
       maskB=0
       maskG=0
       maskR=0
       outputIndex=0
      }
     }
     {
      type=Scalar
      expressionInput {
       expression=&self:MaterialExpressionNamedRerouteUsage_67
       inputName=None
       mask=0
       maskA=0
       maskB=0
       maskG=0
       maskR=0
       outputIndex=0
      }
     }
    ]
    convertOutputs [{type=Vector4, defaultValue=<rgba 0 0 0 1>}]
   }
  }
  MaterialExpressionConvert_2 {
   class=MaterialExpressionConvert
   inputs [
    None {from=MaterialExpressionNamedRerouteUsage_88}
    None {from=MaterialExpressionNamedRerouteUsage_89}
    None {from=MaterialExpressionNamedRerouteUsage_90}
   ]
   properties {
    convertInputs [
     {
      type=Scalar
      expressionInput {
       expression=&self:MaterialExpressionNamedRerouteUsage_88
       inputName=None
       mask=0
       maskA=0
       maskB=0
       maskG=0
       maskR=0
       outputIndex=0
      }
     }
     {
      type=Scalar
      expressionInput {
       expression=&self:MaterialExpressionNamedRerouteUsage_89
       inputName=None
       mask=0
       maskA=0
       maskB=0
       maskG=0
       maskR=0
       outputIndex=0
      }
     }
     {
      type=Scalar
      expressionInput {
       expression=&self:MaterialExpressionNamedRerouteUsage_90
       inputName=None
       mask=0
       maskA=0
       maskB=0
       maskG=0
       maskR=0
       outputIndex=0
      }
     }
    ]
    convertOutputs [{type=Vector3}]
   }
  }
  MaterialExpressionDivide_0 {
   class=MaterialExpressionDivide
   inputs [
    A {from=MaterialExpressionAdd_0}
    B {from=""}
   ]
  }
  MaterialExpressionDivide_6 {
   class=MaterialExpressionDivide
   inputs [
    A {from=MaterialExpressionScalarParameter_33}
    B {from=""}
   ]
   properties {constA=10, constB=10}
  }
  MaterialExpressionDivide_7 {
   class=MaterialExpressionDivide
   inputs [
    A {from=MaterialExpressionScalarParameter_34}
    B {from=""}
   ]
   properties {constA=10, constB=10}
  }
  MaterialExpressionDotProduct_1 {
   class=MaterialExpressionDotProduct
   inputs [
    A {from=MaterialExpressionShaderStageSwitch_0, output=""}
    B {from=MaterialExpressionVertexNormalWS_2, output=""}
   ]
   properties {}
  }
  MaterialExpressionMaterialFunctionCall_0 {
   class=MaterialExpressionMaterialFunctionCall
   inputs [
    Coordinates {from=MaterialExpressionNamedRerouteUsage_51}
    TexObject {from=MaterialExpressionTextureObjectParameter_3}
    DDX {from=MaterialExpressionNamedRerouteUsage_28}
    DDY {from=MaterialExpressionNamedRerouteUsage_29}
    Mask {from=""}
    DefaultUV {from=""}
   ]
   properties {materialFunction=/BaseMaterial/Materials/Functions/MF_TexPointSample_RGB}
  }
  MaterialExpressionMaterialFunctionCall_1 {
   class=MaterialExpressionMaterialFunctionCall
   inputs [
    Coordinates {from=MaterialExpressionNamedRerouteUsage_50}
    TexObject {from=MaterialExpressionTextureObjectParameter_4}
    DDX {from=MaterialExpressionNamedRerouteUsage_24}
    DDY {from=MaterialExpressionNamedRerouteUsage_27}
    Mask {from=""}
    DefaultUV {from=""}
   ]
   properties {materialFunction=&MF_TexPointSample_RGB}
  }
  MaterialExpressionMaterialFunctionCall_12 {
   class=MaterialExpressionMaterialFunctionCall
   inputs [
    NormalIn {from=MaterialExpressionStaticSwitch_4}
    NormalStrength {from=MaterialExpressionScalarParameter_8}
    ReFresnelVX {from=MaterialExpressionNamedRerouteUsage_33}
   ]
   properties {materialFunction=/InterchangeAxfAssets/MaterialFunctions/MF_NormalStrength}
  }
  MaterialExpressionMaterialFunctionCall_13 {
   class=MaterialExpressionMaterialFunctionCall
   inputs []
   properties {materialFunction=&CameraDirectionVector}
  }
  MaterialExpressionMaterialFunctionCall_14 {
   class=MaterialExpressionMaterialFunctionCall
   inputs [
    NormalTexOBJ {from=MaterialExpressionTextureObjectParameter_5}
    UV_TC {from=MaterialExpressionNamedRerouteUsage_112}
    Rotation_TC {from=MaterialExpressionNamedRerouteUsage_40}
   ]
   properties {materialFunction=/InterchangeAxfAssets/MaterialFunctions/MF_TexCoodUVNormalSampler}
  }
  MaterialExpressionMaterialFunctionCall_15 {
   class=MaterialExpressionMaterialFunctionCall
   inputs [
    IsotropicRoughness {from=MaterialExpressionNamedRerouteUsage_54}
    F0 {from=MaterialExpressionStaticSwitchParameter_4}
   ]
   properties {materialFunction=/InterchangeAxfAssets/MaterialFunctions/MF_InvertMultipleScatteringGGX}
  }
  MaterialExpressionMaterialFunctionCall_16 {
   class=MaterialExpressionMaterialFunctionCall
   inputs [
    AxF_Scale_X {from=MaterialExpressionNamedRerouteUsage_75}
    AxF_Scale_Y {from=MaterialExpressionNamedRerouteUsage_86}
    AxF_UnitScale {from=MaterialExpressionNamedRerouteUsage_87}
    Tile XYZ {from=MaterialExpressionNamedRerouteUsage_91}
    Offset XYZ {from=MaterialExpressionNamedRerouteUsage_93}
    Transition Softness {from=MaterialExpressionNamedRerouteUsage_107}
    Coordinate Space {from=MaterialExpressionStaticSwitchParameter_9}
    Rotation {from=MaterialExpressionNamedRerouteUsage_92}
   ]
   properties {materialFunction=/InterchangeAxfAssets/MaterialFunctions/MF_AxFMakeCoordinatesTP}
  }
  MaterialExpressionMaterialFunctionCall_18 {
   class=MaterialExpressionMaterialFunctionCall
   inputs [
    AxF_Scale_X {from=MaterialExpressionNamedRerouteUsage_104}
    AxF_Scale_Y {from=MaterialExpressionNamedRerouteUsage_105}
    AxF_UnitScale {from=MaterialExpressionNamedRerouteUsage_106}
    Tile U {from=MaterialExpressionNamedRerouteUsage_100}
    Tile V {from=MaterialExpressionNamedRerouteUsage_101}
    Offset U {from=MaterialExpressionNamedRerouteUsage_102}
    Offset V {from=MaterialExpressionNamedRerouteUsage_103}
    Rotation {from=MaterialExpressionNamedRerouteUsage_99}
    UV Index {from=MaterialExpressionScalarParameter_36}
   ]
   properties {materialFunction=/InterchangeAxfAssets/MaterialFunctions/MF_AxFMakeCoordinatesUV}
  }
  MaterialExpressionMaterialFunctionCall_19 {
   class=MaterialExpressionMaterialFunctionCall
   inputs [
    Coordinates {from=MaterialExpressionNamedRerouteUsage_41}
    TexObject {from=MaterialExpressionTextureObjectParameter_0}
    DDX {from=MaterialExpressionNamedRerouteUsage_14}
    DDY {from=MaterialExpressionNamedRerouteUsage_15}
    Mask {from=""}
    DefaultUV {from=""}
   ]
   properties {materialFunction=&MF_TexPointSample_RGB}
  }
  MaterialExpressionMaterialFunctionCall_20 {
   class=MaterialExpressionMaterialFunctionCall
   inputs [
    Coordinates {from=MaterialExpressionNamedRerouteUsage_42}
    TexObject {from=MaterialExpressionTextureObjectParameter_6}
    DDX {from=MaterialExpressionNamedRerouteUsage_16}
    DDY {from=MaterialExpressionNamedRerouteUsage_17}
    Mask {from=""}
    DefaultUV {from=""}
   ]
   properties {materialFunction=/BaseMaterial/Materials/Functions/MF_TexPointSample_NRM}
  }
  MaterialExpressionMaterialFunctionCall_21 {
   class=MaterialExpressionMaterialFunctionCall
   inputs [
    Coordinates {from=MaterialExpressionNamedRerouteUsage_53}
    TexObject {from=MaterialExpressionTextureObjectParameter_2}
    DDX {from=MaterialExpressionNamedRerouteUsage_34}
    DDY {from=MaterialExpressionNamedRerouteUsage_35}
    Mask {from=""}
    DefaultUV {from=""}
   ]
   properties {materialFunction=&MF_TexPointSample_RGB}
  }
  MaterialExpressionMaterialFunctionCall_22 {
   class=MaterialExpressionMaterialFunctionCall
   inputs [
    Coordinates {from=MaterialExpressionNamedRerouteUsage_52}
    TexObject {from=MaterialExpressionTextureObjectParameter_5}
    DDX {from=MaterialExpressionNamedRerouteUsage_37}
    DDY {from=MaterialExpressionNamedRerouteUsage_38}
    Mask {from=""}
    DefaultUV {from=""}
   ]
   properties {materialFunction=&MF_TexPointSample_NRM}
  }
  MaterialExpressionMaterialFunctionCall_23 {
   class=MaterialExpressionMaterialFunctionCall
   inputs [
    TSN {from=MaterialExpressionMaterialFunctionCall_22, output=RGB}
    Tangent {from=MaterialExpressionNamedRerouteUsage_96}
    Bitangent {from=MaterialExpressionNamedRerouteUsage_97}
    "Rotation (Turns)" {from=MaterialExpressionNamedRerouteUsage_39}
   ]
   properties {materialFunction=/BaseMaterial/Materials/Functions/MF_TangentNormal_OrientWorld}
  }
  MaterialExpressionMaterialFunctionCall_24 {
   class=MaterialExpressionMaterialFunctionCall
   inputs [
    TSN {from=MaterialExpressionMaterialFunctionCall_20, output=RGB}
    Tangent {from=MaterialExpressionNamedRerouteUsage_94}
    Bitangent {from=MaterialExpressionNamedRerouteUsage_95}
    "Rotation (Turns)" {from=MaterialExpressionNamedRerouteUsage_18}
   ]
   properties {materialFunction=&MF_TangentNormal_OrientWorld}
  }
  MaterialExpressionMaterialFunctionCall_3 {
   class=MaterialExpressionMaterialFunctionCall
   inputs [ClearCoatNormalV {from=MaterialExpressionNamedRerouteUsage_13}]
   properties {materialFunction=/InterchangeAxfAssets/MaterialFunctions/MF_ClearCoatTerm}
  }
  MaterialExpressionMaterialFunctionCall_6 {
   class=MaterialExpressionMaterialFunctionCall
   inputs [
    NormalTexOBJ {from=MaterialExpressionTextureObjectParameter_6}
    UV_TC {from=MaterialExpressionNamedRerouteUsage_110}
    Rotation_TC {from=MaterialExpressionNamedRerouteUsage_19}
   ]
   properties {materialFunction=&MF_TexCoodUVNormalSampler}
  }
  MaterialExpressionMaterialFunctionCall_7 {
   class=MaterialExpressionMaterialFunctionCall
   inputs [
    NormalIn {from=MaterialExpressionStaticSwitch_1}
    NormalStrength {from=MaterialExpressionScalarParameter_1}
    ReFresnelVX {from=MaterialExpressionNamedRerouteUsage_23}
   ]
   properties {materialFunction=&MF_NormalStrength}
  }
  MaterialExpressionMaterialFunctionCall_8 {
   class=MaterialExpressionMaterialFunctionCall
   inputs [Color {from=MaterialExpressionStaticSwitchParameter_3}]
   properties {materialFunction=&SplitComponents}
  }
  MaterialExpressionMultiply_4 {
   class=MaterialExpressionMultiply
   inputs [
    A {from=MaterialExpressionStaticSwitchParameter_4}
    B {from=MaterialExpressionMaterialFunctionCall_15, output=OneOverMSTerm}
   ]
  }
  MaterialExpressionMultiply_6 {
   class=MaterialExpressionMultiply
   inputs [
    A {from=MaterialExpressionMaterialFunctionCall_15, output=OneOverMSTerm}
    B {from=MaterialExpressionStaticSwitchParameter_7}
   ]
  }
  MaterialExpressionMultiply_8 {
   class=MaterialExpressionMultiply
   inputs [
    A {from=MaterialExpressionScalarParameter_35}
    B {from=""}
   ]
   properties {constB=10}
  }
  MaterialExpressionNamedRerouteDeclaration_1 {
   class=MaterialExpressionNamedRerouteDeclaration
   inputs [None {from=MaterialExpressionMaterialFunctionCall_3, output=Top Layer Slab}]
   properties {name=ClearCoatTerm_O, nodeColor=<rgba 1 0.6797 0 1>}
  }
  MaterialExpressionNamedRerouteDeclaration_10 {
   class=MaterialExpressionNamedRerouteDeclaration
   inputs [None {from=MaterialExpressionMultiply_6}]
   properties {name=F90_O, nodeColor=<rgba 1 0 0.2812 1>}
  }
  MaterialExpressionNamedRerouteDeclaration_11 {
   class=MaterialExpressionNamedRerouteDeclaration
   inputs [None {from=MaterialExpressionSquareRoot_0}]
   properties {name=Roughness_O, nodeColor=<rgba 0.9219 1 0 1>}
  }
  MaterialExpressionNamedRerouteDeclaration_12 {
   class=MaterialExpressionNamedRerouteDeclaration
   inputs [None {from=MaterialExpressionStaticSwitchParameter_2}]
   properties {name=Normal_O, nodeColor=<rgba 1 0.9844 0 1>}
  }
  MaterialExpressionNamedRerouteDeclaration_17 {
   class=MaterialExpressionNamedRerouteDeclaration
   inputs [None {from=MaterialExpressionStaticSwitchParameter_8}]
   properties {name=CCNormal, nodeColor=<rgba 0.5469 0 1 1>}
  }
  MaterialExpressionNamedRerouteDeclaration_18 {
   class=MaterialExpressionNamedRerouteDeclaration
   inputs [None {from=MaterialExpressionMaterialFunctionCall_16, output=Turns}]
   properties {name=TP_Rotation, nodeColor=<rgba 0.9453 1 0 1>}
  }
  MaterialExpressionNamedRerouteDeclaration_19 {
   class=MaterialExpressionNamedRerouteDeclaration
   inputs [None {from=MaterialExpressionConvert_0}]
   properties {name=TP_DDY, nodeColor=<rgba 0 0.7891 1 1>}
  }
  MaterialExpressionNamedRerouteDeclaration_20 {
   class=MaterialExpressionNamedRerouteDeclaration
   inputs [None {from=MaterialExpressionConvert_0}]
   properties {name=TP_DDX, nodeColor=<rgba 0.5703 0 1 1>}
  }
  MaterialExpressionNamedRerouteDeclaration_21 {
   class=MaterialExpressionNamedRerouteDeclaration
   inputs [None {from=MaterialExpressionMaterialFunctionCall_16, output=Output UV}]
   properties {name=TP_UV, nodeColor=<rgba 0 0.1094 1 1>}
  }
  MaterialExpressionNamedRerouteDeclaration_25 {
   class=MaterialExpressionNamedRerouteDeclaration
   inputs [None {from=MaterialExpressionSaturate_1}]
   properties {name=ReFresnelVX, nodeColor=<rgba 0.2656 0 1 1>}
  }
  MaterialExpressionNamedRerouteDeclaration_29 {
   class=MaterialExpressionNamedRerouteDeclaration
   inputs [None {from=MaterialExpressionScalarParameter_15}]
   properties {name=UV_Scale, nodeColor=<rgba 0.7578 1 0 1>}
  }
  MaterialExpressionNamedRerouteDeclaration_3 {
   class=MaterialExpressionNamedRerouteDeclaration
   inputs [None {from=MaterialExpressionStaticSwitchParameter_1}]
   properties {name=DiffuseAlbedoTerm_O, nodeColor=<rgba 0 1 0.0391 1>}
  }
  MaterialExpressionNamedRerouteDeclaration_30 {
   class=MaterialExpressionNamedRerouteDeclaration
   inputs [None {from=MaterialExpressionScalarParameter_13}]
   properties {name=UV_Rot, nodeColor=<rgba 0.7578 1 0 1>}
  }
  MaterialExpressionNamedRerouteDeclaration_31 {
   class=MaterialExpressionNamedRerouteDeclaration
   inputs [None {from=MaterialExpressionScalarParameter_16}]
   properties {name=UV_X, nodeColor=<rgba 0 1 0.6016 1>}
  }
  MaterialExpressionNamedRerouteDeclaration_32 {
   class=MaterialExpressionNamedRerouteDeclaration
   inputs [None {from=MaterialExpressionScalarParameter_17}]
   properties {name=UV_Y, nodeColor=<rgba 0 1 0.6016 1>}
  }
  MaterialExpressionNamedRerouteDeclaration_33 {
   class=MaterialExpressionNamedRerouteDeclaration
   inputs [None {from=MaterialExpressionScalarParameter_20}]
   properties {name=UV_Z, nodeColor=<rgba 0 1 0.6016 1>}
  }
  MaterialExpressionNamedRerouteDeclaration_34 {
   class=MaterialExpressionNamedRerouteDeclaration
   inputs [None {from=MaterialExpressionScalarParameter_19}]
   properties {name=OffsetX, nodeColor=<rgba 0.3594 1 0 1>}
  }
  MaterialExpressionNamedRerouteDeclaration_35 {
   class=MaterialExpressionNamedRerouteDeclaration
   inputs [None {from=MaterialExpressionScalarParameter_18}]
   properties {name=OffsetY, nodeColor=<rgba 0.3594 1 0 1>}
  }
  MaterialExpressionNamedRerouteDeclaration_36 {
   class=MaterialExpressionNamedRerouteDeclaration
   inputs [None {from=MaterialExpressionScalarParameter_21}]
   properties {name=OffsetZ, nodeColor=<rgba 0.3594 1 0 1>}
  }
  MaterialExpressionNamedRerouteDeclaration_48 {
   class=MaterialExpressionNamedRerouteDeclaration
   inputs [None {from=MaterialExpressionDivide_6}]
   properties {name=AxF_ScaleX, nodeColor=<rgba 0.7605 0 0.0009 1>}
  }
  MaterialExpressionNamedRerouteDeclaration_57 {
   class=MaterialExpressionNamedRerouteDeclaration
   inputs [None {from=MaterialExpressionDivide_7}]
   properties {name=AxF_ScaleY, nodeColor=<rgba 0.7605 0 0.0009 1>}
  }
  MaterialExpressionNamedRerouteDeclaration_58 {
   class=MaterialExpressionNamedRerouteDeclaration
   inputs [None {from=MaterialExpressionMultiply_8}]
   properties {name=AxF_UnitScale_MM, nodeColor=<rgba 0.7605 0 0.0009 1>}
  }
  MaterialExpressionNamedRerouteDeclaration_60 {
   class=MaterialExpressionNamedRerouteDeclaration
   inputs [None {from=MaterialExpressionConvert_1}]
   properties {name=TP_TileXYZ, nodeColor=<rgba 1 0.8438 0 1>}
  }
  MaterialExpressionNamedRerouteDeclaration_61 {
   class=MaterialExpressionNamedRerouteDeclaration
   inputs [None {from=MaterialExpressionConvert_2}]
   properties {name=TP_OffsetXYZ, nodeColor=<rgba 0 1 0.5078 1>}
  }
  MaterialExpressionNamedRerouteDeclaration_62 {
   class=MaterialExpressionNamedRerouteDeclaration
   inputs [None {from=MaterialExpressionMaterialFunctionCall_16, output=T}]
   properties {name=TP_Tangent, nodeColor=<rgba 1 0 0.0938 1>}
  }
  MaterialExpressionNamedRerouteDeclaration_63 {
   class=MaterialExpressionNamedRerouteDeclaration
   inputs [None {from=MaterialExpressionMaterialFunctionCall_16, output=B}]
   properties {name=TP_Bitangent, nodeColor=<rgba 0 0.4141 1 1>}
  }
  MaterialExpressionNamedRerouteDeclaration_64 {
   class=MaterialExpressionNamedRerouteDeclaration
   inputs [None {from=MaterialExpressionMaterialFunctionCall_18, output=Output UV}]
   properties {name=TC_UV, nodeColor=<rgba 0 1 0.25 1>}
  }
  MaterialExpressionNamedRerouteDeclaration_66 {
   class=MaterialExpressionNamedRerouteDeclaration
   inputs [None {from=MaterialExpressionStaticBoolParameter_1}]
   properties {name=UseGeoUVs, nodeColor=<rgba 0 0.8359 1 1>}
  }
  MaterialExpressionNamedRerouteDeclaration_67 {
   class=MaterialExpressionNamedRerouteDeclaration
   inputs [None {from=MaterialExpressionScalarParameter_37}]
   properties {name=TransitionSoftness, nodeColor=<rgba 0 1 0.2031 1>}
  }
  MaterialExpressionNamedRerouteDeclaration_9 {
   class=MaterialExpressionNamedRerouteDeclaration
   inputs [None {from=MaterialExpressionMultiply_4}]
   properties {name=F0_O, nodeColor=<rgba 1 0 0.2812 1>}
  }
  MaterialExpressionNamedRerouteUsage_100 {class=MaterialExpressionNamedRerouteUsage}
  MaterialExpressionNamedRerouteUsage_101 {class=MaterialExpressionNamedRerouteUsage}
  MaterialExpressionNamedRerouteUsage_102 {class=MaterialExpressionNamedRerouteUsage}
  MaterialExpressionNamedRerouteUsage_103 {class=MaterialExpressionNamedRerouteUsage}
  MaterialExpressionNamedRerouteUsage_104 {class=MaterialExpressionNamedRerouteUsage}
  MaterialExpressionNamedRerouteUsage_105 {class=MaterialExpressionNamedRerouteUsage}
  MaterialExpressionNamedRerouteUsage_106 {class=MaterialExpressionNamedRerouteUsage}
  MaterialExpressionNamedRerouteUsage_107 {class=MaterialExpressionNamedRerouteUsage}
  MaterialExpressionNamedRerouteUsage_109 {class=MaterialExpressionNamedRerouteUsage}
  MaterialExpressionNamedRerouteUsage_11 {class=MaterialExpressionNamedRerouteUsage}
  MaterialExpressionNamedRerouteUsage_110 {class=MaterialExpressionNamedRerouteUsage}
  MaterialExpressionNamedRerouteUsage_111 {class=MaterialExpressionNamedRerouteUsage}
  MaterialExpressionNamedRerouteUsage_112 {class=MaterialExpressionNamedRerouteUsage}
  MaterialExpressionNamedRerouteUsage_113 {class=MaterialExpressionNamedRerouteUsage}
  MaterialExpressionNamedRerouteUsage_114 {class=MaterialExpressionNamedRerouteUsage}
  MaterialExpressionNamedRerouteUsage_115 {class=MaterialExpressionNamedRerouteUsage}
  MaterialExpressionNamedRerouteUsage_116 {class=MaterialExpressionNamedRerouteUsage}
  MaterialExpressionNamedRerouteUsage_117 {class=MaterialExpressionNamedRerouteUsage}
  MaterialExpressionNamedRerouteUsage_119 {class=MaterialExpressionNamedRerouteUsage}
  MaterialExpressionNamedRerouteUsage_120 {class=MaterialExpressionNamedRerouteUsage}
  MaterialExpressionNamedRerouteUsage_121 {class=MaterialExpressionNamedRerouteUsage}
  MaterialExpressionNamedRerouteUsage_13 {class=MaterialExpressionNamedRerouteUsage}
  MaterialExpressionNamedRerouteUsage_14 {class=MaterialExpressionNamedRerouteUsage}
  MaterialExpressionNamedRerouteUsage_15 {class=MaterialExpressionNamedRerouteUsage}
  MaterialExpressionNamedRerouteUsage_16 {class=MaterialExpressionNamedRerouteUsage}
  MaterialExpressionNamedRerouteUsage_17 {class=MaterialExpressionNamedRerouteUsage}
  MaterialExpressionNamedRerouteUsage_18 {class=MaterialExpressionNamedRerouteUsage}
  MaterialExpressionNamedRerouteUsage_19 {class=MaterialExpressionNamedRerouteUsage}
  MaterialExpressionNamedRerouteUsage_23 {class=MaterialExpressionNamedRerouteUsage}
  MaterialExpressionNamedRerouteUsage_24 {class=MaterialExpressionNamedRerouteUsage}
  MaterialExpressionNamedRerouteUsage_27 {class=MaterialExpressionNamedRerouteUsage}
  MaterialExpressionNamedRerouteUsage_28 {class=MaterialExpressionNamedRerouteUsage}
  MaterialExpressionNamedRerouteUsage_29 {class=MaterialExpressionNamedRerouteUsage}
  MaterialExpressionNamedRerouteUsage_33 {class=MaterialExpressionNamedRerouteUsage}
  MaterialExpressionNamedRerouteUsage_34 {class=MaterialExpressionNamedRerouteUsage}
  MaterialExpressionNamedRerouteUsage_35 {class=MaterialExpressionNamedRerouteUsage}
  MaterialExpressionNamedRerouteUsage_37 {class=MaterialExpressionNamedRerouteUsage}
  MaterialExpressionNamedRerouteUsage_38 {class=MaterialExpressionNamedRerouteUsage}
  MaterialExpressionNamedRerouteUsage_39 {class=MaterialExpressionNamedRerouteUsage}
  MaterialExpressionNamedRerouteUsage_40 {class=MaterialExpressionNamedRerouteUsage}
  MaterialExpressionNamedRerouteUsage_41 {class=MaterialExpressionNamedRerouteUsage}
  MaterialExpressionNamedRerouteUsage_42 {class=MaterialExpressionNamedRerouteUsage}
  MaterialExpressionNamedRerouteUsage_50 {class=MaterialExpressionNamedRerouteUsage}
  MaterialExpressionNamedRerouteUsage_51 {class=MaterialExpressionNamedRerouteUsage}
  MaterialExpressionNamedRerouteUsage_52 {class=MaterialExpressionNamedRerouteUsage}
  MaterialExpressionNamedRerouteUsage_53 {class=MaterialExpressionNamedRerouteUsage}
  MaterialExpressionNamedRerouteUsage_54 {class=MaterialExpressionNamedRerouteUsage}
  MaterialExpressionNamedRerouteUsage_56 {class=MaterialExpressionNamedRerouteUsage}
  MaterialExpressionNamedRerouteUsage_57 {class=MaterialExpressionNamedRerouteUsage}
  MaterialExpressionNamedRerouteUsage_58 {class=MaterialExpressionNamedRerouteUsage}
  MaterialExpressionNamedRerouteUsage_59 {class=MaterialExpressionNamedRerouteUsage}
  MaterialExpressionNamedRerouteUsage_60 {class=MaterialExpressionNamedRerouteUsage}
  MaterialExpressionNamedRerouteUsage_67 {class=MaterialExpressionNamedRerouteUsage}
  MaterialExpressionNamedRerouteUsage_68 {class=MaterialExpressionNamedRerouteUsage}
  MaterialExpressionNamedRerouteUsage_69 {class=MaterialExpressionNamedRerouteUsage}
  MaterialExpressionNamedRerouteUsage_74 {class=MaterialExpressionNamedRerouteUsage}
  MaterialExpressionNamedRerouteUsage_75 {class=MaterialExpressionNamedRerouteUsage}
  MaterialExpressionNamedRerouteUsage_86 {class=MaterialExpressionNamedRerouteUsage}
  MaterialExpressionNamedRerouteUsage_87 {class=MaterialExpressionNamedRerouteUsage}
  MaterialExpressionNamedRerouteUsage_88 {class=MaterialExpressionNamedRerouteUsage}
  MaterialExpressionNamedRerouteUsage_89 {class=MaterialExpressionNamedRerouteUsage}
  MaterialExpressionNamedRerouteUsage_90 {class=MaterialExpressionNamedRerouteUsage}
  MaterialExpressionNamedRerouteUsage_91 {class=MaterialExpressionNamedRerouteUsage}
  MaterialExpressionNamedRerouteUsage_92 {class=MaterialExpressionNamedRerouteUsage}
  MaterialExpressionNamedRerouteUsage_93 {class=MaterialExpressionNamedRerouteUsage}
  MaterialExpressionNamedRerouteUsage_94 {class=MaterialExpressionNamedRerouteUsage}
  MaterialExpressionNamedRerouteUsage_95 {class=MaterialExpressionNamedRerouteUsage}
  MaterialExpressionNamedRerouteUsage_96 {class=MaterialExpressionNamedRerouteUsage}
  MaterialExpressionNamedRerouteUsage_97 {class=MaterialExpressionNamedRerouteUsage}
  MaterialExpressionNamedRerouteUsage_99 {class=MaterialExpressionNamedRerouteUsage}
  MaterialExpressionSaturate_1 {class=MaterialExpressionSaturate, inputs [None {from=MaterialExpressionAbs_1}]}
  MaterialExpressionScalarParameter_1 {
   class=MaterialExpressionScalarParameter
   properties {defaultValue=1, group=Normal, parameterName=NormalStrength, sliderMax=2, sliderMin=-1, sortPriority=12}
  }
  MaterialExpressionScalarParameter_13 {class=MaterialExpressionScalarParameter, properties {group=UV Transforms, parameterName="Rotation (degree)", sliderMax=360, sortPriority=30}}
  MaterialExpressionScalarParameter_15 {
   class=MaterialExpressionScalarParameter
   properties {defaultValue=1, group=UV Transforms, parameterName=Scale, sliderMax=10000, sliderMin=0.001, sortPriority=10}
  }
  MaterialExpressionScalarParameter_16 {
   class=MaterialExpressionScalarParameter
   properties {defaultValue=1, group=UV Transforms, parameterName=StretchOnX, sliderMax=100, sliderMin=0.001, sortPriority=100}
  }
  MaterialExpressionScalarParameter_17 {
   class=MaterialExpressionScalarParameter
   properties {defaultValue=1, group=UV Transforms, parameterName=StretchOnY, sliderMax=1000, sliderMin=0.001, sortPriority=110}
  }
  MaterialExpressionScalarParameter_18 {class=MaterialExpressionScalarParameter, properties {group=UV Transforms, parameterName=OffsetOnY, sortPriority=140}}
  MaterialExpressionScalarParameter_19 {class=MaterialExpressionScalarParameter, properties {group=UV Transforms, parameterName=OffsetOnX, sortPriority=130}}
  MaterialExpressionScalarParameter_20 {
   class=MaterialExpressionScalarParameter
   properties {defaultValue=1, group=UV Transforms, parameterName=StretchOnZ, sliderMax=1000, sliderMin=0.001, sortPriority=120}
  }
  MaterialExpressionScalarParameter_21 {class=MaterialExpressionScalarParameter, properties {group=UV Transforms, parameterName=OffsetOnZ, sortPriority=150}}
  MaterialExpressionScalarParameter_33 {class=MaterialExpressionScalarParameter, properties {defaultValue=100, group=MillimeterSettings, parameterName=WidthMM, sortPriority=10}}
  MaterialExpressionScalarParameter_34 {class=MaterialExpressionScalarParameter, properties {defaultValue=100, group=MillimeterSettings, parameterName=HeightMM, sortPriority=11}}
  MaterialExpressionScalarParameter_35 {class=MaterialExpressionScalarParameter, properties {defaultValue=1, group=MillimeterSettings, parameterName=UnitScaleCM, sortPriority=12}}
  MaterialExpressionScalarParameter_36 {class=MaterialExpressionScalarParameter, properties {group=UV Transforms, parameterName=UV Index}}
  MaterialExpressionScalarParameter_37 {
   class=MaterialExpressionScalarParameter
   properties {
    defaultValue=20
    desc="Controls the contrast of the blend mask. Internally applied as pow(mask, 8 × BlendFactor). A default of 0.5 corresponds to pow4 contrast."
    group=UV Transforms
    parameterName=TransitionSoftness
    sliderMax=1
    sortPriority=3
   }
  }
  MaterialExpressionScalarParameter_8 {
   class=MaterialExpressionScalarParameter
   properties {defaultValue=1, group=ClearCoat, parameterName=ClearCoatNormalStrength, sliderMax=2, sliderMin=-1}
  }
  MaterialExpressionShaderStageSwitch_0 {
   class=MaterialExpressionShaderStageSwitch
   inputs [
    PixelShader {from=MaterialExpressionCameraVectorWS_1, output=""}
    VertexShader {from=MaterialExpressionMaterialFunctionCall_13, output=Result}
   ]
   properties {}
  }
  MaterialExpressionSquareRoot_0 {class=MaterialExpressionSquareRoot, inputs [None {from=MaterialExpressionDivide_0, output=""}], properties {}}
  MaterialExpressionStaticBoolParameter_1 {
   class=MaterialExpressionStaticBoolParameter
   inputs []
   properties {defaultValue=true, group=UV Transforms, parameterName=Use Geo UVs, sortPriority=0}
  }
  MaterialExpressionStaticSwitchParameter_0 {
   class=MaterialExpressionStaticSwitchParameter
   inputs [
    True {from=MaterialExpressionSubstrateVerticalLayering_0, output=""}
    False {from=MaterialExpressionSubstrateSlabBSDF_3, output=""}
   ]
   properties {group=ClearCoat, parameterName=HasClearCoat, sortPriority=30}
  }
  MaterialExpressionStaticSwitchParameter_1 {
   class=MaterialExpressionStaticSwitchParameter
   inputs [
    True {from=MaterialExpressionStaticSwitch_0, output=""}
    False {from=MaterialExpressionVectorParameter_0, output=RGB}
   ]
   properties {group=Diffuse, parameterName=HasDiffuseColorTexture}
  }
  MaterialExpressionStaticSwitchParameter_2 {
   class=MaterialExpressionStaticSwitchParameter
   inputs [
    True {from=MaterialExpressionMaterialFunctionCall_7, output=NormalOut}
    False {from=MaterialExpressionComponentMask_0, output=""}
   ]
   properties {group=Normal, parameterName=HasNormalTexture, sortPriority=30}
  }
  MaterialExpressionStaticSwitchParameter_3 {
   class=MaterialExpressionStaticSwitchParameter
   inputs [
    True {from=MaterialExpressionStaticSwitch_2, output=""}
    False {from=MaterialExpressionVectorParameter_2, output=RGB}
   ]
   properties {group=Roughness, parameterName=HasRoughnessTexture, sortPriority=30}
  }
  MaterialExpressionStaticSwitchParameter_4 {
   class=MaterialExpressionStaticSwitchParameter
   inputs [
    True {from=MaterialExpressionStaticSwitch_6, output=""}
    False {from=MaterialExpressionVectorParameter_3, output=RGB}
   ]
   properties {group=Fresnel, parameterName=HasF0Texture}
  }
  MaterialExpressionStaticSwitchParameter_7 {
   class=MaterialExpressionStaticSwitchParameter
   inputs [
    True {from=MaterialExpressionStaticSwitch_5, output=""}
    False {from=MaterialExpressionVectorParameter_4, output=RGB}
   ]
   properties {group=Fresnel, parameterName=HasF90Texture}
  }
  MaterialExpressionStaticSwitchParameter_8 {
   class=MaterialExpressionStaticSwitchParameter
   inputs [
    True {from=MaterialExpressionMaterialFunctionCall_12, output=NormalOut}
    False {from=MaterialExpressionComponentMask_2, output=""}
   ]
   properties {group=ClearCoat, parameterName=HasClearCoatNormalTexture, sortPriority=30}
  }
  MaterialExpressionStaticSwitchParameter_9 {
   class=MaterialExpressionStaticSwitchParameter
   inputs [
    True {from=MaterialExpressionConstant_3, output=""}
    False {from=MaterialExpressionConstant_5, output=""}
   ]
   properties {defaultValue=true, group=UV Transforms, parameterName=UsingLocalSpace, sortPriority=2}
  }
  MaterialExpressionStaticSwitch_0 {
   class=MaterialExpressionStaticSwitch
   inputs [
    True {from=MaterialExpressionTextureSample_0, output=RGB}
    False {from=MaterialExpressionMaterialFunctionCall_19, output=RGB}
    Value {from=MaterialExpressionNamedRerouteUsage_115, output=""}
   ]
   properties {}
  }
  MaterialExpressionStaticSwitch_1 {
   class=MaterialExpressionStaticSwitch
   inputs [
    True {from=MaterialExpressionMaterialFunctionCall_6, output=Result}
    False {from=MaterialExpressionMaterialFunctionCall_24, output=TSN}
    Value {from=MaterialExpressionNamedRerouteUsage_116, output=""}
   ]
   properties {}
  }
  MaterialExpressionStaticSwitch_2 {
   class=MaterialExpressionStaticSwitch
   inputs [
    True {from=MaterialExpressionTextureSample_1, output=RGB}
    False {from=MaterialExpressionMaterialFunctionCall_21, output=RGB}
    Value {from=MaterialExpressionNamedRerouteUsage_117, output=""}
   ]
   properties {}
  }
  MaterialExpressionStaticSwitch_4 {
   class=MaterialExpressionStaticSwitch
   inputs [
    True {from=MaterialExpressionMaterialFunctionCall_14, output=Result}
    False {from=MaterialExpressionMaterialFunctionCall_23, output=TSN}
    Value {from=MaterialExpressionNamedRerouteUsage_119, output=""}
   ]
   properties {}
  }
  MaterialExpressionStaticSwitch_5 {
   class=MaterialExpressionStaticSwitch
   inputs [
    True {from=MaterialExpressionTextureSample_2, output=RGB}
    False {from=MaterialExpressionMaterialFunctionCall_0, output=RGB}
    Value {from=MaterialExpressionNamedRerouteUsage_120, output=""}
   ]
   properties {}
  }
  MaterialExpressionStaticSwitch_6 {
   class=MaterialExpressionStaticSwitch
   inputs [
    True {from=MaterialExpressionTextureSample_3, output=RGB}
    False {from=MaterialExpressionMaterialFunctionCall_1, output=RGB}
    Value {from=MaterialExpressionNamedRerouteUsage_121, output=""}
   ]
   properties {}
  }
  MaterialExpressionSubstrateSlabBSDF_3 {
   class=MaterialExpressionSubstrateSlabBSDF
   inputs [
    Diffuse Albedo {from=MaterialExpressionNamedRerouteUsage_56, output=""}
    F0 {from=MaterialExpressionNamedRerouteUsage_57, output=""}
    F90 {from=MaterialExpressionNamedRerouteUsage_60, output=""}
    Roughness {from=MaterialExpressionNamedRerouteUsage_58, output=""}
    Anisotropy {from="", output=""}
    Normal {from=MaterialExpressionNamedRerouteUsage_59, output=""}
    Tangent {from="", output=""}
    SSS MFP {from="", output=""}
    SSS MFP Scale {from="", output=""}
    SSS Phase Anisotropy {from="", output=""}
    Emissive Color {from="", output=""}
    Second Roughness {from="", output=""}
    Second Roughness Weight {from="", output=""}
    Fuzz Roughness {from="", output=""}
    Fuzz Amount {from="", output=""}
    Fuzz Color {from="", output=""}
    "Glint Density (Disabled)" {from="", output=""}
    "Glint UVs (Disabled)" {from="", output=""}
   ]
   properties {specularProfile=&/DefaultSpecularProfile}
  }
  MaterialExpressionSubstrateVerticalLayering_0 {
   class=MaterialExpressionSubstrateVerticalLayering
   inputs [
    Top {from=MaterialExpressionNamedRerouteUsage_11, output=""}
    Bottom {from=MaterialExpressionSubstrateSlabBSDF_3, output=""}
    Top Thickness {from="", output=""}
   ]
   properties {}
  }
  MaterialExpressionTextureObjectParameter_0 {
   class=MaterialExpressionTextureObjectParameter
   inputs [
    UVs {from="", output=""}
    Apply View MipBias {from="", output=""}
   ]
   properties {
    group=Diffuse
    parameterName=DiffuseColorTexture
    samplerType=SAMPLERTYPE_LinearColor
    sortPriority=1
    texture=/Engine/Functions/Engine_MaterialFunctions02/PivotPainter2/Black_1x1_EXR_Texture
   }
  }
  MaterialExpressionTextureObjectParameter_2 {
   class=MaterialExpressionTextureObjectParameter
   inputs [
    UVs {from="", output=""}
    Apply View MipBias {from="", output=""}
   ]
   properties {
    group=Roughness
    parameterName=RoughnessTexture
    samplerType=SAMPLERTYPE_LinearColor
    sortPriority=1
    texture=&Black_1x1_EXR_Texture
   }
  }
  MaterialExpressionTextureObjectParameter_3 {
   class=MaterialExpressionTextureObjectParameter
   inputs [
    UVs {from="", output=""}
    Apply View MipBias {from="", output=""}
   ]
   properties {
    group=Fresnel
    parameterName=F90Texture
    samplerType=SAMPLERTYPE_LinearColor
    sortPriority=2
    texture=&Black_1x1_EXR_Texture
   }
  }
  MaterialExpressionTextureObjectParameter_4 {
   class=MaterialExpressionTextureObjectParameter
   inputs [
    UVs {from="", output=""}
    Apply View MipBias {from="", output=""}
   ]
   properties {
    group=Fresnel
    parameterName=F0Texture
    samplerType=SAMPLERTYPE_LinearColor
    sortPriority=1
    texture=&Black_1x1_EXR_Texture
   }
  }
  MaterialExpressionTextureObjectParameter_5 {
   class=MaterialExpressionTextureObjectParameter
   inputs [
    UVs {from="", output=""}
    Apply View MipBias {from="", output=""}
   ]
   properties {
    group=ClearCoat
    parameterName=ClearCoatNormalTexture
    samplerType=SAMPLERTYPE_Normal
    sortPriority=1
    texture=/Engine/EngineMaterials/DefaultNormal
   }
  }
  MaterialExpressionTextureObjectParameter_6 {
   class=MaterialExpressionTextureObjectParameter
   inputs [
    UVs {from="", output=""}
    Apply View MipBias {from="", output=""}
   ]
   properties {group=Normal, parameterName=NormalTexture, samplerType=SAMPLERTYPE_Normal, sortPriority=1, texture=&DefaultNormal}
  }
  MaterialExpressionTextureSample_0 {
   class=MaterialExpressionTextureSample
   inputs [
    UVs {from=MaterialExpressionNamedRerouteUsage_109}
    Tex {from=MaterialExpressionTextureObjectParameter_0}
    Apply View MipBias {from=""}
   ]
   properties {samplerType=SAMPLERTYPE_LinearColor}
  }
  MaterialExpressionTextureSample_1 {
   class=MaterialExpressionTextureSample
   inputs [
    UVs {from=MaterialExpressionNamedRerouteUsage_111}
    Tex {from=MaterialExpressionTextureObjectParameter_2}
    Apply View MipBias {from=""}
   ]
   properties {samplerType=SAMPLERTYPE_LinearColor}
  }
  MaterialExpressionTextureSample_2 {
   class=MaterialExpressionTextureSample
   inputs [
    UVs {from=MaterialExpressionNamedRerouteUsage_113}
    Tex {from=MaterialExpressionTextureObjectParameter_3}
    Apply View MipBias {from=""}
   ]
   properties {samplerType=SAMPLERTYPE_LinearColor}
  }
  MaterialExpressionTextureSample_3 {
   class=MaterialExpressionTextureSample
   inputs [
    UVs {from=MaterialExpressionNamedRerouteUsage_114}
    Tex {from=MaterialExpressionTextureObjectParameter_4}
    Apply View MipBias {from=""}
   ]
   properties {samplerType=SAMPLERTYPE_LinearColor}
  }
  MaterialExpressionVectorParameter_0 {class=MaterialExpressionVectorParameter, properties {group=Diffuse, parameterName=DiffuseColor, defaultValue=<rgba 0 0 0 1>}}
  MaterialExpressionVectorParameter_1 {class=MaterialExpressionVectorParameter, properties {group=Normal, parameterName=Normal, defaultValue=<rgba 0 0 1 1>}}
  MaterialExpressionVectorParameter_2 {
   class=MaterialExpressionVectorParameter
   properties {group=Roughness, parameterName=Roughness, sortPriority=30, defaultValue=<rgba 0.2122 0.2122 0 1>}
  }
  MaterialExpressionVectorParameter_3 {class=MaterialExpressionVectorParameter, properties {group=Fresnel, parameterName=F0, defaultValue=<rgba 0.04 0.04 0.04 1>}}
  MaterialExpressionVectorParameter_4 {class=MaterialExpressionVectorParameter, properties {group=Fresnel, parameterName=F90, defaultValue=<rgba 1 1 1 0>}}
  MaterialExpressionVectorParameter_5 {class=MaterialExpressionVectorParameter, properties {group=ClearCoat, parameterName=ClearCoatNormal, defaultValue=<rgba 0 0 1 1>}}
  MaterialExpressionVertexNormalWS_2 {class=MaterialExpressionVertexNormalWS, inputs [], properties {}}
 }
 properties {bUsedWithNanite=true, bUsedWithStaticLighting=true, shadingModel=MSM_SubsurfaceProfile, translucencyLightingMode=TLM_SurfacePerPixelLighting}
}
