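# Substrate base material for AxF CPA2 flake surfaces, likely car paint, in the InterchangeAxfAssets plugin.
# Static switches select textures or constants for diffuse, normal, roughness, F0 and F90, sampled with millimeter-scaled mesh UVs or triplanar coordinates.
# F0 and F90 get a GGX multiple-scattering inverse correction before feeding a slab with second-roughness and glint inputs.
# An optional clear coat, with its own normal, is vertically layered on top.
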Material /InterchangeAxfAssets/BaseMaterials/CPA2_Flakes
graph {
 expressions {
  MaterialExpressionAbs_0 {class=MaterialExpressionAbs, inputs [None {from=MaterialExpressionDotProduct_0, output=""}], properties {}}
  MaterialExpressionAdd_0 {
   class=MaterialExpressionAdd
   inputs [
    A {from=MaterialExpressionDivide_1}
    B {from=""}
   ]
   properties {constB=0.4762, desc=10/21 to enforce max MicrofacetRelativeArea value}
  }
  MaterialExpressionAdd_2 {
   class=MaterialExpressionAdd
   inputs [
    A {from=MaterialExpressionMaterialFunctionCall_59, output=R}
    B {from=MaterialExpressionMaterialFunctionCall_59, output=R}
   ]
  }
  MaterialExpressionCameraVectorWS_0 {class=MaterialExpressionCameraVectorWS, inputs [], properties {}}
  MaterialExpressionComponentMask_1 {
   class=MaterialExpressionComponentMask
   inputs [None {from=MaterialExpressionVectorParameter_11, output=RGB}]
   properties {b=true, g=true, r=true}
  }
  MaterialExpressionComponentMask_8 {
   class=MaterialExpressionComponentMask
   inputs [None {from=MaterialExpressionVectorParameter_19, output=RGB}]
   properties {b=true, g=true, r=true}
  }
  MaterialExpressionConstant_2 {class=MaterialExpressionConstant, properties {desc="1 = local", r=1}}
  MaterialExpressionConstant_3 {class=MaterialExpressionConstant, properties {desc="0 = world"}}
  MaterialExpressionConvert_0 {
   class=MaterialExpressionConvert
   inputs [None {from=MaterialExpressionMaterialFunctionCall_8, output=DXY}]
   properties {
    convertInputs [{
      type=Vector4
      defaultValue=<rgba 0 0 0 1>
      expressionInput {
       expression=&self:MaterialExpressionMaterialFunctionCall_8
       inputName=None
       mask=0
       maskA=0
       maskB=0
       maskG=0
       maskR=0
       outputIndex=4
      }
     }]
    convertOutputs [{type=Vector2}, {type=Vector2}]
   }
  }
  MaterialExpressionConvert_1 {
   class=MaterialExpressionConvert
   inputs [
    None {from=MaterialExpressionNamedRerouteUsage_53}
    None {from=MaterialExpressionNamedRerouteUsage_54}
    None {from=MaterialExpressionNamedRerouteUsage_55}
    None {from=MaterialExpressionNamedRerouteUsage_52}
   ]
   properties {
    convertInputs [
     {
      type=Scalar
      expressionInput {
       expression=&self:MaterialExpressionNamedRerouteUsage_53
       inputName=None
       mask=0
       maskA=0
       maskB=0
       maskG=0
       maskR=0
       outputIndex=0
      }
     }
     {
      type=Scalar
      expressionInput {
       expression=&self:MaterialExpressionNamedRerouteUsage_54
       inputName=None
       mask=0
       maskA=0
       maskB=0
       maskG=0
       maskR=0
       outputIndex=0
      }
     }
     {
      type=Scalar
      expressionInput {
       expression=&self:MaterialExpressionNamedRerouteUsage_55
       inputName=None
       mask=0
       maskA=0
       maskB=0
       maskG=0
       maskR=0
       outputIndex=0
      }
     }
     {
      type=Scalar
      expressionInput {
       expression=&self:MaterialExpressionNamedRerouteUsage_52
       inputName=None
       mask=0
       maskA=0
       maskB=0
       maskG=0
       maskR=0
       outputIndex=0
      }
     }
    ]
    convertOutputs [{type=Vector4, defaultValue=<rgba 0 0 0 1>}]
   }
  }
  MaterialExpressionConvert_2 {
   class=MaterialExpressionConvert
   inputs [
    None {from=MaterialExpressionNamedRerouteUsage_33}
    None {from=MaterialExpressionNamedRerouteUsage_34}
    None {from=MaterialExpressionNamedRerouteUsage_35}
   ]
   properties {
    convertInputs [
     {
      type=Scalar
      expressionInput {
       expression=&self:MaterialExpressionNamedRerouteUsage_33
       inputName=None
       mask=0
       maskA=0
       maskB=0
       maskG=0
       maskR=0
       outputIndex=0
      }
     }
     {
      type=Scalar
      expressionInput {
       expression=&self:MaterialExpressionNamedRerouteUsage_34
       inputName=None
       mask=0
       maskA=0
       maskB=0
       maskG=0
       maskR=0
       outputIndex=0
      }
     }
     {
      type=Scalar
      expressionInput {
       expression=&self:MaterialExpressionNamedRerouteUsage_35
       inputName=None
       mask=0
       maskA=0
       maskB=0
       maskG=0
       maskR=0
       outputIndex=0
      }
     }
    ]
    convertOutputs [{type=Vector3}]
   }
  }
  MaterialExpressionDivide_0 {
   class=MaterialExpressionDivide
   inputs [
    A {from=MaterialExpressionScalarParameter_22}
    B {from=MaterialExpressionScalarParameter_1}
   ]
   properties {constB=25}
  }
  MaterialExpressionDivide_1 {
   class=MaterialExpressionDivide
   inputs [
    A {from=MaterialExpressionDivide_0}
    B {from=""}
   ]
  }
  MaterialExpressionDivide_2 {
   class=MaterialExpressionDivide
   inputs [
    A {from=MaterialExpressionAdd_2}
    B {from=""}
   ]
  }
  MaterialExpressionDivide_3 {
   class=MaterialExpressionDivide
   inputs [
    A {from=MaterialExpressionScalarParameter_9}
    B {from=""}
   ]
   properties {constA=10, constB=10}
  }
  MaterialExpressionDivide_4 {
   class=MaterialExpressionDivide
   inputs [
    A {from=MaterialExpressionScalarParameter_12}
    B {from=""}
   ]
   properties {constA=10, constB=10}
  }
  MaterialExpressionDotProduct_0 {
   class=MaterialExpressionDotProduct
   inputs [
    A {from=MaterialExpressionShaderStageSwitch_2, output=""}
    B {from=MaterialExpressionVertexNormalWS_0, output=""}
   ]
   properties {}
  }
  MaterialExpressionMaterialFunctionCall_0 {
   class=MaterialExpressionMaterialFunctionCall
   inputs [
    Coordinates {from=MaterialExpressionNamedRerouteUsage_0}
    TexObject {from=MaterialExpressionTextureObjectParameter_19}
    DDX {from=MaterialExpressionNamedRerouteUsage_244}
    DDY {from=MaterialExpressionNamedRerouteUsage_245}
    Mask {from=""}
    DefaultUV {from=""}
   ]
   properties {materialFunction=/BaseMaterial/Materials/Functions/MF_TexPointSample_RGB}
  }
  MaterialExpressionMaterialFunctionCall_1 {
   class=MaterialExpressionMaterialFunctionCall
   inputs [
    IsotropicRoughness {from=MaterialExpressionNamedRerouteUsage_3}
    F0 {from=MaterialExpressionStaticSwitchParameter_75}
   ]
   properties {materialFunction=/InterchangeAxfAssets/MaterialFunctions/MF_InvertMultipleScatteringGGX}
  }
  MaterialExpressionMaterialFunctionCall_10 {
   class=MaterialExpressionMaterialFunctionCall
   inputs [
    TSN {from=MaterialExpressionMaterialFunctionCall_9, output=RGB}
    Tangent {from=MaterialExpressionNamedRerouteUsage_66}
    Bitangent {from=MaterialExpressionNamedRerouteUsage_67}
    "Rotation (Turns)" {from=MaterialExpressionNamedRerouteUsage_268}
   ]
   properties {materialFunction=/BaseMaterial/Materials/Functions/MF_TangentNormal_OrientWorld}
  }
  MaterialExpressionMaterialFunctionCall_11 {
   class=MaterialExpressionMaterialFunctionCall
   inputs [
    TSN {from=MaterialExpressionMaterialFunctionCall_3, output=RGB}
    Tangent {from=MaterialExpressionNamedRerouteUsage_64}
    Bitangent {from=MaterialExpressionNamedRerouteUsage_65}
    "Rotation (Turns)" {from=MaterialExpressionNamedRerouteUsage_250}
   ]
   properties {materialFunction=&MF_TangentNormal_OrientWorld}
  }
  MaterialExpressionMaterialFunctionCall_17 {
   class=MaterialExpressionMaterialFunctionCall
   inputs [ClearCoatNormalV {from=MaterialExpressionNamedRerouteUsage_128}]
   properties {materialFunction=/InterchangeAxfAssets/MaterialFunctions/MF_ClearCoatTerm}
  }
  MaterialExpressionMaterialFunctionCall_2 {
   class=MaterialExpressionMaterialFunctionCall
   inputs [
    Coordinates {from=MaterialExpressionNamedRerouteUsage_12}
    TexObject {from=MaterialExpressionTextureObjectParameter_23}
    DDX {from=MaterialExpressionNamedRerouteUsage_255}
    DDY {from=MaterialExpressionNamedRerouteUsage_256}
    Mask {from=""}
    DefaultUV {from=""}
   ]
   properties {materialFunction=&MF_TexPointSample_RGB}
  }
  MaterialExpressionMaterialFunctionCall_3 {
   class=MaterialExpressionMaterialFunctionCall
   inputs [
    Coordinates {from=MaterialExpressionNamedRerouteUsage_2}
    TexObject {from=MaterialExpressionTextureObjectParameter_1}
    DDX {from=MaterialExpressionNamedRerouteUsage_246}
    DDY {from=MaterialExpressionNamedRerouteUsage_247}
    Mask {from=""}
    DefaultUV {from=""}
   ]
   properties {materialFunction=/BaseMaterial/Materials/Functions/MF_TexPointSample_NRM}
  }
  MaterialExpressionMaterialFunctionCall_4 {
   class=MaterialExpressionMaterialFunctionCall
   inputs [
    Coordinates {from=MaterialExpressionNamedRerouteUsage_26}
    TexObject {from=MaterialExpressionTextureObjectParameter_22}
    DDX {from=MaterialExpressionNamedRerouteUsage_257}
    DDY {from=MaterialExpressionNamedRerouteUsage_258}
    Mask {from=""}
    DefaultUV {from=""}
   ]
   properties {materialFunction=&MF_TexPointSample_RGB}
  }
  MaterialExpressionMaterialFunctionCall_5 {
   class=MaterialExpressionMaterialFunctionCall
   inputs [
    NormalIn {from=MaterialExpressionStaticSwitch_4}
    NormalStrength {from=MaterialExpressionScalarParameter_0}
    ReFresnelVX {from=MaterialExpressionNamedRerouteUsage_259}
   ]
   properties {materialFunction=/InterchangeAxfAssets/MaterialFunctions/MF_NormalStrength}
  }
  MaterialExpressionMaterialFunctionCall_57 {
   class=MaterialExpressionMaterialFunctionCall
   inputs [
    NormalTexOBJ {from=MaterialExpressionTextureObjectParameter_1}
    UV_TC {from=MaterialExpressionNamedRerouteUsage_103}
    Rotation_TC {from=MaterialExpressionNamedRerouteUsage_251}
   ]
   properties {materialFunction=/InterchangeAxfAssets/MaterialFunctions/MF_TexCoodUVNormalSampler}
  }
  MaterialExpressionMaterialFunctionCall_58 {
   class=MaterialExpressionMaterialFunctionCall
   inputs [
    NormalIn {from=MaterialExpressionStaticSwitch_1}
    NormalStrength {from=MaterialExpressionScalarParameter_27}
    ReFresnelVX {from=MaterialExpressionNamedRerouteUsage_252}
   ]
   properties {materialFunction=&MF_NormalStrength}
  }
  MaterialExpressionMaterialFunctionCall_59 {
   class=MaterialExpressionMaterialFunctionCall
   inputs [Color {from=MaterialExpressionStaticSwitchParameter_71}]
   properties {materialFunction=&SplitComponents}
  }
  MaterialExpressionMaterialFunctionCall_6 {
   class=MaterialExpressionMaterialFunctionCall
   inputs [
    Coordinates {from=MaterialExpressionNamedRerouteUsage_32}
    TexObject {from=MaterialExpressionTextureObjectParameter_20}
    DDX {from=MaterialExpressionNamedRerouteUsage_260}
    DDY {from=MaterialExpressionNamedRerouteUsage_261}
    Mask {from=""}
    DefaultUV {from=""}
   ]
   properties {materialFunction=&MF_TexPointSample_RGB}
  }
  MaterialExpressionMaterialFunctionCall_7 {
   class=MaterialExpressionMaterialFunctionCall
   inputs [
    AxF_Scale_X {from=MaterialExpressionNamedRerouteUsage_97}
    AxF_Scale_Y {from=MaterialExpressionNamedRerouteUsage_98}
    AxF_UnitScale {from=MaterialExpressionNamedRerouteUsage_99}
    Tile U {from=MaterialExpressionNamedRerouteUsage_93}
    Tile V {from=MaterialExpressionNamedRerouteUsage_94}
    Offset U {from=MaterialExpressionNamedRerouteUsage_95}
    Offset V {from=MaterialExpressionNamedRerouteUsage_96}
    Rotation {from=MaterialExpressionNamedRerouteUsage_92}
    UV Index {from=MaterialExpressionScalarParameter_16}
   ]
   properties {materialFunction=/InterchangeAxfAssets/MaterialFunctions/MF_AxFMakeCoordinatesUV}
  }
  MaterialExpressionMaterialFunctionCall_72 {
   class=MaterialExpressionMaterialFunctionCall
   inputs []
   properties {materialFunction=&CameraDirectionVector}
  }
  MaterialExpressionMaterialFunctionCall_74 {
   class=MaterialExpressionMaterialFunctionCall
   inputs [
    NormalTexOBJ {from=MaterialExpressionTextureObjectParameter_10}
    UV_TC {from=MaterialExpressionNamedRerouteUsage_105}
    Rotation_TC {from=MaterialExpressionNamedRerouteUsage_269}
   ]
   properties {materialFunction=&MF_TexCoodUVNormalSampler}
  }
  MaterialExpressionMaterialFunctionCall_8 {
   class=MaterialExpressionMaterialFunctionCall
   inputs [
    AxF_Scale_X {from=MaterialExpressionNamedRerouteUsage_56}
    AxF_Scale_Y {from=MaterialExpressionNamedRerouteUsage_59}
    AxF_UnitScale {from=MaterialExpressionNamedRerouteUsage_60}
    Tile XYZ {from=MaterialExpressionNamedRerouteUsage_61}
    Offset XYZ {from=MaterialExpressionNamedRerouteUsage_63}
    Transition Softness {from=MaterialExpressionNamedRerouteUsage_100}
    Coordinate Space {from=MaterialExpressionStaticSwitchParameter_2}
    Rotation {from=MaterialExpressionNamedRerouteUsage_62}
   ]
   properties {materialFunction=/InterchangeAxfAssets/MaterialFunctions/MF_AxFMakeCoordinatesTP}
  }
  MaterialExpressionMaterialFunctionCall_9 {
   class=MaterialExpressionMaterialFunctionCall
   inputs [
    Coordinates {from=MaterialExpressionNamedRerouteUsage_30}
    TexObject {from=MaterialExpressionTextureObjectParameter_10}
    DDX {from=MaterialExpressionNamedRerouteUsage_264}
    DDY {from=MaterialExpressionNamedRerouteUsage_265}
    Mask {from=""}
    DefaultUV {from=""}
   ]
   properties {materialFunction=&MF_TexPointSample_NRM}
  }
  MaterialExpressionMultiply_2 {
   class=MaterialExpressionMultiply
   inputs [
    A {from=MaterialExpressionStaticSwitchParameter_75}
    B {from=MaterialExpressionMaterialFunctionCall_1, output=OneOverMSTerm}
   ]
  }
  MaterialExpressionMultiply_3 {
   class=MaterialExpressionMultiply
   inputs [
    A {from=MaterialExpressionMaterialFunctionCall_1, output=OneOverMSTerm}
    B {from=MaterialExpressionStaticSwitchParameter_76}
   ]
  }
  MaterialExpressionMultiply_4 {
   class=MaterialExpressionMultiply
   inputs [
    A {from=MaterialExpressionScalarParameter_15}
    B {from=""}
   ]
   properties {constB=0.5}
  }
  MaterialExpressionMultiply_5 {
   class=MaterialExpressionMultiply
   inputs [
    A {from=MaterialExpressionScalarParameter_14}
    B {from=""}
   ]
   properties {constB=10}
  }
  MaterialExpressionNamedRerouteDeclaration_0 {
   class=MaterialExpressionNamedRerouteDeclaration
   inputs [None {from=MaterialExpressionMaterialFunctionCall_17, output=Top Layer Slab}]
   properties {name=ClearCoatTerm_O, nodeColor=<rgba 1 0.6797 0 1>}
  }
  MaterialExpressionNamedRerouteDeclaration_1 {
   class=MaterialExpressionNamedRerouteDeclaration
   inputs [None {from=MaterialExpressionStaticSwitch_3}]
   properties {name=GlintUVs_O, nodeColor=<rgba 0.5469 0 1 1>}
  }
  MaterialExpressionNamedRerouteDeclaration_11 {
   class=MaterialExpressionNamedRerouteDeclaration
   inputs [None {from=MaterialExpressionAdd_0}]
   properties {name=GlintsDensity_O, nodeColor=<rgba 0 1 0.2266 1>}
  }
  MaterialExpressionNamedRerouteDeclaration_116 {
   class=MaterialExpressionNamedRerouteDeclaration
   inputs [None {from=MaterialExpressionStaticSwitchParameter_67}]
   properties {name=DiffuseAlbedoTerm_O, nodeColor=<rgba 0 1 0.0391 1>}
  }
  MaterialExpressionNamedRerouteDeclaration_117 {
   class=MaterialExpressionNamedRerouteDeclaration
   inputs [None {from=MaterialExpressionMultiply_2}]
   properties {name=F0_O, nodeColor=<rgba 1 0 0.2812 1>}
  }
  MaterialExpressionNamedRerouteDeclaration_118 {
   class=MaterialExpressionNamedRerouteDeclaration
   inputs [None {from=MaterialExpressionMultiply_3}]
   properties {name=F90_O, nodeColor=<rgba 1 0 0.2812 1>}
  }
  MaterialExpressionNamedRerouteDeclaration_119 {
   class=MaterialExpressionNamedRerouteDeclaration
   inputs [None {from=MaterialExpressionSquareRoot_2}]
   properties {name=Roughness_O, nodeColor=<rgba 0.9219 1 0 1>}
  }
  MaterialExpressionNamedRerouteDeclaration_125 {
   class=MaterialExpressionNamedRerouteDeclaration
   inputs [None {from=MaterialExpressionStaticSwitchParameter_69}]
   properties {name=Normal_O, nodeColor=<rgba 1 0.9844 0 1>}
  }
  MaterialExpressionNamedRerouteDeclaration_129 {
   class=MaterialExpressionNamedRerouteDeclaration
   inputs [None {from=MaterialExpressionMaterialFunctionCall_8, output=Turns}]
   properties {name=TP_Rotation, nodeColor=<rgba 0.9453 1 0 1>}
  }
  MaterialExpressionNamedRerouteDeclaration_13 {
   class=MaterialExpressionNamedRerouteDeclaration
   inputs [None {from=MaterialExpressionConvert_0}]
   properties {name=TP_DDY, nodeColor=<rgba 0 0.7891 1 1>}
  }
  MaterialExpressionNamedRerouteDeclaration_15 {
   class=MaterialExpressionNamedRerouteDeclaration
   inputs [None {from=MaterialExpressionOneMinus_0}]
   properties {name=GGXGlintMix_O, nodeColor=<rgba 0 0.2969 1 1>}
  }
  MaterialExpressionNamedRerouteDeclaration_16 {
   class=MaterialExpressionNamedRerouteDeclaration
   inputs [None {from=MaterialExpressionConvert_0}]
   properties {name=TP_DDX, nodeColor=<rgba 0.5703 0 1 1>}
  }
  MaterialExpressionNamedRerouteDeclaration_2 {
   class=MaterialExpressionNamedRerouteDeclaration
   inputs [None {from=MaterialExpressionDivide_3}]
   properties {name=AxF_ScaleX, nodeColor=<rgba 0.7605 0 0.0009 1>}
  }
  MaterialExpressionNamedRerouteDeclaration_25 {
   class=MaterialExpressionNamedRerouteDeclaration
   inputs [None {from=MaterialExpressionConvert_1}]
   properties {name=TP_TileXYZ, nodeColor=<rgba 1 0.8438 0 1>}
  }
  MaterialExpressionNamedRerouteDeclaration_26 {
   class=MaterialExpressionNamedRerouteDeclaration
   inputs [None {from=MaterialExpressionConvert_2}]
   properties {name=TP_OffsetXYZ, nodeColor=<rgba 0 1 0.5078 1>}
  }
  MaterialExpressionNamedRerouteDeclaration_27 {
   class=MaterialExpressionNamedRerouteDeclaration
   inputs [None {from=MaterialExpressionSaturate_0}]
   properties {name=ReFresnelVX, nodeColor=<rgba 0.2656 0 1 1>}
  }
  MaterialExpressionNamedRerouteDeclaration_28 {
   class=MaterialExpressionNamedRerouteDeclaration
   inputs [None {from=MaterialExpressionMaterialFunctionCall_8, output=T}]
   properties {name=TP_Tangent, nodeColor=<rgba 1 0 0.0937 1>}
  }
  MaterialExpressionNamedRerouteDeclaration_29 {
   class=MaterialExpressionNamedRerouteDeclaration
   inputs [None {from=MaterialExpressionMaterialFunctionCall_8, output=B}]
   properties {name=TP_Bitangent, nodeColor=<rgba 0 0.4141 1 1>}
  }
  MaterialExpressionNamedRerouteDeclaration_3 {
   class=MaterialExpressionNamedRerouteDeclaration
   inputs [None {from=MaterialExpressionSubstrateSlabBSDF_3}]
   properties {name=GlintLayer_O, nodeColor=<rgba 0.0781 1 0 1>}
  }
  MaterialExpressionNamedRerouteDeclaration_32 {
   class=MaterialExpressionNamedRerouteDeclaration
   inputs [None {from=MaterialExpressionMaterialFunctionCall_7, output=Output UV}]
   properties {name=TC_UV, nodeColor=<rgba 0 1 0.25 1>}
  }
  MaterialExpressionNamedRerouteDeclaration_34 {
   class=MaterialExpressionNamedRerouteDeclaration
   inputs [None {from=MaterialExpressionStaticBoolParameter_1}]
   properties {name=UseGeoUVs, nodeColor=<rgba 0 0.8359 1 1>}
  }
  MaterialExpressionNamedRerouteDeclaration_35 {
   class=MaterialExpressionNamedRerouteDeclaration
   inputs [None {from=MaterialExpressionScalarParameter_17}]
   properties {name=TransitionSoftness, nodeColor=<rgba 0 1 0.2031 1>}
  }
  MaterialExpressionNamedRerouteDeclaration_4 {
   class=MaterialExpressionNamedRerouteDeclaration
   inputs [None {from=MaterialExpressionMaterialFunctionCall_8, output=Output UV}]
   properties {name=TP_UV, nodeColor=<rgba 0 0.1094 1 1>}
  }
  MaterialExpressionNamedRerouteDeclaration_49 {
   class=MaterialExpressionNamedRerouteDeclaration
   inputs [None {from=MaterialExpressionScalarParameter_2}]
   properties {name=UV_Scale, nodeColor=<rgba 0.7578 1 0 1>}
  }
  MaterialExpressionNamedRerouteDeclaration_50 {
   class=MaterialExpressionNamedRerouteDeclaration
   inputs [None {from=MaterialExpressionScalarParameter_28}]
   properties {name=UV_Rot, nodeColor=<rgba 0.7578 1 0 1>}
  }
  MaterialExpressionNamedRerouteDeclaration_51 {
   class=MaterialExpressionNamedRerouteDeclaration
   inputs [None {from=MaterialExpressionScalarParameter_3}]
   properties {name=UV_X, nodeColor=<rgba 0 1 0.6016 1>}
  }
  MaterialExpressionNamedRerouteDeclaration_52 {
   class=MaterialExpressionNamedRerouteDeclaration
   inputs [None {from=MaterialExpressionScalarParameter_4}]
   properties {name=UV_Y, nodeColor=<rgba 0 1 0.6016 1>}
  }
  MaterialExpressionNamedRerouteDeclaration_53 {
   class=MaterialExpressionNamedRerouteDeclaration
   inputs [None {from=MaterialExpressionScalarParameter_10}]
   properties {name=UV_Z, nodeColor=<rgba 0 1 0.6016 1>}
  }
  MaterialExpressionNamedRerouteDeclaration_54 {
   class=MaterialExpressionNamedRerouteDeclaration
   inputs [None {from=MaterialExpressionScalarParameter_6}]
   properties {name=OffsetX, nodeColor=<rgba 0.3594 1 0 1>}
  }
  MaterialExpressionNamedRerouteDeclaration_55 {
   class=MaterialExpressionNamedRerouteDeclaration
   inputs [None {from=MaterialExpressionScalarParameter_5}]
   properties {name=OffsetY, nodeColor=<rgba 0.3594 1 0 1>}
  }
  MaterialExpressionNamedRerouteDeclaration_56 {
   class=MaterialExpressionNamedRerouteDeclaration
   inputs [None {from=MaterialExpressionScalarParameter_11}]
   properties {name=OffsetZ, nodeColor=<rgba 0.3594 1 0 1>}
  }
  MaterialExpressionNamedRerouteDeclaration_59 {
   class=MaterialExpressionNamedRerouteDeclaration
   inputs [None {from=MaterialExpressionStaticSwitchParameter_29}]
   properties {name=CCNormal, nodeColor=<rgba 0.5469 0 1 1>}
  }
  MaterialExpressionNamedRerouteDeclaration_6 {
   class=MaterialExpressionNamedRerouteDeclaration
   inputs [None {from=MaterialExpressionDivide_4}]
   properties {name=AxF_ScaleY, nodeColor=<rgba 0.7605 0 0.0009 1>}
  }
  MaterialExpressionNamedRerouteDeclaration_8 {
   class=MaterialExpressionNamedRerouteDeclaration
   inputs [None {from=MaterialExpressionMultiply_5}]
   properties {name=AxF_UnitScale_MM, nodeColor=<rgba 0.7605 0 0.0009 1>}
  }
  MaterialExpressionNamedRerouteDeclaration_9 {
   class=MaterialExpressionNamedRerouteDeclaration
   inputs [None {from=MaterialExpressionScalarParameter_13}]
   properties {name=GlintRoughness_O, nodeColor=<rgba 0 1 0.0391 1>}
  }
  MaterialExpressionNamedRerouteUsage_0 {class=MaterialExpressionNamedRerouteUsage}
  MaterialExpressionNamedRerouteUsage_100 {class=MaterialExpressionNamedRerouteUsage}
  MaterialExpressionNamedRerouteUsage_101 {class=MaterialExpressionNamedRerouteUsage}
  MaterialExpressionNamedRerouteUsage_102 {class=MaterialExpressionNamedRerouteUsage}
  MaterialExpressionNamedRerouteUsage_103 {class=MaterialExpressionNamedRerouteUsage}
  MaterialExpressionNamedRerouteUsage_104 {class=MaterialExpressionNamedRerouteUsage}
  MaterialExpressionNamedRerouteUsage_105 {class=MaterialExpressionNamedRerouteUsage}
  MaterialExpressionNamedRerouteUsage_106 {class=MaterialExpressionNamedRerouteUsage}
  MaterialExpressionNamedRerouteUsage_107 {class=MaterialExpressionNamedRerouteUsage}
  MaterialExpressionNamedRerouteUsage_108 {class=MaterialExpressionNamedRerouteUsage}
  MaterialExpressionNamedRerouteUsage_109 {class=MaterialExpressionNamedRerouteUsage}
  MaterialExpressionNamedRerouteUsage_11 {class=MaterialExpressionNamedRerouteUsage}
  MaterialExpressionNamedRerouteUsage_110 {class=MaterialExpressionNamedRerouteUsage}
  MaterialExpressionNamedRerouteUsage_111 {class=MaterialExpressionNamedRerouteUsage}
  MaterialExpressionNamedRerouteUsage_112 {class=MaterialExpressionNamedRerouteUsage}
  MaterialExpressionNamedRerouteUsage_113 {class=MaterialExpressionNamedRerouteUsage}
  MaterialExpressionNamedRerouteUsage_114 {class=MaterialExpressionNamedRerouteUsage}
  MaterialExpressionNamedRerouteUsage_12 {class=MaterialExpressionNamedRerouteUsage}
  MaterialExpressionNamedRerouteUsage_128 {class=MaterialExpressionNamedRerouteUsage}
  MaterialExpressionNamedRerouteUsage_13 {class=MaterialExpressionNamedRerouteUsage}
  MaterialExpressionNamedRerouteUsage_15 {class=MaterialExpressionNamedRerouteUsage}
  MaterialExpressionNamedRerouteUsage_17 {class=MaterialExpressionNamedRerouteUsage}
  MaterialExpressionNamedRerouteUsage_18 {class=MaterialExpressionNamedRerouteUsage}
  MaterialExpressionNamedRerouteUsage_19 {class=MaterialExpressionNamedRerouteUsage}
  MaterialExpressionNamedRerouteUsage_2 {class=MaterialExpressionNamedRerouteUsage}
  MaterialExpressionNamedRerouteUsage_20 {class=MaterialExpressionNamedRerouteUsage}
  MaterialExpressionNamedRerouteUsage_23 {class=MaterialExpressionNamedRerouteUsage}
  MaterialExpressionNamedRerouteUsage_24 {class=MaterialExpressionNamedRerouteUsage}
  MaterialExpressionNamedRerouteUsage_244 {class=MaterialExpressionNamedRerouteUsage}
  MaterialExpressionNamedRerouteUsage_245 {class=MaterialExpressionNamedRerouteUsage}
  MaterialExpressionNamedRerouteUsage_246 {class=MaterialExpressionNamedRerouteUsage}
  MaterialExpressionNamedRerouteUsage_247 {class=MaterialExpressionNamedRerouteUsage}
  MaterialExpressionNamedRerouteUsage_250 {class=MaterialExpressionNamedRerouteUsage}
  MaterialExpressionNamedRerouteUsage_251 {class=MaterialExpressionNamedRerouteUsage}
  MaterialExpressionNamedRerouteUsage_252 {class=MaterialExpressionNamedRerouteUsage}
  MaterialExpressionNamedRerouteUsage_255 {class=MaterialExpressionNamedRerouteUsage}
  MaterialExpressionNamedRerouteUsage_256 {class=MaterialExpressionNamedRerouteUsage}
  MaterialExpressionNamedRerouteUsage_257 {class=MaterialExpressionNamedRerouteUsage}
  MaterialExpressionNamedRerouteUsage_258 {class=MaterialExpressionNamedRerouteUsage}
  MaterialExpressionNamedRerouteUsage_259 {class=MaterialExpressionNamedRerouteUsage}
  MaterialExpressionNamedRerouteUsage_26 {class=MaterialExpressionNamedRerouteUsage}
  MaterialExpressionNamedRerouteUsage_260 {class=MaterialExpressionNamedRerouteUsage}
  MaterialExpressionNamedRerouteUsage_261 {class=MaterialExpressionNamedRerouteUsage}
  MaterialExpressionNamedRerouteUsage_264 {class=MaterialExpressionNamedRerouteUsage}
  MaterialExpressionNamedRerouteUsage_265 {class=MaterialExpressionNamedRerouteUsage}
  MaterialExpressionNamedRerouteUsage_268 {class=MaterialExpressionNamedRerouteUsage}
  MaterialExpressionNamedRerouteUsage_269 {class=MaterialExpressionNamedRerouteUsage}
  MaterialExpressionNamedRerouteUsage_3 {class=MaterialExpressionNamedRerouteUsage}
  MaterialExpressionNamedRerouteUsage_30 {class=MaterialExpressionNamedRerouteUsage}
  MaterialExpressionNamedRerouteUsage_32 {class=MaterialExpressionNamedRerouteUsage}
  MaterialExpressionNamedRerouteUsage_33 {class=MaterialExpressionNamedRerouteUsage}
  MaterialExpressionNamedRerouteUsage_34 {class=MaterialExpressionNamedRerouteUsage}
  MaterialExpressionNamedRerouteUsage_35 {class=MaterialExpressionNamedRerouteUsage}
  MaterialExpressionNamedRerouteUsage_4 {class=MaterialExpressionNamedRerouteUsage}
  MaterialExpressionNamedRerouteUsage_52 {class=MaterialExpressionNamedRerouteUsage}
  MaterialExpressionNamedRerouteUsage_53 {class=MaterialExpressionNamedRerouteUsage}
  MaterialExpressionNamedRerouteUsage_54 {class=MaterialExpressionNamedRerouteUsage}
  MaterialExpressionNamedRerouteUsage_55 {class=MaterialExpressionNamedRerouteUsage}
  MaterialExpressionNamedRerouteUsage_56 {class=MaterialExpressionNamedRerouteUsage}
  MaterialExpressionNamedRerouteUsage_57 {class=MaterialExpressionNamedRerouteUsage}
  MaterialExpressionNamedRerouteUsage_58 {class=MaterialExpressionNamedRerouteUsage}
  MaterialExpressionNamedRerouteUsage_59 {class=MaterialExpressionNamedRerouteUsage}
  MaterialExpressionNamedRerouteUsage_60 {class=MaterialExpressionNamedRerouteUsage}
  MaterialExpressionNamedRerouteUsage_61 {class=MaterialExpressionNamedRerouteUsage}
  MaterialExpressionNamedRerouteUsage_62 {class=MaterialExpressionNamedRerouteUsage}
  MaterialExpressionNamedRerouteUsage_63 {class=MaterialExpressionNamedRerouteUsage}
  MaterialExpressionNamedRerouteUsage_64 {class=MaterialExpressionNamedRerouteUsage}
  MaterialExpressionNamedRerouteUsage_65 {class=MaterialExpressionNamedRerouteUsage}
  MaterialExpressionNamedRerouteUsage_66 {class=MaterialExpressionNamedRerouteUsage}
  MaterialExpressionNamedRerouteUsage_67 {class=MaterialExpressionNamedRerouteUsage}
  MaterialExpressionNamedRerouteUsage_9 {class=MaterialExpressionNamedRerouteUsage}
  MaterialExpressionNamedRerouteUsage_92 {class=MaterialExpressionNamedRerouteUsage}
  MaterialExpressionNamedRerouteUsage_93 {class=MaterialExpressionNamedRerouteUsage}
  MaterialExpressionNamedRerouteUsage_94 {class=MaterialExpressionNamedRerouteUsage}
  MaterialExpressionNamedRerouteUsage_95 {class=MaterialExpressionNamedRerouteUsage}
  MaterialExpressionNamedRerouteUsage_96 {class=MaterialExpressionNamedRerouteUsage}
  MaterialExpressionNamedRerouteUsage_97 {class=MaterialExpressionNamedRerouteUsage}
  MaterialExpressionNamedRerouteUsage_98 {class=MaterialExpressionNamedRerouteUsage}
  MaterialExpressionNamedRerouteUsage_99 {class=MaterialExpressionNamedRerouteUsage}
  MaterialExpressionOneMinus_0 {class=MaterialExpressionOneMinus, inputs [None {from=MaterialExpressionMultiply_4}]}
  MaterialExpressionSaturate_0 {class=MaterialExpressionSaturate, inputs [None {from=MaterialExpressionAbs_0}]}
  MaterialExpressionScalarParameter_0 {
   class=MaterialExpressionScalarParameter
   properties {defaultValue=1, group=ClearCoat, parameterName=ClearCoatNormalStrength, sliderMax=2, sliderMin=-1}
  }
  MaterialExpressionScalarParameter_1 {class=MaterialExpressionScalarParameter, properties {defaultValue=25, group=Glints, parameterName=SparkleDensityMaxValue}}
  MaterialExpressionScalarParameter_10 {
   class=MaterialExpressionScalarParameter
   properties {defaultValue=1, group=UV Transforms, parameterName=StretchOnZ, sliderMax=1000, sliderMin=0.001, sortPriority=120}
  }
  MaterialExpressionScalarParameter_11 {class=MaterialExpressionScalarParameter, properties {group=UV Transforms, parameterName=OffsetOnZ, sortPriority=150}}
  MaterialExpressionScalarParameter_12 {class=MaterialExpressionScalarParameter, properties {defaultValue=100, group=MillimeterSettings, parameterName=HeightMM, sortPriority=11}}
  MaterialExpressionScalarParameter_13 {class=MaterialExpressionScalarParameter, properties {group=Glints, parameterName=GlintRoughness}}
  MaterialExpressionScalarParameter_14 {class=MaterialExpressionScalarParameter, properties {defaultValue=1, group=MillimeterSettings, parameterName=UnitScaleCM, sortPriority=12}}
  MaterialExpressionScalarParameter_15 {class=MaterialExpressionScalarParameter, properties {defaultValue=1, group=Glints, parameterName=GlintBrightness, sliderMax=1}}
  MaterialExpressionScalarParameter_16 {class=MaterialExpressionScalarParameter, properties {group=UV Transforms, parameterName=UV Index}}
  MaterialExpressionScalarParameter_17 {
   class=MaterialExpressionScalarParameter
   properties {
    defaultValue=0.5
    desc="Controls the contrast of the blend mask. Internally applied as pow(mask, 8 × BlendFactor). A default of 0.5 corresponds to pow4 contrast."
    group=UV Transforms
    parameterName=TransitionSoftness
    sliderMax=1
    sortPriority=3
   }
  }
  MaterialExpressionScalarParameter_2 {
   class=MaterialExpressionScalarParameter
   properties {defaultValue=0.1, group=UV Transforms, parameterName=Scale, sliderMax=10000, sliderMin=0.001, sortPriority=10}
  }
  MaterialExpressionScalarParameter_22 {class=MaterialExpressionScalarParameter, properties {group=Glints, parameterName=GlintDensity}}
  MaterialExpressionScalarParameter_27 {
   class=MaterialExpressionScalarParameter
   properties {defaultValue=1, group=Normal, parameterName=NormalStrength, sliderMax=2, sliderMin=-1, sortPriority=12}
  }
  MaterialExpressionScalarParameter_28 {class=MaterialExpressionScalarParameter, properties {group=UV Transforms, parameterName="Rotation (degree)", sliderMax=360, sortPriority=30}}
  MaterialExpressionScalarParameter_3 {
   class=MaterialExpressionScalarParameter
   properties {defaultValue=1, group=UV Transforms, parameterName=StretchOnX, sliderMax=100, sliderMin=0.001, sortPriority=100}
  }
  MaterialExpressionScalarParameter_4 {
   class=MaterialExpressionScalarParameter
   properties {defaultValue=1, group=UV Transforms, parameterName=StretchOnY, sliderMax=1000, sliderMin=0.001, sortPriority=110}
  }
  MaterialExpressionScalarParameter_5 {class=MaterialExpressionScalarParameter, properties {group=UV Transforms, parameterName=OffsetOnY, sortPriority=140}}
  MaterialExpressionScalarParameter_6 {class=MaterialExpressionScalarParameter, properties {group=UV Transforms, parameterName=OffsetOnX, sortPriority=130}}
  MaterialExpressionScalarParameter_9 {class=MaterialExpressionScalarParameter, properties {defaultValue=100, group=MillimeterSettings, parameterName=WidthMM, sortPriority=10}}
  MaterialExpressionShaderStageSwitch_2 {
   class=MaterialExpressionShaderStageSwitch
   inputs [
    PixelShader {from=MaterialExpressionCameraVectorWS_0, output=""}
    VertexShader {from=MaterialExpressionMaterialFunctionCall_72, output=Result}
   ]
   properties {}
  }
  MaterialExpressionSquareRoot_2 {class=MaterialExpressionSquareRoot, inputs [None {from=MaterialExpressionDivide_2, output=""}], properties {}}
  MaterialExpressionStaticBoolParameter_1 {
   class=MaterialExpressionStaticBoolParameter
   inputs []
   properties {defaultValue=true, group=UV Transforms, parameterName=Use Geo UVs, sortPriority=0}
  }
  MaterialExpressionStaticSwitchParameter_2 {
   class=MaterialExpressionStaticSwitchParameter
   inputs [
    True {from=MaterialExpressionConstant_2, output=""}
    False {from=MaterialExpressionConstant_3, output=""}
   ]
   properties {defaultValue=true, group=UV Transforms, parameterName=UsingLocalSpace, sortPriority=2}
  }
  MaterialExpressionStaticSwitchParameter_29 {
   class=MaterialExpressionStaticSwitchParameter
   inputs [
    True {from=MaterialExpressionMaterialFunctionCall_5, output=NormalOut}
    False {from=MaterialExpressionComponentMask_1, output=""}
   ]
   properties {group=ClearCoat, parameterName=HasClearCoatNormalTexture, sortPriority=30}
  }
  MaterialExpressionStaticSwitchParameter_66 {
   class=MaterialExpressionStaticSwitchParameter
   inputs [
    True {from=MaterialExpressionSubstrateVerticalLayering_2, output=""}
    False {from=MaterialExpressionNamedRerouteUsage_58, output=""}
   ]
   properties {group=ClearCoat, parameterName=HasClearCoat, sortPriority=30}
  }
  MaterialExpressionStaticSwitchParameter_67 {
   class=MaterialExpressionStaticSwitchParameter
   inputs [
    True {from=MaterialExpressionStaticSwitch_0, output=""}
    False {from=MaterialExpressionVectorParameter_18, output=RGB}
   ]
   properties {group=Diffuse, parameterName=HasDiffuseColorTexture}
  }
  MaterialExpressionStaticSwitchParameter_69 {
   class=MaterialExpressionStaticSwitchParameter
   inputs [
    True {from=MaterialExpressionMaterialFunctionCall_58, output=NormalOut}
    False {from=MaterialExpressionComponentMask_8, output=""}
   ]
   properties {group=Normal, parameterName=HasNormalTexture, sortPriority=30}
  }
  MaterialExpressionStaticSwitchParameter_71 {
   class=MaterialExpressionStaticSwitchParameter
   inputs [
    True {from=MaterialExpressionStaticSwitch_2, output=""}
    False {from=MaterialExpressionVectorParameter_20, output=RGB}
   ]
   properties {group=Roughness, parameterName=HasRoughnessTexture, sortPriority=30}
  }
  MaterialExpressionStaticSwitchParameter_75 {
   class=MaterialExpressionStaticSwitchParameter
   inputs [
    True {from=MaterialExpressionStaticSwitch_6, output=""}
    False {from=MaterialExpressionVectorParameter_21, output=RGB}
   ]
   properties {group=Fresnel, parameterName=HasF0Texture}
  }
  MaterialExpressionStaticSwitchParameter_76 {
   class=MaterialExpressionStaticSwitchParameter
   inputs [
    True {from=MaterialExpressionStaticSwitch_5, output=""}
    False {from=MaterialExpressionVectorParameter_22, output=RGB}
   ]
   properties {group=Fresnel, parameterName=HasF90Texture}
  }
  MaterialExpressionStaticSwitch_0 {
   class=MaterialExpressionStaticSwitch
   inputs [
    True {from=MaterialExpressionTextureSample_15, output=RGB}
    False {from=MaterialExpressionMaterialFunctionCall_0, output=RGB}
    Value {from=MaterialExpressionNamedRerouteUsage_108, output=""}
   ]
   properties {}
  }
  MaterialExpressionStaticSwitch_1 {
   class=MaterialExpressionStaticSwitch
   inputs [
    True {from=MaterialExpressionMaterialFunctionCall_57, output=Result}
    False {from=MaterialExpressionMaterialFunctionCall_11, output=TSN}
    Value {from=MaterialExpressionNamedRerouteUsage_109, output=""}
   ]
   properties {}
  }
  MaterialExpressionStaticSwitch_2 {
   class=MaterialExpressionStaticSwitch
   inputs [
    True {from=MaterialExpressionTextureSample_16, output=RGB}
    False {from=MaterialExpressionMaterialFunctionCall_6, output=RGB}
    Value {from=MaterialExpressionNamedRerouteUsage_110, output=""}
   ]
   properties {}
  }
  MaterialExpressionStaticSwitch_3 {
   class=MaterialExpressionStaticSwitch
   inputs [
    True {from=MaterialExpressionNamedRerouteUsage_101, output=""}
    False {from=MaterialExpressionNamedRerouteUsage_4, output=""}
    Value {from=MaterialExpressionNamedRerouteUsage_111, output=""}
   ]
   properties {}
  }
  MaterialExpressionStaticSwitch_4 {
   class=MaterialExpressionStaticSwitch
   inputs [
    True {from=MaterialExpressionMaterialFunctionCall_74, output=Result}
    False {from=MaterialExpressionMaterialFunctionCall_10, output=TSN}
    Value {from=MaterialExpressionNamedRerouteUsage_112, output=""}
   ]
   properties {}
  }
  MaterialExpressionStaticSwitch_5 {
   class=MaterialExpressionStaticSwitch
   inputs [
    True {from=MaterialExpressionTextureSample_18, output=RGB}
    False {from=MaterialExpressionMaterialFunctionCall_4, output=RGB}
    Value {from=MaterialExpressionNamedRerouteUsage_113, output=""}
   ]
   properties {}
  }
  MaterialExpressionStaticSwitch_6 {
   class=MaterialExpressionStaticSwitch
   inputs [
    True {from=MaterialExpressionTextureSample_19, output=RGB}
    False {from=MaterialExpressionMaterialFunctionCall_2, output=RGB}
    Value {from=MaterialExpressionNamedRerouteUsage_114, output=""}
   ]
   properties {}
  }
  MaterialExpressionSubstrateSlabBSDF_3 {
   class=MaterialExpressionSubstrateSlabBSDF
   inputs [
    Diffuse Albedo {from=MaterialExpressionNamedRerouteUsage_9, output=""}
    F0 {from=MaterialExpressionNamedRerouteUsage_11, output=""}
    F90 {from=MaterialExpressionNamedRerouteUsage_17, output=""}
    Roughness {from=MaterialExpressionNamedRerouteUsage_23, output=""}
    Anisotropy {from="", output=""}
    Normal {from=MaterialExpressionNamedRerouteUsage_15, output=""}
    Tangent {from="", output=""}
    SSS MFP {from="", output=""}
    SSS MFP Scale {from="", output=""}
    SSS Phase Anisotropy {from="", output=""}
    Emissive Color {from="", output=""}
    Second Roughness {from=MaterialExpressionNamedRerouteUsage_13, output=""}
    Second Roughness Weight {from=MaterialExpressionNamedRerouteUsage_24, output=""}
    Fuzz Roughness {from="", output=""}
    Fuzz Amount {from="", output=""}
    Fuzz Color {from="", output=""}
    "Glint Density (Disabled)" {from=MaterialExpressionNamedRerouteUsage_19, output=""}
    "Glint UVs (Disabled)" {from=MaterialExpressionNamedRerouteUsage_18, output=""}
   ]
   properties {specularProfile=&/DefaultSpecularProfile}
  }
  MaterialExpressionSubstrateVerticalLayering_2 {
   class=MaterialExpressionSubstrateVerticalLayering
   inputs [
    Top {from=MaterialExpressionNamedRerouteUsage_20, output=""}
    Bottom {from=MaterialExpressionNamedRerouteUsage_57, output=""}
    Top Thickness {from="", output=""}
   ]
   properties {}
  }
  MaterialExpressionTextureObjectParameter_1 {
   class=MaterialExpressionTextureObjectParameter
   inputs [
    UVs {from="", output=""}
    Apply View MipBias {from="", output=""}
   ]
   properties {group=Normal, parameterName=NormalTexture, samplerType=SAMPLERTYPE_Normal, sortPriority=1, texture=/Engine/EngineMaterials/DefaultNormal}
  }
  MaterialExpressionTextureObjectParameter_10 {
   class=MaterialExpressionTextureObjectParameter
   inputs [
    UVs {from="", output=""}
    Apply View MipBias {from="", output=""}
   ]
   properties {
    group=ClearCoat
    parameterName=ClearCoatNormalTexture
    samplerType=SAMPLERTYPE_Normal
    sortPriority=1
    texture=&DefaultNormal
   }
  }
  MaterialExpressionTextureObjectParameter_19 {
   class=MaterialExpressionTextureObjectParameter
   inputs [
    UVs {from="", output=""}
    Apply View MipBias {from="", output=""}
   ]
   properties {
    group=Diffuse
    parameterName=DiffuseColorTexture
    samplerType=SAMPLERTYPE_LinearColor
    sortPriority=1
    texture=/Engine/Functions/Engine_MaterialFunctions02/PivotPainter2/Black_1x1_EXR_Texture
   }
  }
  MaterialExpressionTextureObjectParameter_20 {
   class=MaterialExpressionTextureObjectParameter
   inputs [
    UVs {from="", output=""}
    Apply View MipBias {from="", output=""}
   ]
   properties {
    group=Roughness
    parameterName=RoughnessTexture
    samplerType=SAMPLERTYPE_LinearColor
    sortPriority=1
    texture=&Black_1x1_EXR_Texture
   }
  }
  MaterialExpressionTextureObjectParameter_22 {
   class=MaterialExpressionTextureObjectParameter
   inputs [
    UVs {from="", output=""}
    Apply View MipBias {from="", output=""}
   ]
   properties {
    group=Fresnel
    parameterName=F90Texture
    samplerType=SAMPLERTYPE_LinearColor
    sortPriority=2
    texture=&Black_1x1_EXR_Texture
   }
  }
  MaterialExpressionTextureObjectParameter_23 {
   class=MaterialExpressionTextureObjectParameter
   inputs [
    UVs {from="", output=""}
    Apply View MipBias {from="", output=""}
   ]
   properties {
    group=Fresnel
    parameterName=F0Texture
    samplerType=SAMPLERTYPE_LinearColor
    sortPriority=1
    texture=&Black_1x1_EXR_Texture
   }
  }
  MaterialExpressionTextureSample_15 {
   class=MaterialExpressionTextureSample
   inputs [
    UVs {from=MaterialExpressionNamedRerouteUsage_102}
    Tex {from=MaterialExpressionTextureObjectParameter_19}
    Apply View MipBias {from=""}
   ]
   properties {samplerType=SAMPLERTYPE_LinearColor}
  }
  MaterialExpressionTextureSample_16 {
   class=MaterialExpressionTextureSample
   inputs [
    UVs {from=MaterialExpressionNamedRerouteUsage_104}
    Tex {from=MaterialExpressionTextureObjectParameter_20}
    Apply View MipBias {from=""}
   ]
   properties {samplerType=SAMPLERTYPE_LinearColor}
  }
  MaterialExpressionTextureSample_18 {
   class=MaterialExpressionTextureSample
   inputs [
    UVs {from=MaterialExpressionNamedRerouteUsage_106}
    Tex {from=MaterialExpressionTextureObjectParameter_22}
    Apply View MipBias {from=""}
   ]
   properties {samplerType=SAMPLERTYPE_LinearColor}
  }
  MaterialExpressionTextureSample_19 {
   class=MaterialExpressionTextureSample
   inputs [
    UVs {from=MaterialExpressionNamedRerouteUsage_107}
    Tex {from=MaterialExpressionTextureObjectParameter_23}
    Apply View MipBias {from=""}
   ]
   properties {samplerType=SAMPLERTYPE_LinearColor}
  }
  MaterialExpressionVectorParameter_11 {class=MaterialExpressionVectorParameter, properties {group=ClearCoat, parameterName=ClearCoatNormal, defaultValue=<rgba 0 0 1 1>}}
  MaterialExpressionVectorParameter_18 {class=MaterialExpressionVectorParameter, properties {group=Diffuse, parameterName=DiffuseColor, defaultValue=<rgba 0 0 0 1>}}
  MaterialExpressionVectorParameter_19 {class=MaterialExpressionVectorParameter, properties {group=Normal, parameterName=Normal, defaultValue=<rgba 0 0 1 1>}}
  MaterialExpressionVectorParameter_20 {
   class=MaterialExpressionVectorParameter
   properties {group=Roughness, parameterName=Roughness, sortPriority=30, defaultValue=<rgba 0.2122 0.2122 0 1>}
  }
  MaterialExpressionVectorParameter_21 {class=MaterialExpressionVectorParameter, properties {group=Fresnel, parameterName=F0, defaultValue=<rgba 0.04 0.04 0.04 1>}}
  MaterialExpressionVectorParameter_22 {class=MaterialExpressionVectorParameter, properties {group=Fresnel, parameterName=F90, defaultValue=<rgba 1 1 1 0>}}
  MaterialExpressionVertexNormalWS_0 {class=MaterialExpressionVertexNormalWS, inputs [], properties {}}
 }
 properties {bUsedWithNanite=true, bUsedWithStaticLighting=true, shadingModel=MSM_SubsurfaceProfile, translucencyLightingMode=TLM_SurfacePerPixelLighting}
}
