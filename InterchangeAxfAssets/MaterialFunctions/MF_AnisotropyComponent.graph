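# Material function for Interchange AxF materials that converts anisotropic roughness and rotation inputs into Anisotropy and Tangent outputs.
# Anisotropy is (AlphaX − AlphaY) / (AlphaX + AlphaY), or a constant when the HasAnisotropy static switch is off.
# AnisotropicRotation is remapped and normalized as a double-angle cosine/sine vector, with a (1,0) fallback.
# Half-angle identities and a sign test recover theta's cosine and sine for the tangent, with Y negated.
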MaterialFunction /InterchangeAxfAssets/MaterialFunctions/MF_AnisotropyComponent
graph {nodes [
  n1 = NamedRerouteUsage()
  FunctionOutput(None=MaterialFunctionCall(X=NamedRerouteUsage(), Y=Multiply(A=@n1) {constB=-1}) {materialFunction=&MakeFloat3}.Result) {outputName=Tangent, sortPriority=1}
  n2 = FunctionInput() {inputName=AlphaX, inputType=FunctionInput_Scalar}
  n3 = FunctionInput() {inputName=AlphaY, inputType=FunctionInput_Scalar, sortPriority=10}
  n4 = Subtract(A=@n2, B=@n3)
  n5 = Add(A=@n2, B=@n3)
  FunctionOutput(None=StaticSwitchParameter(True=Divide(A=@n4, B=@n5), False=Constant()) {group=Anisotropy, parameterName=HasAnisotropy, sortPriority=30}) {outputName=Anisotropy}
  n6 = Subtract(A=Multiply(A=FunctionInput() {inputName=AnisotropicRotation, inputType=FunctionInput_Vector2, previewValue=<1 1 1 0>}) {constB=2})
  n7 = MaterialFunctionCall(X=Constant() {r=1}) {materialFunction=&MakeFloat2}
  n8 = If(A=NamedRerouteUsage(), "A > B"=Divide(A=@n6, B=NamedRerouteUsage()), "A == B"=@n7.Result, "A < B"=@n7.Result)
  NamedRerouteDeclaration(None=ComponentMask(None=@n8) {r=true}) {name="cos(2*theta)", nodeColor=<rgba 0.0078 1 0 1>}
  NamedRerouteDeclaration(None=MaterialFunctionCall(Vector 2=@n6) {materialFunction=/Engine/Functions/Engine_MaterialFunctions02/Utility/VectorLength}.V2 Length) {name=Norm, nodeColor=<rgba 1 0 0.0937 1>}
  n9 = Constant() {r=1}
  NamedRerouteDeclaration(None=If(A=NamedRerouteUsage(), "A > B"=@n9, "A == B"=@n9, "A < B"=Constant() {r=-1})) {name="sign(sin(2*theta))", nodeColor=<rgba 1 0 0.1406 1>}
  NamedRerouteDeclaration(None=SquareRoot(None=NamedRerouteUsage())) {name="cos(theta)", nodeColor=<rgba 1 0 0.0703 1>}
  NamedRerouteDeclaration(None=ComponentMask(None=@n8) {g=true}) {name="sin(2*theta)", nodeColor=<rgba 1 0 0.3281 1>}
  n10 = NamedRerouteUsage()
  NamedRerouteDeclaration(None=Divide(A=Add(A=@n10))) {name="cos^2(theta)", nodeColor=<rgba 0.9688 1 0 1>}
  NamedRerouteDeclaration(None=Subtract(B=NamedRerouteUsage())) {name="sin^2(theta)", nodeColor=<rgba 0.7109 0 1 1>}
  n11 = NamedRerouteUsage()
  NamedRerouteDeclaration(None=Multiply(A=NamedRerouteUsage(), B=SquareRoot(None=@n11))) {name="sin(theta)", nodeColor=<rgba 1 0.0703 0 1>}
 ]}
