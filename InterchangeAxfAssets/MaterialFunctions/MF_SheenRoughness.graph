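# MF_SheenRoughness is an AxF import material function that converts a SheenRoughness input into an IsotropicRoughness output.
# It splits the input, adds the R component to itself and halves it, so only R contributes.
# The UseCharlieSheenFitting static switch, on by default, remaps that value through a fitted sigmoid-style curve with hard-coded constants.
# With the switch off, the unfitted value passes straight through.
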MaterialFunction /InterchangeAxfAssets/MaterialFunctions/MF_SheenRoughness
graph {nodes [
  n1 = Constant() {r=2.2935}
  n2 = MaterialFunctionCall(Color=FunctionInput() {inputName=SheenRoughness}) {materialFunction=&SplitComponents}
  n3 = Divide(A=Add(A=@n2.R, B=@n2.R))
  n4 = Divide(A=@n3, B=Constant() {r=1.2363})
  n5 = Constant() {r=1.5842}
  n6 = Divide(A=Subtract(A=Constant() {r=0.0545}, B=@n1), B=Add(A=Power(Base=@n4, Exp=@n5)))
  FunctionOutput(None=StaticSwitchParameter(True=Add(A=@n6, B=@n1), False=@n3) {defaultValue=true, group=Sheen, parameterName=UseCharlieSheenFitting}) {outputName=IsotropicRoughness}
 ]}
