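# Base Substrate material for AxF SVBRDF assets, with static switches choosing texture or constant for each channel.
# Coordinates use millimeter-based AxF scale, sampled through mesh UVs or triplanar projection with explicit derivatives and reoriented normals.
# Diffuse, F0/F90, normal, anisotropic roughness and optional sheen fuzz feed a slab, optionally vertically layered under clear coat.
# Alpha X/Y become roughness, anisotropy and tangent; a view-angle term drives normal-strength adjustment.
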Material /InterchangeAxfAssets/BaseMaterials/SVBRDF
graph {
 expressions {
  MaterialExpressionAbs_0 {class=MaterialExpressionAbs, inputs [None {from=MaterialExpressionDotProduct_0, output=""}], properties {}}
  MaterialExpressionAdd_2 {
   class=MaterialExpressionAdd
   inputs [
    A {from=MaterialExpressionMaterialFunctionCall_59, output=R}
    B {from=MaterialExpressionMaterialFunctionCall_59, output=R}
   ]
  }
  MaterialExpressionCameraVectorWS_0 {class=MaterialExpressionCameraVectorWS, inputs [], properties {}}
  MaterialExpressionComponentMask_1 {
   class=MaterialExpressionComponentMask
   inputs [None {from=MaterialExpressionVectorParameter_11, output=RGB}]
   properties {b=true, g=true, r=true}
  }
  MaterialExpressionComponentMask_8 {
   class=MaterialExpressionComponentMask
   inputs [None {from=MaterialExpressionVectorParameter_19, output=RGB}]
   properties {b=true, g=true, r=true}
  }
  MaterialExpressionConstant_15 {class=MaterialExpressionConstant}
  MaterialExpressionConstant_16 {class=MaterialExpressionConstant}
  MaterialExpressionConstant_17 {class=MaterialExpressionConstant}
  MaterialExpressionConstant_18 {class=MaterialExpressionConstant, properties {r=1}}
  MaterialExpressionConstant_2 {class=MaterialExpressionConstant, properties {desc="1 = local", r=1}}
  MaterialExpressionConstant_3 {class=MaterialExpressionConstant, properties {desc="0 = world"}}
  MaterialExpressionConvert_0 {
   class=MaterialExpressionConvert
   inputs [None {from=MaterialExpressionMaterialFunctionCall_8, output=DXY}]
   properties {
    convertInputs [{
      type=Vector4
      defaultValue=<rgba 0 0 0 1>
      expressionInput {
       expression=&self:MaterialExpressionMaterialFunctionCall_8
       inputName=None
       mask=0
       maskA=0
       maskB=0
       maskG=0
       maskR=0
       outputIndex=4
      }
     }]
    convertOutputs [{type=Vector2}, {type=Vector2}]
   }
  }
  MaterialExpressionConvert_1 {
   class=MaterialExpressionConvert
   inputs [
    None {from=MaterialExpressionNamedRerouteUsage_76}
    None {from=MaterialExpressionNamedRerouteUsage_77}
    None {from=MaterialExpressionNamedRerouteUsage_78}
    None {from=MaterialExpressionNamedRerouteUsage_70}
   ]
   properties {
    convertInputs [
     {
      type=Scalar
      expressionInput {
       expression=&self:MaterialExpressionNamedRerouteUsage_76
       inputName=None
       mask=0
       maskA=0
       maskB=0
       maskG=0
       maskR=0
       outputIndex=0
      }
     }
     {
      type=Scalar
      expressionInput {
       expression=&self:MaterialExpressionNamedRerouteUsage_77
       inputName=None
       mask=0
       maskA=0
       maskB=0
       maskG=0
       maskR=0
       outputIndex=0
      }
     }
     {
      type=Scalar
      expressionInput {
       expression=&self:MaterialExpressionNamedRerouteUsage_78
       inputName=None
       mask=0
       maskA=0
       maskB=0
       maskG=0
       maskR=0
       outputIndex=0
      }
     }
     {
      type=Scalar
      expressionInput {
       expression=&self:MaterialExpressionNamedRerouteUsage_70
       inputName=None
       mask=0
       maskA=0
       maskB=0
       maskG=0
       maskR=0
       outputIndex=0
      }
     }
    ]
    convertOutputs [{type=Vector4, defaultValue=<rgba 0 0 0 1>}]
   }
  }
  MaterialExpressionConvert_2 {
   class=MaterialExpressionConvert
   inputs [
    None {from=MaterialExpressionNamedRerouteUsage_33}
    None {from=MaterialExpressionNamedRerouteUsage_34}
    None {from=MaterialExpressionNamedRerouteUsage_35}
   ]
   properties {
    convertInputs [
     {
      type=Scalar
      expressionInput {
       expression=&self:MaterialExpressionNamedRerouteUsage_33
       inputName=None
       mask=0
       maskA=0
       maskB=0
       maskG=0
       maskR=0
       outputIndex=0
      }
     }
     {
      type=Scalar
      expressionInput {
       expression=&self:MaterialExpressionNamedRerouteUsage_34
       inputName=None
       mask=0
       maskA=0
       maskB=0
       maskG=0
       maskR=0
       outputIndex=0
      }
     }
     {
      type=Scalar
      expressionInput {
       expression=&self:MaterialExpressionNamedRerouteUsage_35
       inputName=None
       mask=0
       maskA=0
       maskB=0
       maskG=0
       maskR=0
       outputIndex=0
      }
     }
    ]
    convertOutputs [{type=Vector3}]
   }
  }
  MaterialExpressionDivide_0 {
   class=MaterialExpressionDivide
   inputs [
    A {from=""}
    B {from=MaterialExpressionScalarParameter_1}
   ]
   properties {constA=10, constB=10}
  }
  MaterialExpressionDivide_1 {
   class=MaterialExpressionDivide
   inputs [
    A {from=""}
    B {from=MaterialExpressionScalarParameter_9}
   ]
   properties {constA=10, constB=10}
  }
  MaterialExpressionDivide_2 {
   class=MaterialExpressionDivide
   inputs [
    A {from=MaterialExpressionAdd_2}
    B {from=""}
   ]
  }
  MaterialExpressionDotProduct_0 {
   class=MaterialExpressionDotProduct
   inputs [
    A {from=MaterialExpressionShaderStageSwitch_2, output=""}
    B {from=MaterialExpressionVertexNormalWS_0, output=""}
   ]
   properties {}
  }
  MaterialExpressionMaterialFunctionCall_0 {
   class=MaterialExpressionMaterialFunctionCall
   inputs [
    Coordinates {from=MaterialExpressionNamedRerouteUsage_0}
    TexObject {from=MaterialExpressionTextureObjectParameter_19}
    DDX {from=MaterialExpressionNamedRerouteUsage_244}
    DDY {from=MaterialExpressionNamedRerouteUsage_245}
    Mask {from=""}
    DefaultUV {from=""}
   ]
   properties {materialFunction=/BaseMaterial/Materials/Functions/MF_TexPointSample_RGB}
  }
  MaterialExpressionMaterialFunctionCall_1 {
   class=MaterialExpressionMaterialFunctionCall
   inputs [
    NormalIn {from=MaterialExpressionStaticSwitch_6}
    NormalStrength {from=MaterialExpressionScalarParameter_0}
    ReFresnelVX {from=MaterialExpressionNamedRerouteUsage_25}
   ]
   properties {materialFunction=/InterchangeAxfAssets/MaterialFunctions/MF_NormalStrength}
  }
  MaterialExpressionMaterialFunctionCall_12 {
   class=MaterialExpressionMaterialFunctionCall
   inputs [
    Coordinates {from=MaterialExpressionNamedRerouteUsage_26}
    TexObject {from=MaterialExpressionTextureObjectParameter_22}
    DDX {from=MaterialExpressionNamedRerouteUsage_257}
    DDY {from=MaterialExpressionNamedRerouteUsage_258}
    Mask {from=""}
    DefaultUV {from=""}
   ]
   properties {materialFunction=&MF_TexPointSample_RGB}
  }
  MaterialExpressionMaterialFunctionCall_13 {
   class=MaterialExpressionMaterialFunctionCall
   inputs [
    Coordinates {from=MaterialExpressionNamedRerouteUsage_30}
    TexObject {from=MaterialExpressionTextureObjectParameter_10}
    DDX {from=MaterialExpressionNamedRerouteUsage_264}
    DDY {from=MaterialExpressionNamedRerouteUsage_265}
    Mask {from=""}
    DefaultUV {from=""}
   ]
   properties {materialFunction=/BaseMaterial/Materials/Functions/MF_TexPointSample_NRM}
  }
  MaterialExpressionMaterialFunctionCall_15 {
   class=MaterialExpressionMaterialFunctionCall
   inputs [
    Coordinates {from=MaterialExpressionNamedRerouteUsage_32}
    TexObject {from=MaterialExpressionTextureObjectParameter_20}
    DDX {from=MaterialExpressionNamedRerouteUsage_260}
    DDY {from=MaterialExpressionNamedRerouteUsage_261}
    Mask {from=""}
    DefaultUV {from=""}
   ]
   properties {materialFunction=&MF_TexPointSample_RGB}
  }
  MaterialExpressionMaterialFunctionCall_17 {
   class=MaterialExpressionMaterialFunctionCall
   inputs [ClearCoatNormalV {from=MaterialExpressionNamedRerouteUsage_128}]
   properties {materialFunction=/InterchangeAxfAssets/MaterialFunctions/MF_ClearCoatTerm}
  }
  MaterialExpressionMaterialFunctionCall_18 {
   class=MaterialExpressionMaterialFunctionCall
   inputs [
    Coordinates {from=MaterialExpressionNamedRerouteUsage_37}
    TexObject {from=MaterialExpressionTextureObjectParameter_21}
    DDX {from=MaterialExpressionNamedRerouteUsage_262}
    DDY {from=MaterialExpressionNamedRerouteUsage_263}
    Mask {from=""}
    DefaultUV {from=""}
   ]
   properties {materialFunction=&MF_TexPointSample_RGB}
  }
  MaterialExpressionMaterialFunctionCall_19 {
   class=MaterialExpressionMaterialFunctionCall
   inputs [
    X {from=MaterialExpressionVectorParameter_25, output=R}
    Y {from=MaterialExpressionVectorParameter_25, output=R}
   ]
   properties {materialFunction=&MakeFloat2}
  }
  MaterialExpressionMaterialFunctionCall_2 {
   class=MaterialExpressionMaterialFunctionCall
   inputs [
    NormalIn {from=MaterialExpressionStaticSwitch_0}
    NormalStrength {from=MaterialExpressionScalarParameter_27}
    ReFresnelVX {from=MaterialExpressionNamedRerouteUsage_29}
   ]
   properties {materialFunction=&MF_NormalStrength}
  }
  MaterialExpressionMaterialFunctionCall_21 {
   class=MaterialExpressionMaterialFunctionCall
   inputs [
    Coordinates {from=MaterialExpressionNamedRerouteUsage_6}
    TexObject {from=MaterialExpressionTextureObjectParameter_25}
    DDX {from=MaterialExpressionNamedRerouteUsage_253}
    DDY {from=MaterialExpressionNamedRerouteUsage_254}
    Mask {from=""}
    DefaultUV {from=""}
   ]
   properties {materialFunction=&MF_TexPointSample_RGB}
  }
  MaterialExpressionMaterialFunctionCall_23 {
   class=MaterialExpressionMaterialFunctionCall
   inputs [
    Coordinates {from=MaterialExpressionNamedRerouteUsage_9}
    TexObject {from=MaterialExpressionTextureObjectParameter_24}
    DDX {from=MaterialExpressionNamedRerouteUsage_116}
    DDY {from=MaterialExpressionNamedRerouteUsage_117}
    Mask {from=""}
    DefaultUV {from=""}
   ]
   properties {materialFunction=&MF_TexPointSample_RGB}
  }
  MaterialExpressionMaterialFunctionCall_3 {
   class=MaterialExpressionMaterialFunctionCall
   inputs [
    Coordinates {from=MaterialExpressionNamedRerouteUsage_2}
    TexObject {from=MaterialExpressionTextureObjectParameter_1}
    DDX {from=MaterialExpressionNamedRerouteUsage_246}
    DDY {from=MaterialExpressionNamedRerouteUsage_247}
    Mask {from=""}
    DefaultUV {from=""}
   ]
   properties {materialFunction=&MF_TexPointSample_NRM}
  }
  MaterialExpressionMaterialFunctionCall_4 {
   class=MaterialExpressionMaterialFunctionCall
   inputs [
    TSN {from=MaterialExpressionMaterialFunctionCall_3, output=RGB}
    Tangent {from=MaterialExpressionNamedRerouteUsage_40}
    Bitangent {from=MaterialExpressionNamedRerouteUsage_41}
    "Rotation (Turns)" {from=MaterialExpressionNamedRerouteUsage_42}
   ]
   properties {materialFunction=/BaseMaterial/Materials/Functions/MF_TangentNormal_OrientWorld}
  }
  MaterialExpressionMaterialFunctionCall_5 {
   class=MaterialExpressionMaterialFunctionCall
   inputs [
    TSN {from=MaterialExpressionMaterialFunctionCall_13, output=RGB}
    Tangent {from=MaterialExpressionNamedRerouteUsage_44}
    Bitangent {from=MaterialExpressionNamedRerouteUsage_43}
    "Rotation (Turns)" {from=MaterialExpressionNamedRerouteUsage_45}
   ]
   properties {materialFunction=&MF_TangentNormal_OrientWorld}
  }
  MaterialExpressionMaterialFunctionCall_59 {
   class=MaterialExpressionMaterialFunctionCall
   inputs [Color {from=MaterialExpressionStaticSwitchParameter_71}]
   properties {materialFunction=&SplitComponents}
  }
  MaterialExpressionMaterialFunctionCall_66 {
   class=MaterialExpressionMaterialFunctionCall
   inputs [
    X {from=MaterialExpressionTextureSample_7, output=R}
    Y {from=MaterialExpressionTextureSample_7, output=R}
   ]
   properties {materialFunction=&MakeFloat2}
  }
  MaterialExpressionMaterialFunctionCall_67 {
   class=MaterialExpressionMaterialFunctionCall
   inputs [
    X {from=MaterialExpressionMaterialFunctionCall_18, output=R}
    Y {from=MaterialExpressionMaterialFunctionCall_18, output=R}
   ]
   properties {materialFunction=&MakeFloat2}
  }
  MaterialExpressionMaterialFunctionCall_68 {
   class=MaterialExpressionMaterialFunctionCall
   inputs [
    AnisotropicRotation {from=MaterialExpressionStaticSwitchParameter_73}
    AlphaX {from=MaterialExpressionNamedRerouteUsage_241}
    AlphaY {from=MaterialExpressionNamedRerouteUsage_242}
   ]
   properties {materialFunction=/InterchangeAxfAssets/MaterialFunctions/MF_AnisotropyComponent}
  }
  MaterialExpressionMaterialFunctionCall_69 {
   class=MaterialExpressionMaterialFunctionCall
   inputs [SheenRoughness {from=MaterialExpressionStaticSwitchParameter_84}]
   properties {materialFunction=/InterchangeAxfAssets/MaterialFunctions/MF_SheenRoughness}
  }
  MaterialExpressionMaterialFunctionCall_7 {
   class=MaterialExpressionMaterialFunctionCall
   inputs [
    AxF_Scale_X {from=MaterialExpressionNamedRerouteUsage_11}
    AxF_Scale_Y {from=MaterialExpressionNamedRerouteUsage_13}
    AxF_UnitScale {from=MaterialExpressionNamedRerouteUsage_14}
    Tile U {from=MaterialExpressionNamedRerouteUsage_82}
    Tile V {from=MaterialExpressionNamedRerouteUsage_83}
    Offset U {from=MaterialExpressionNamedRerouteUsage_84}
    Offset V {from=MaterialExpressionNamedRerouteUsage_85}
    Rotation {from=MaterialExpressionNamedRerouteUsage_73}
    UV Index {from=MaterialExpressionScalarParameter_12}
   ]
   properties {materialFunction=/InterchangeAxfAssets/MaterialFunctions/MF_AxFMakeCoordinatesUV}
  }
  MaterialExpressionMaterialFunctionCall_72 {
   class=MaterialExpressionMaterialFunctionCall
   inputs []
   properties {materialFunction=&CameraDirectionVector}
  }
  MaterialExpressionMaterialFunctionCall_8 {
   class=MaterialExpressionMaterialFunctionCall
   inputs [
    AxF_Scale_X {from=MaterialExpressionNamedRerouteUsage_3}
    AxF_Scale_Y {from=MaterialExpressionNamedRerouteUsage_4}
    AxF_UnitScale {from=MaterialExpressionNamedRerouteUsage_10}
    Tile XYZ {from=MaterialExpressionNamedRerouteUsage_60}
    Offset XYZ {from=MaterialExpressionNamedRerouteUsage_39}
    Transition Softness {from=MaterialExpressionNamedRerouteUsage_49}
    Coordinate Space {from=MaterialExpressionStaticSwitchParameter_2}
    Rotation {from=MaterialExpressionNamedRerouteUsage_61}
   ]
   properties {materialFunction=/InterchangeAxfAssets/MaterialFunctions/MF_AxFMakeCoordinatesTP}
  }
  MaterialExpressionMaterialFunctionCall_9 {
   class=MaterialExpressionMaterialFunctionCall
   inputs [
    Coordinates {from=MaterialExpressionNamedRerouteUsage_12}
    TexObject {from=MaterialExpressionTextureObjectParameter_23}
    DDX {from=MaterialExpressionNamedRerouteUsage_255}
    DDY {from=MaterialExpressionNamedRerouteUsage_256}
    Mask {from=""}
    DefaultUV {from=""}
   ]
   properties {materialFunction=&MF_TexPointSample_RGB}
  }
  MaterialExpressionMultiply_2 {
   class=MaterialExpressionMultiply
   inputs [
    A {from=MaterialExpressionScalarParameter_8}
    B {from=""}
   ]
   properties {constB=10}
  }
  MaterialExpressionNamedRerouteDeclaration_0 {
   class=MaterialExpressionNamedRerouteDeclaration
   inputs [None {from=MaterialExpressionMaterialFunctionCall_17, output=Top Layer Slab}]
   properties {name=ClearCoatTerm_O, nodeColor=<rgba 1 0.6797 0 1>}
  }
  MaterialExpressionNamedRerouteDeclaration_1 {
   class=MaterialExpressionNamedRerouteDeclaration
   inputs [None {from=MaterialExpressionDivide_0}]
   properties {name=AxF_ScaleX, nodeColor=<rgba 0.7605 0 0.0009 1>}
  }
  MaterialExpressionNamedRerouteDeclaration_116 {
   class=MaterialExpressionNamedRerouteDeclaration
   inputs [None {from=MaterialExpressionStaticSwitchParameter_67}]
   properties {name=DiffuseAlbedoTerm_O, nodeColor=<rgba 0 1 0.0391 1>}
  }
  MaterialExpressionNamedRerouteDeclaration_117 {
   class=MaterialExpressionNamedRerouteDeclaration
   inputs [None {from=MaterialExpressionStaticSwitchParameter_75}]
   properties {name=F0_O, nodeColor=<rgba 1 0 0.2812 1>}
  }
  MaterialExpressionNamedRerouteDeclaration_118 {
   class=MaterialExpressionNamedRerouteDeclaration
   inputs [None {from=MaterialExpressionStaticSwitchParameter_76}]
   properties {name=F90_O, nodeColor=<rgba 1 0 0.2812 1>}
  }
  MaterialExpressionNamedRerouteDeclaration_119 {
   class=MaterialExpressionNamedRerouteDeclaration
   inputs [None {from=MaterialExpressionSquareRoot_2}]
   properties {name=Roughness_O, nodeColor=<rgba 0.9219 1 0 1>}
  }
  MaterialExpressionNamedRerouteDeclaration_120 {
   class=MaterialExpressionNamedRerouteDeclaration
   inputs [None {from=MaterialExpressionMaterialFunctionCall_68, output=Anisotropy}]
   properties {name=Anisotropy_O, nodeColor=<rgba 0.9219 1 0 1>}
  }
  MaterialExpressionNamedRerouteDeclaration_121 {
   class=MaterialExpressionNamedRerouteDeclaration
   inputs [None {from=MaterialExpressionMaterialFunctionCall_68, output=Tangent}]
   properties {name=Tangent_O, nodeColor=<rgba 0.9219 1 0 1>}
  }
  MaterialExpressionNamedRerouteDeclaration_122 {
   class=MaterialExpressionNamedRerouteDeclaration
   inputs [None {from=MaterialExpressionStaticSwitchParameter_79}]
   properties {name=FuzzRoughness_O, nodeColor=<rgba 0.0547 1 0 1>}
  }
  MaterialExpressionNamedRerouteDeclaration_123 {
   class=MaterialExpressionNamedRerouteDeclaration
   inputs [None {from=MaterialExpressionStaticSwitchParameter_80}]
   properties {name=FuzzAmount_O, nodeColor=<rgba 0.0547 1 0 1>}
  }
  MaterialExpressionNamedRerouteDeclaration_124 {
   class=MaterialExpressionNamedRerouteDeclaration
   inputs [None {from=MaterialExpressionStaticSwitchParameter_81}]
   properties {name=FuzzColor_O, nodeColor=<rgba 0.0547 1 0 1>}
  }
  MaterialExpressionNamedRerouteDeclaration_125 {
   class=MaterialExpressionNamedRerouteDeclaration
   inputs [None {from=MaterialExpressionStaticSwitchParameter_69}]
   properties {name=Normal_O, nodeColor=<rgba 1 0.9844 0 1>}
  }
  MaterialExpressionNamedRerouteDeclaration_126 {
   class=MaterialExpressionNamedRerouteDeclaration
   inputs [None {from=MaterialExpressionMaterialFunctionCall_59, output=R}]
   properties {name=AlphaX, nodeColor=<rgba 0.1484 1 0 1>}
  }
  MaterialExpressionNamedRerouteDeclaration_127 {
   class=MaterialExpressionNamedRerouteDeclaration
   inputs [None {from=MaterialExpressionMaterialFunctionCall_59, output=G}]
   properties {name=AlphaY, nodeColor=<rgba 0.1484 1 0 1>}
  }
  MaterialExpressionNamedRerouteDeclaration_128 {
   class=MaterialExpressionNamedRerouteDeclaration
   inputs [None {from=MaterialExpressionMaterialFunctionCall_69, output=IsotropicRoughness}]
   properties {name=SheenRoughness, nodeColor=<rgba 1 0 0.1641 1>}
  }
  MaterialExpressionNamedRerouteDeclaration_129 {
   class=MaterialExpressionNamedRerouteDeclaration
   inputs [None {from=MaterialExpressionMaterialFunctionCall_8, output=Turns}]
   properties {name=TP_Rotation, nodeColor=<rgba 0.9453 1 0 1>}
  }
  MaterialExpressionNamedRerouteDeclaration_13 {
   class=MaterialExpressionNamedRerouteDeclaration
   inputs [None {from=MaterialExpressionConvert_0}]
   properties {name=TP_DDY, nodeColor=<rgba 0 0.7891 1 1>}
  }
  MaterialExpressionNamedRerouteDeclaration_16 {
   class=MaterialExpressionNamedRerouteDeclaration
   inputs [None {from=MaterialExpressionConvert_0}]
   properties {name=TP_DDX, nodeColor=<rgba 0.5703 0 1 1>}
  }
  MaterialExpressionNamedRerouteDeclaration_2 {
   class=MaterialExpressionNamedRerouteDeclaration
   inputs [None {from=MaterialExpressionDivide_1}]
   properties {name=AxF_ScaleY, nodeColor=<rgba 0.7605 0 0.0009 1>}
  }
  MaterialExpressionNamedRerouteDeclaration_23 {
   class=MaterialExpressionNamedRerouteDeclaration
   inputs [None {from=MaterialExpressionMaterialFunctionCall_7, output=Output UV}]
   properties {name=TC_UV, nodeColor=<rgba 0 1 0.25 1>}
  }
  MaterialExpressionNamedRerouteDeclaration_27 {
   class=MaterialExpressionNamedRerouteDeclaration
   inputs [None {from=MaterialExpressionSaturate_0}]
   properties {name=ReFresnelVX, nodeColor=<rgba 0.2656 0 1 1>}
  }
  MaterialExpressionNamedRerouteDeclaration_3 {
   class=MaterialExpressionNamedRerouteDeclaration
   inputs [None {from=MaterialExpressionMultiply_2}]
   properties {name=AxF_UnitScale_MM, nodeColor=<rgba 0.7605 0 0.0009 1>}
  }
  MaterialExpressionNamedRerouteDeclaration_4 {
   class=MaterialExpressionNamedRerouteDeclaration
   inputs [None {from=MaterialExpressionMaterialFunctionCall_8, output=Output UV}]
   properties {name=TP_UV, nodeColor=<rgba 0 0.1094 1 1>}
  }
  MaterialExpressionNamedRerouteDeclaration_46 {
   class=MaterialExpressionNamedRerouteDeclaration
   inputs [None {from=MaterialExpressionScalarParameter_7}]
   properties {name=TransitionSoftness, nodeColor=<rgba 0 1 0.2031 1>}
  }
  MaterialExpressionNamedRerouteDeclaration_49 {
   class=MaterialExpressionNamedRerouteDeclaration
   inputs [None {from=MaterialExpressionScalarParameter_2}]
   properties {name=UV_Scale, nodeColor=<rgba 0.7578 1 0 1>}
  }
  MaterialExpressionNamedRerouteDeclaration_5 {
   class=MaterialExpressionNamedRerouteDeclaration
   inputs [None {from=MaterialExpressionMaterialFunctionCall_8, output=T}]
   properties {name=TP_Tangent, nodeColor=<rgba 1 0 0.8672 1>}
  }
  MaterialExpressionNamedRerouteDeclaration_50 {
   class=MaterialExpressionNamedRerouteDeclaration
   inputs [None {from=MaterialExpressionScalarParameter_28}]
   properties {name=UV_Rot, nodeColor=<rgba 0.7578 1 0 1>}
  }
  MaterialExpressionNamedRerouteDeclaration_51 {
   class=MaterialExpressionNamedRerouteDeclaration
   inputs [None {from=MaterialExpressionScalarParameter_3}]
   properties {name=UV_X, nodeColor=<rgba 0 1 0.6016 1>}
  }
  MaterialExpressionNamedRerouteDeclaration_52 {
   class=MaterialExpressionNamedRerouteDeclaration
   inputs [None {from=MaterialExpressionScalarParameter_4}]
   properties {name=UV_Y, nodeColor=<rgba 0 1 0.6016 1>}
  }
  MaterialExpressionNamedRerouteDeclaration_53 {
   class=MaterialExpressionNamedRerouteDeclaration
   inputs [None {from=MaterialExpressionScalarParameter_10}]
   properties {name=UV_Z, nodeColor=<rgba 0 1 0.6016 1>}
  }
  MaterialExpressionNamedRerouteDeclaration_54 {
   class=MaterialExpressionNamedRerouteDeclaration
   inputs [None {from=MaterialExpressionScalarParameter_6}]
   properties {name=OffsetX, nodeColor=<rgba 0.3594 1 0 1>}
  }
  MaterialExpressionNamedRerouteDeclaration_55 {
   class=MaterialExpressionNamedRerouteDeclaration
   inputs [None {from=MaterialExpressionScalarParameter_5}]
   properties {name=OffsetY, nodeColor=<rgba 0.3594 1 0 1>}
  }
  MaterialExpressionNamedRerouteDeclaration_56 {
   class=MaterialExpressionNamedRerouteDeclaration
   inputs [None {from=MaterialExpressionScalarParameter_11}]
   properties {name=OffsetZ, nodeColor=<rgba 0.3594 1 0 1>}
  }
  MaterialExpressionNamedRerouteDeclaration_59 {
   class=MaterialExpressionNamedRerouteDeclaration
   inputs [None {from=MaterialExpressionStaticSwitchParameter_29}]
   properties {name=CCNormal, nodeColor=<rgba 0.5469 0 1 1>}
  }
  MaterialExpressionNamedRerouteDeclaration_6 {
   class=MaterialExpressionNamedRerouteDeclaration
   inputs [None {from=MaterialExpressionStaticBoolParameter_1}]
   properties {name=UseGeoUVs, nodeColor=<rgba 0 0.8359 1 1>}
  }
  MaterialExpressionNamedRerouteDeclaration_7 {
   class=MaterialExpressionNamedRerouteDeclaration
   inputs [None {from=MaterialExpressionMaterialFunctionCall_8, output=B}]
   properties {name=TP_Bitangent, nodeColor=<rgba 0 1 0.8125 1>}
  }
  MaterialExpressionNamedRerouteDeclaration_8 {
   class=MaterialExpressionNamedRerouteDeclaration
   inputs [None {from=MaterialExpressionConvert_1}]
   properties {name=TP_TileXYZ, nodeColor=<rgba 1 0.8438 0 1>}
  }
  MaterialExpressionNamedRerouteDeclaration_9 {
   class=MaterialExpressionNamedRerouteDeclaration
   inputs [None {from=MaterialExpressionConvert_2}]
   properties {name=TP_OffsetXYZ, nodeColor=<rgba 0.5234 0 1 1>}
  }
  MaterialExpressionNamedRerouteUsage_0 {class=MaterialExpressionNamedRerouteUsage}
  MaterialExpressionNamedRerouteUsage_1 {class=MaterialExpressionNamedRerouteUsage}
  MaterialExpressionNamedRerouteUsage_10 {class=MaterialExpressionNamedRerouteUsage}
  MaterialExpressionNamedRerouteUsage_11 {class=MaterialExpressionNamedRerouteUsage}
  MaterialExpressionNamedRerouteUsage_116 {class=MaterialExpressionNamedRerouteUsage}
  MaterialExpressionNamedRerouteUsage_117 {class=MaterialExpressionNamedRerouteUsage}
  MaterialExpressionNamedRerouteUsage_12 {class=MaterialExpressionNamedRerouteUsage}
  MaterialExpressionNamedRerouteUsage_128 {class=MaterialExpressionNamedRerouteUsage}
  MaterialExpressionNamedRerouteUsage_13 {class=MaterialExpressionNamedRerouteUsage}
  MaterialExpressionNamedRerouteUsage_14 {class=MaterialExpressionNamedRerouteUsage}
  MaterialExpressionNamedRerouteUsage_15 {class=MaterialExpressionNamedRerouteUsage}
  MaterialExpressionNamedRerouteUsage_16 {class=MaterialExpressionNamedRerouteUsage}
  MaterialExpressionNamedRerouteUsage_17 {class=MaterialExpressionNamedRerouteUsage}
  MaterialExpressionNamedRerouteUsage_18 {class=MaterialExpressionNamedRerouteUsage}
  MaterialExpressionNamedRerouteUsage_19 {class=MaterialExpressionNamedRerouteUsage}
  MaterialExpressionNamedRerouteUsage_2 {class=MaterialExpressionNamedRerouteUsage}
  MaterialExpressionNamedRerouteUsage_20 {class=MaterialExpressionNamedRerouteUsage}
  MaterialExpressionNamedRerouteUsage_21 {class=MaterialExpressionNamedRerouteUsage}
  MaterialExpressionNamedRerouteUsage_22 {class=MaterialExpressionNamedRerouteUsage}
  MaterialExpressionNamedRerouteUsage_23 {class=MaterialExpressionNamedRerouteUsage}
  MaterialExpressionNamedRerouteUsage_231 {class=MaterialExpressionNamedRerouteUsage}
  MaterialExpressionNamedRerouteUsage_232 {class=MaterialExpressionNamedRerouteUsage}
  MaterialExpressionNamedRerouteUsage_233 {class=MaterialExpressionNamedRerouteUsage}
  MaterialExpressionNamedRerouteUsage_234 {class=MaterialExpressionNamedRerouteUsage}
  MaterialExpressionNamedRerouteUsage_235 {class=MaterialExpressionNamedRerouteUsage}
  MaterialExpressionNamedRerouteUsage_236 {class=MaterialExpressionNamedRerouteUsage}
  MaterialExpressionNamedRerouteUsage_237 {class=MaterialExpressionNamedRerouteUsage}
  MaterialExpressionNamedRerouteUsage_238 {class=MaterialExpressionNamedRerouteUsage}
  MaterialExpressionNamedRerouteUsage_239 {class=MaterialExpressionNamedRerouteUsage}
  MaterialExpressionNamedRerouteUsage_24 {class=MaterialExpressionNamedRerouteUsage}
  MaterialExpressionNamedRerouteUsage_240 {class=MaterialExpressionNamedRerouteUsage}
  MaterialExpressionNamedRerouteUsage_241 {class=MaterialExpressionNamedRerouteUsage}
  MaterialExpressionNamedRerouteUsage_242 {class=MaterialExpressionNamedRerouteUsage}
  MaterialExpressionNamedRerouteUsage_243 {class=MaterialExpressionNamedRerouteUsage}
  MaterialExpressionNamedRerouteUsage_244 {class=MaterialExpressionNamedRerouteUsage}
  MaterialExpressionNamedRerouteUsage_245 {class=MaterialExpressionNamedRerouteUsage}
  MaterialExpressionNamedRerouteUsage_246 {class=MaterialExpressionNamedRerouteUsage}
  MaterialExpressionNamedRerouteUsage_247 {class=MaterialExpressionNamedRerouteUsage}
  MaterialExpressionNamedRerouteUsage_25 {class=MaterialExpressionNamedRerouteUsage}
  MaterialExpressionNamedRerouteUsage_253 {class=MaterialExpressionNamedRerouteUsage}
  MaterialExpressionNamedRerouteUsage_254 {class=MaterialExpressionNamedRerouteUsage}
  MaterialExpressionNamedRerouteUsage_255 {class=MaterialExpressionNamedRerouteUsage}
  MaterialExpressionNamedRerouteUsage_256 {class=MaterialExpressionNamedRerouteUsage}
  MaterialExpressionNamedRerouteUsage_257 {class=MaterialExpressionNamedRerouteUsage}
  MaterialExpressionNamedRerouteUsage_258 {class=MaterialExpressionNamedRerouteUsage}
  MaterialExpressionNamedRerouteUsage_26 {class=MaterialExpressionNamedRerouteUsage}
  MaterialExpressionNamedRerouteUsage_260 {class=MaterialExpressionNamedRerouteUsage}
  MaterialExpressionNamedRerouteUsage_261 {class=MaterialExpressionNamedRerouteUsage}
  MaterialExpressionNamedRerouteUsage_262 {class=MaterialExpressionNamedRerouteUsage}
  MaterialExpressionNamedRerouteUsage_263 {class=MaterialExpressionNamedRerouteUsage}
  MaterialExpressionNamedRerouteUsage_264 {class=MaterialExpressionNamedRerouteUsage}
  MaterialExpressionNamedRerouteUsage_265 {class=MaterialExpressionNamedRerouteUsage}
  MaterialExpressionNamedRerouteUsage_27 {class=MaterialExpressionNamedRerouteUsage}
  MaterialExpressionNamedRerouteUsage_28 {class=MaterialExpressionNamedRerouteUsage}
  MaterialExpressionNamedRerouteUsage_29 {class=MaterialExpressionNamedRerouteUsage}
  MaterialExpressionNamedRerouteUsage_3 {class=MaterialExpressionNamedRerouteUsage}
  MaterialExpressionNamedRerouteUsage_30 {class=MaterialExpressionNamedRerouteUsage}
  MaterialExpressionNamedRerouteUsage_31 {class=MaterialExpressionNamedRerouteUsage}
  MaterialExpressionNamedRerouteUsage_32 {class=MaterialExpressionNamedRerouteUsage}
  MaterialExpressionNamedRerouteUsage_33 {class=MaterialExpressionNamedRerouteUsage}
  MaterialExpressionNamedRerouteUsage_34 {class=MaterialExpressionNamedRerouteUsage}
  MaterialExpressionNamedRerouteUsage_35 {class=MaterialExpressionNamedRerouteUsage}
  MaterialExpressionNamedRerouteUsage_36 {class=MaterialExpressionNamedRerouteUsage}
  MaterialExpressionNamedRerouteUsage_37 {class=MaterialExpressionNamedRerouteUsage}
  MaterialExpressionNamedRerouteUsage_38 {class=MaterialExpressionNamedRerouteUsage}
  MaterialExpressionNamedRerouteUsage_39 {class=MaterialExpressionNamedRerouteUsage}
  MaterialExpressionNamedRerouteUsage_4 {class=MaterialExpressionNamedRerouteUsage}
  MaterialExpressionNamedRerouteUsage_40 {class=MaterialExpressionNamedRerouteUsage}
  MaterialExpressionNamedRerouteUsage_41 {class=MaterialExpressionNamedRerouteUsage}
  MaterialExpressionNamedRerouteUsage_42 {class=MaterialExpressionNamedRerouteUsage}
  MaterialExpressionNamedRerouteUsage_43 {class=MaterialExpressionNamedRerouteUsage}
  MaterialExpressionNamedRerouteUsage_44 {class=MaterialExpressionNamedRerouteUsage}
  MaterialExpressionNamedRerouteUsage_45 {class=MaterialExpressionNamedRerouteUsage}
  MaterialExpressionNamedRerouteUsage_49 {class=MaterialExpressionNamedRerouteUsage}
  MaterialExpressionNamedRerouteUsage_5 {class=MaterialExpressionNamedRerouteUsage}
  MaterialExpressionNamedRerouteUsage_6 {class=MaterialExpressionNamedRerouteUsage}
  MaterialExpressionNamedRerouteUsage_60 {class=MaterialExpressionNamedRerouteUsage}
  MaterialExpressionNamedRerouteUsage_61 {class=MaterialExpressionNamedRerouteUsage}
  MaterialExpressionNamedRerouteUsage_7 {class=MaterialExpressionNamedRerouteUsage}
  MaterialExpressionNamedRerouteUsage_70 {class=MaterialExpressionNamedRerouteUsage}
  MaterialExpressionNamedRerouteUsage_73 {class=MaterialExpressionNamedRerouteUsage}
  MaterialExpressionNamedRerouteUsage_76 {class=MaterialExpressionNamedRerouteUsage}
  MaterialExpressionNamedRerouteUsage_77 {class=MaterialExpressionNamedRerouteUsage}
  MaterialExpressionNamedRerouteUsage_78 {class=MaterialExpressionNamedRerouteUsage}
  MaterialExpressionNamedRerouteUsage_8 {class=MaterialExpressionNamedRerouteUsage}
  MaterialExpressionNamedRerouteUsage_82 {class=MaterialExpressionNamedRerouteUsage}
  MaterialExpressionNamedRerouteUsage_83 {class=MaterialExpressionNamedRerouteUsage}
  MaterialExpressionNamedRerouteUsage_84 {class=MaterialExpressionNamedRerouteUsage}
  MaterialExpressionNamedRerouteUsage_85 {class=MaterialExpressionNamedRerouteUsage}
  MaterialExpressionNamedRerouteUsage_9 {class=MaterialExpressionNamedRerouteUsage}
  MaterialExpressionSaturate_0 {class=MaterialExpressionSaturate, inputs [None {from=MaterialExpressionAbs_0}]}
  MaterialExpressionScalarParameter_0 {
   class=MaterialExpressionScalarParameter
   properties {defaultValue=1, group=ClearCoat, parameterName=ClearCoatNormalStrength, sliderMax=2, sliderMin=-1}
  }
  MaterialExpressionScalarParameter_1 {class=MaterialExpressionScalarParameter, properties {defaultValue=100, group=MillimeterSettings, parameterName=WidthMM, sortPriority=10}}
  MaterialExpressionScalarParameter_10 {
   class=MaterialExpressionScalarParameter
   properties {defaultValue=1, group=UV Transforms, parameterName=StretchOnZ, sliderMax=1000, sliderMin=0.001, sortPriority=120}
  }
  MaterialExpressionScalarParameter_11 {class=MaterialExpressionScalarParameter, properties {group=UV Transforms, parameterName=OffsetOnZ, sortPriority=150}}
  MaterialExpressionScalarParameter_12 {class=MaterialExpressionScalarParameter, properties {group=UV Transforms, parameterName=UV Index}}
  MaterialExpressionScalarParameter_2 {
   class=MaterialExpressionScalarParameter
   properties {defaultValue=1, group=UV Transforms, parameterName=Scale, sliderMax=10000, sliderMin=0.001, sortPriority=10}
  }
  MaterialExpressionScalarParameter_27 {
   class=MaterialExpressionScalarParameter
   properties {defaultValue=1, group=Normal, parameterName=NormalStrength, sliderMax=2, sliderMin=-1, sortPriority=12}
  }
  MaterialExpressionScalarParameter_28 {class=MaterialExpressionScalarParameter, properties {group=UV Transforms, parameterName="Rotation (degree)", sliderMax=360, sortPriority=30}}
  MaterialExpressionScalarParameter_3 {
   class=MaterialExpressionScalarParameter
   properties {defaultValue=1, group=UV Transforms, parameterName=StretchOnX, sliderMax=100, sliderMin=0.001, sortPriority=100}
  }
  MaterialExpressionScalarParameter_4 {
   class=MaterialExpressionScalarParameter
   properties {defaultValue=1, group=UV Transforms, parameterName=StretchOnY, sliderMax=1000, sliderMin=0.001, sortPriority=110}
  }
  MaterialExpressionScalarParameter_5 {class=MaterialExpressionScalarParameter, properties {group=UV Transforms, parameterName=OffsetOnY, sortPriority=140}}
  MaterialExpressionScalarParameter_6 {class=MaterialExpressionScalarParameter, properties {group=UV Transforms, parameterName=OffsetOnX, sortPriority=130}}
  MaterialExpressionScalarParameter_7 {
   class=MaterialExpressionScalarParameter
   properties {
    defaultValue=0.5
    desc="Controls the contrast of the blend mask. Internally applied as pow(mask, 8 × BlendFactor). A default of 0.5 corresponds to pow4 contrast."
    group=UV Transforms
    parameterName=TransitionSoftness
    sliderMax=1
    sortPriority=3
   }
  }
  MaterialExpressionScalarParameter_8 {class=MaterialExpressionScalarParameter, properties {defaultValue=1, parameterName=UnitScaleCM}}
  MaterialExpressionScalarParameter_9 {class=MaterialExpressionScalarParameter, properties {defaultValue=100, group=MillimeterSettings, parameterName=HeightMM, sortPriority=11}}
  MaterialExpressionShaderStageSwitch_2 {
   class=MaterialExpressionShaderStageSwitch
   inputs [
    PixelShader {from=MaterialExpressionCameraVectorWS_0, output=""}
    VertexShader {from=MaterialExpressionMaterialFunctionCall_72, output=Result}
   ]
   properties {}
  }
  MaterialExpressionSquareRoot_2 {class=MaterialExpressionSquareRoot, inputs [None {from=MaterialExpressionDivide_2, output=""}], properties {}}
  MaterialExpressionStaticBoolParameter_1 {
   class=MaterialExpressionStaticBoolParameter
   inputs []
   properties {defaultValue=true, group=UV Transforms, parameterName=Use Geo UVs, sortPriority=0}
  }
  MaterialExpressionStaticSwitchParameter_2 {
   class=MaterialExpressionStaticSwitchParameter
   inputs [
    True {from=MaterialExpressionConstant_2, output=""}
    False {from=MaterialExpressionConstant_3, output=""}
   ]
   properties {defaultValue=true, group=UV Transforms, parameterName=UsingLocalSpace, sortPriority=2}
  }
  MaterialExpressionStaticSwitchParameter_29 {
   class=MaterialExpressionStaticSwitchParameter
   inputs [
    True {from=MaterialExpressionMaterialFunctionCall_1, output=NormalOut}
    False {from=MaterialExpressionComponentMask_1, output=""}
   ]
   properties {group=ClearCoat, parameterName=HasClearCoatNormalTexture, sortPriority=30}
  }
  MaterialExpressionStaticSwitchParameter_66 {
   class=MaterialExpressionStaticSwitchParameter
   inputs [
    True {from=MaterialExpressionSubstrateVerticalLayering_2, output=""}
    False {from=MaterialExpressionSubstrateSlabBSDF_2, output=""}
   ]
   properties {group=ClearCoat, parameterName=HasClearCoat, sortPriority=30}
  }
  MaterialExpressionStaticSwitchParameter_67 {
   class=MaterialExpressionStaticSwitchParameter
   inputs [
    True {from=MaterialExpressionStaticSwitch_1, output=""}
    False {from=MaterialExpressionVectorParameter_18, output=RGB}
   ]
   properties {group=Diffuse, parameterName=HasDiffuseColorTexture}
  }
  MaterialExpressionStaticSwitchParameter_69 {
   class=MaterialExpressionStaticSwitchParameter
   inputs [
    True {from=MaterialExpressionMaterialFunctionCall_2, output=NormalOut}
    False {from=MaterialExpressionComponentMask_8, output=""}
   ]
   properties {group=Normal, parameterName=HasNormalTexture, sortPriority=30}
  }
  MaterialExpressionStaticSwitchParameter_71 {
   class=MaterialExpressionStaticSwitchParameter
   inputs [
    True {from=MaterialExpressionStaticSwitch_7, output=""}
    False {from=MaterialExpressionVectorParameter_20, output=RGB}
   ]
   properties {group=Roughness, parameterName=HasRoughnessTexture, sortPriority=30}
  }
  MaterialExpressionStaticSwitchParameter_73 {
   class=MaterialExpressionStaticSwitchParameter
   inputs [
    True {from=MaterialExpressionStaticSwitch_8, output=""}
    False {from=MaterialExpressionMaterialFunctionCall_19, output=Result}
   ]
   properties {group=Anisotropy, parameterName=HasAnisotropicRotationTexture}
  }
  MaterialExpressionStaticSwitchParameter_75 {
   class=MaterialExpressionStaticSwitchParameter
   inputs [
    True {from=MaterialExpressionStaticSwitch_4, output=""}
    False {from=MaterialExpressionVectorParameter_21, output=RGB}
   ]
   properties {group=Fresnel, parameterName=HasF0Texture}
  }
  MaterialExpressionStaticSwitchParameter_76 {
   class=MaterialExpressionStaticSwitchParameter
   inputs [
    True {from=MaterialExpressionStaticSwitch_5, output=""}
    False {from=MaterialExpressionVectorParameter_22, output=RGB}
   ]
   properties {group=Fresnel, parameterName=HasF90Texture}
  }
  MaterialExpressionStaticSwitchParameter_79 {
   class=MaterialExpressionStaticSwitchParameter
   inputs [
    True {from=MaterialExpressionNamedRerouteUsage_243, output=""}
    False {from=MaterialExpressionConstant_17, output=""}
   ]
   properties {group=Sheen, parameterName=HasSheen, sortPriority=30}
  }
  MaterialExpressionStaticSwitchParameter_80 {
   class=MaterialExpressionStaticSwitchParameter
   inputs [
    True {from=MaterialExpressionConstant_18, output=""}
    False {from=MaterialExpressionConstant_15, output=""}
   ]
   properties {group=Sheen, parameterName=HasSheen, sortPriority=30}
  }
  MaterialExpressionStaticSwitchParameter_81 {
   class=MaterialExpressionStaticSwitchParameter
   inputs [
    True {from=MaterialExpressionStaticSwitchParameter_82, output=""}
    False {from=MaterialExpressionConstant_16, output=""}
   ]
   properties {group=Sheen, parameterName=HasSheen, sortPriority=30}
  }
  MaterialExpressionStaticSwitchParameter_82 {
   class=MaterialExpressionStaticSwitchParameter
   inputs [
    True {from=MaterialExpressionStaticSwitch_3, output=""}
    False {from=MaterialExpressionVectorParameter_23, output=RGB}
   ]
   properties {group=Sheen, parameterName=HasSheenColorTexture}
  }
  MaterialExpressionStaticSwitchParameter_84 {
   class=MaterialExpressionStaticSwitchParameter
   inputs [
    True {from=MaterialExpressionStaticSwitch_2, output=""}
    False {from=MaterialExpressionVectorParameter_24, output=RGB}
   ]
   properties {group=Sheen, parameterName=HasSheenRoughnessTexture}
  }
  MaterialExpressionStaticSwitch_0 {
   class=MaterialExpressionStaticSwitch
   inputs [
    True {from=MaterialExpressionTextureSample_6, output=RGB}
    False {from=MaterialExpressionMaterialFunctionCall_4, output=TSN}
    Value {from=MaterialExpressionNamedRerouteUsage_15, output=""}
   ]
   properties {}
  }
  MaterialExpressionStaticSwitch_1 {
   class=MaterialExpressionStaticSwitch
   inputs [
    True {from=MaterialExpressionTextureSample_15, output=RGB}
    False {from=MaterialExpressionMaterialFunctionCall_0, output=RGB}
    Value {from=MaterialExpressionNamedRerouteUsage_16, output=""}
   ]
   properties {}
  }
  MaterialExpressionStaticSwitch_2 {
   class=MaterialExpressionStaticSwitch
   inputs [
    True {from=MaterialExpressionTextureSample_4, output=RGB}
    False {from=MaterialExpressionMaterialFunctionCall_21, output=RGB}
    Value {from=MaterialExpressionNamedRerouteUsage_17, output=""}
   ]
   properties {}
  }
  MaterialExpressionStaticSwitch_3 {
   class=MaterialExpressionStaticSwitch
   inputs [
    True {from=MaterialExpressionTextureSample_5, output=RGB}
    False {from=MaterialExpressionMaterialFunctionCall_23, output=RGB}
    Value {from=MaterialExpressionNamedRerouteUsage_18, output=""}
   ]
   properties {}
  }
  MaterialExpressionStaticSwitch_4 {
   class=MaterialExpressionStaticSwitch
   inputs [
    True {from=MaterialExpressionTextureSample_2, output=RGB}
    False {from=MaterialExpressionMaterialFunctionCall_9, output=RGB}
    Value {from=MaterialExpressionNamedRerouteUsage_19, output=""}
   ]
   properties {}
  }
  MaterialExpressionStaticSwitch_5 {
   class=MaterialExpressionStaticSwitch
   inputs [
    True {from=MaterialExpressionTextureSample_3, output=RGB}
    False {from=MaterialExpressionMaterialFunctionCall_12, output=RGB}
    Value {from=MaterialExpressionNamedRerouteUsage_23, output=""}
   ]
   properties {}
  }
  MaterialExpressionStaticSwitch_6 {
   class=MaterialExpressionStaticSwitch
   inputs [
    True {from=MaterialExpressionTextureSample_0, output=RGB}
    False {from=MaterialExpressionMaterialFunctionCall_5, output=TSN}
    Value {from=MaterialExpressionNamedRerouteUsage_24, output=""}
   ]
   properties {}
  }
  MaterialExpressionStaticSwitch_7 {
   class=MaterialExpressionStaticSwitch
   inputs [
    True {from=MaterialExpressionTextureSample_8, output=RGB}
    False {from=MaterialExpressionMaterialFunctionCall_15, output=RGB}
    Value {from=MaterialExpressionNamedRerouteUsage_27, output=""}
   ]
   properties {}
  }
  MaterialExpressionStaticSwitch_8 {
   class=MaterialExpressionStaticSwitch
   inputs [
    True {from=MaterialExpressionMaterialFunctionCall_66, output=Result}
    False {from=MaterialExpressionMaterialFunctionCall_67, output=Result}
    Value {from=MaterialExpressionNamedRerouteUsage_28, output=""}
   ]
   properties {}
  }
  MaterialExpressionSubstrateSlabBSDF_2 {
   class=MaterialExpressionSubstrateSlabBSDF
   inputs [
    Diffuse Albedo {from=MaterialExpressionNamedRerouteUsage_231, output=""}
    F0 {from=MaterialExpressionNamedRerouteUsage_232, output=""}
    F90 {from=MaterialExpressionNamedRerouteUsage_240, output=""}
    Roughness {from=MaterialExpressionNamedRerouteUsage_233, output=""}
    Anisotropy {from=MaterialExpressionNamedRerouteUsage_234, output=""}
    Normal {from=MaterialExpressionNamedRerouteUsage_239, output=""}
    Tangent {from=MaterialExpressionNamedRerouteUsage_235, output=""}
    SSS MFP {from="", output=""}
    SSS MFP Scale {from="", output=""}
    SSS Phase Anisotropy {from="", output=""}
    Emissive Color {from="", output=""}
    Second Roughness {from="", output=""}
    Second Roughness Weight {from="", output=""}
    Fuzz Roughness {from=MaterialExpressionNamedRerouteUsage_236, output=""}
    Fuzz Amount {from=MaterialExpressionNamedRerouteUsage_237, output=""}
    Fuzz Color {from=MaterialExpressionNamedRerouteUsage_238, output=""}
    "Glint Density (Disabled)" {from="", output=""}
    "Glint UVs (Disabled)" {from="", output=""}
   ]
   properties {}
  }
  MaterialExpressionSubstrateVerticalLayering_2 {
   class=MaterialExpressionSubstrateVerticalLayering
   inputs [
    Top {from=MaterialExpressionNamedRerouteUsage_20, output=""}
    Bottom {from=MaterialExpressionSubstrateSlabBSDF_2, output=""}
    Top Thickness {from="", output=""}
   ]
   properties {}
  }
  MaterialExpressionTextureObjectParameter_1 {
   class=MaterialExpressionTextureObjectParameter
   inputs [
    UVs {from="", output=""}
    Apply View MipBias {from="", output=""}
   ]
   properties {group=Normal, parameterName=NormalTexture, samplerType=SAMPLERTYPE_Normal, sortPriority=1, texture=/Engine/EngineMaterials/DefaultNormal}
  }
  MaterialExpressionTextureObjectParameter_10 {
   class=MaterialExpressionTextureObjectParameter
   inputs [
    UVs {from="", output=""}
    Apply View MipBias {from="", output=""}
   ]
   properties {
    group=ClearCoat
    parameterName=ClearCoatNormalTexture
    samplerType=SAMPLERTYPE_Normal
    sortPriority=1
    texture=&DefaultNormal
   }
  }
  MaterialExpressionTextureObjectParameter_19 {
   class=MaterialExpressionTextureObjectParameter
   inputs [
    UVs {from="", output=""}
    Apply View MipBias {from="", output=""}
   ]
   properties {
    group=Diffuse
    parameterName=DiffuseColorTexture
    samplerType=SAMPLERTYPE_LinearColor
    sortPriority=1
    texture=/Engine/Functions/Engine_MaterialFunctions02/PivotPainter2/Black_1x1_EXR_Texture
   }
  }
  MaterialExpressionTextureObjectParameter_20 {
   class=MaterialExpressionTextureObjectParameter
   inputs [
    UVs {from="", output=""}
    Apply View MipBias {from="", output=""}
   ]
   properties {
    group=Roughness
    parameterName=RoughnessTexture
    samplerType=SAMPLERTYPE_LinearColor
    sortPriority=1
    texture=&Black_1x1_EXR_Texture
   }
  }
  MaterialExpressionTextureObjectParameter_21 {
   class=MaterialExpressionTextureObjectParameter
   inputs [
    UVs {from="", output=""}
    Apply View MipBias {from="", output=""}
   ]
   properties {
    group=Anisotropy
    parameterName=AnisotropicRotationTexture
    samplerType=SAMPLERTYPE_LinearColor
    sortPriority=1
    texture=&Black_1x1_EXR_Texture
   }
  }
  MaterialExpressionTextureObjectParameter_22 {
   class=MaterialExpressionTextureObjectParameter
   inputs [
    UVs {from="", output=""}
    Apply View MipBias {from="", output=""}
   ]
   properties {
    group=Fresnel
    parameterName=F90Texture
    samplerType=SAMPLERTYPE_LinearColor
    sortPriority=2
    texture=&Black_1x1_EXR_Texture
   }
  }
  MaterialExpressionTextureObjectParameter_23 {
   class=MaterialExpressionTextureObjectParameter
   inputs [
    UVs {from="", output=""}
    Apply View MipBias {from="", output=""}
   ]
   properties {
    group=Fresnel
    parameterName=F0Texture
    samplerType=SAMPLERTYPE_LinearColor
    sortPriority=1
    texture=&Black_1x1_EXR_Texture
   }
  }
  MaterialExpressionTextureObjectParameter_24 {
   class=MaterialExpressionTextureObjectParameter
   inputs [
    UVs {from="", output=""}
    Apply View MipBias {from="", output=""}
   ]
   properties {
    group=Sheen
    parameterName=SheenColorTexture
    samplerType=SAMPLERTYPE_LinearColor
    sortPriority=2
    texture=&Black_1x1_EXR_Texture
   }
  }
  MaterialExpressionTextureObjectParameter_25 {
   class=MaterialExpressionTextureObjectParameter
   inputs [
    UVs {from="", output=""}
    Apply View MipBias {from="", output=""}
   ]
   properties {
    group=Sheen
    parameterName=SheenRoughnessTexture
    samplerType=SAMPLERTYPE_LinearColor
    sortPriority=1
    texture=&Black_1x1_EXR_Texture
   }
  }
  MaterialExpressionTextureSample_0 {
   class=MaterialExpressionTextureSample
   inputs [
    UVs {from=MaterialExpressionNamedRerouteUsage_31}
    Tex {from=MaterialExpressionTextureObjectParameter_10}
    Apply View MipBias {from=""}
   ]
  }
  MaterialExpressionTextureSample_15 {
   class=MaterialExpressionTextureSample
   inputs [
    UVs {from=MaterialExpressionNamedRerouteUsage_1}
    Tex {from=MaterialExpressionTextureObjectParameter_19}
    Apply View MipBias {from=""}
   ]
   properties {samplerType=SAMPLERTYPE_LinearColor}
  }
  MaterialExpressionTextureSample_2 {
   class=MaterialExpressionTextureSample
   inputs [
    UVs {from=MaterialExpressionNamedRerouteUsage_21}
    Tex {from=MaterialExpressionTextureObjectParameter_23}
    Apply View MipBias {from=""}
   ]
  }
  MaterialExpressionTextureSample_3 {
   class=MaterialExpressionTextureSample
   inputs [
    UVs {from=MaterialExpressionNamedRerouteUsage_22}
    Tex {from=MaterialExpressionTextureObjectParameter_22}
    Apply View MipBias {from=""}
   ]
  }
  MaterialExpressionTextureSample_4 {
   class=MaterialExpressionTextureSample
   inputs [
    UVs {from=MaterialExpressionNamedRerouteUsage_7}
    Tex {from=MaterialExpressionTextureObjectParameter_25}
    Apply View MipBias {from=""}
   ]
  }
  MaterialExpressionTextureSample_5 {
   class=MaterialExpressionTextureSample
   inputs [
    UVs {from=MaterialExpressionNamedRerouteUsage_8}
    Tex {from=MaterialExpressionTextureObjectParameter_24}
    Apply View MipBias {from=""}
   ]
  }
  MaterialExpressionTextureSample_6 {
   class=MaterialExpressionTextureSample
   inputs [
    UVs {from=MaterialExpressionNamedRerouteUsage_5}
    Tex {from=MaterialExpressionTextureObjectParameter_1}
    Apply View MipBias {from=""}
   ]
  }
  MaterialExpressionTextureSample_7 {
   class=MaterialExpressionTextureSample
   inputs [
    UVs {from=MaterialExpressionNamedRerouteUsage_38}
    Tex {from=MaterialExpressionTextureObjectParameter_21}
    Apply View MipBias {from=""}
   ]
  }
  MaterialExpressionTextureSample_8 {
   class=MaterialExpressionTextureSample
   inputs [
    UVs {from=MaterialExpressionNamedRerouteUsage_36}
    Tex {from=MaterialExpressionTextureObjectParameter_20}
    Apply View MipBias {from=""}
   ]
  }
  MaterialExpressionVectorParameter_11 {class=MaterialExpressionVectorParameter, properties {group=ClearCoat, parameterName=ClearCoatNormal, defaultValue=<rgba 0 0 1 1>}}
  MaterialExpressionVectorParameter_18 {class=MaterialExpressionVectorParameter, properties {group=Diffuse, parameterName=DiffuseColor, defaultValue=<rgba 0 0 0 1>}}
  MaterialExpressionVectorParameter_19 {class=MaterialExpressionVectorParameter, properties {group=Normal, parameterName=Normal, defaultValue=<rgba 0 0 1 1>}}
  MaterialExpressionVectorParameter_20 {
   class=MaterialExpressionVectorParameter
   properties {group=Roughness, parameterName=Roughness, sortPriority=30, defaultValue=<rgba 0.2122 0.2122 0 1>}
  }
  MaterialExpressionVectorParameter_21 {class=MaterialExpressionVectorParameter, properties {group=Fresnel, parameterName=F0, defaultValue=<rgba 0.04 0.04 0.04 1>}}
  MaterialExpressionVectorParameter_22 {class=MaterialExpressionVectorParameter, properties {group=Fresnel, parameterName=F90, defaultValue=<rgba 1 1 1 0>}}
  MaterialExpressionVectorParameter_23 {class=MaterialExpressionVectorParameter, properties {group=Sheen, parameterName=SheenColor, defaultValue=<rgba 0 0 0 1>}}
  MaterialExpressionVectorParameter_24 {class=MaterialExpressionVectorParameter, properties {group=Sheen, parameterName=SheenRoughness, defaultValue=<rgba 0.1 0.1 0 1>}}
  MaterialExpressionVectorParameter_25 {class=MaterialExpressionVectorParameter, properties {group=Anisotropy, parameterName=AnisotropicRotation, defaultValue=<rgba 0.5 0.5 0 0>}}
  MaterialExpressionVertexNormalWS_0 {class=MaterialExpressionVertexNormalWS, inputs [], properties {}}
 }
 properties {
  bUsedWithNanite=true
  bUsedWithStaticLighting=true
  ditherOpacityMask=true
  shadingModel=MSM_SubsurfaceProfile
  translucencyLightingMode=TLM_SurfacePerPixelLighting
 }
}
