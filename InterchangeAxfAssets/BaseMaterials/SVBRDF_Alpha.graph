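# Base material for imported AxF SVBRDF materials with alpha, covering diffuse, normal, Fresnel, anisotropic roughness, sheen and clear coat.
# Millimeter-scaled coordinates come from geometry UVs or triplanar projection in local or world space, with gradients and tangent frames for sampling.
# Static switches choose between textures and constant parameters, feeding named reroutes for each shading term.
# A ShadowReplace node scales alpha by inverted ShadowCompensation for shadows only.
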Material /InterchangeAxfAssets/BaseMaterials/SVBRDF_Alpha
graph {
 expressions {
  MaterialExpressionAbs_1 {class=MaterialExpressionAbs, inputs [None {from=MaterialExpressionDotProduct_1, output=""}], properties {}}
  MaterialExpressionAdd_1 {
   class=MaterialExpressionAdd
   inputs [
    A {from=MaterialExpressionMaterialFunctionCall_36, output=R}
    B {from=MaterialExpressionMaterialFunctionCall_36, output=R}
   ]
  }
  MaterialExpressionCameraVectorWS_1 {class=MaterialExpressionCameraVectorWS, inputs [], properties {}}
  MaterialExpressionClamp_0 {
   class=MaterialExpressionClamp
   inputs [
    None {from=MaterialExpressionMultiply_2, output=""}
    Min {from="", output=""}
    Max {from="", output=""}
   ]
   properties {}
  }
  MaterialExpressionComponentMask_0 {
   class=MaterialExpressionComponentMask
   inputs [None {from=MaterialExpressionVectorParameter_1, output=RGB}]
   properties {b=true, g=true, r=true}
  }
  MaterialExpressionComponentMask_2 {
   class=MaterialExpressionComponentMask
   inputs [None {from=MaterialExpressionVectorParameter_27, output=RGB}]
   properties {b=true, g=true, r=true}
  }
  MaterialExpressionConstant_12 {class=MaterialExpressionConstant}
  MaterialExpressionConstant_13 {class=MaterialExpressionConstant}
  MaterialExpressionConstant_14 {class=MaterialExpressionConstant}
  MaterialExpressionConstant_19 {class=MaterialExpressionConstant, properties {r=1}}
  MaterialExpressionConstant_2 {class=MaterialExpressionConstant, properties {desc="1 = local", r=1}}
  MaterialExpressionConstant_3 {class=MaterialExpressionConstant, properties {desc="0 = world"}}
  MaterialExpressionConvert_0 {
   class=MaterialExpressionConvert
   inputs [None {from=MaterialExpressionMaterialFunctionCall_8, output=DXY}]
   properties {
    convertInputs [{
      type=Vector4
      defaultValue=<rgba 0 0 0 1>
      expressionInput {
       expression=&self:MaterialExpressionMaterialFunctionCall_8
       inputName=None
       mask=0
       maskA=0
       maskB=0
       maskG=0
       maskR=0
       outputIndex=4
      }
     }]
    convertOutputs [{type=Vector2}, {type=Vector2}]
   }
  }
  MaterialExpressionConvert_1 {
   class=MaterialExpressionConvert
   inputs [
    None {from=MaterialExpressionNamedRerouteUsage_53}
    None {from=MaterialExpressionNamedRerouteUsage_54}
    None {from=MaterialExpressionNamedRerouteUsage_55}
    None {from=MaterialExpressionNamedRerouteUsage_52}
   ]
   properties {
    convertInputs [
     {
      type=Scalar
      expressionInput {
       expression=&self:MaterialExpressionNamedRerouteUsage_53
       inputName=None
       mask=0
       maskA=0
       maskB=0
       maskG=0
       maskR=0
       outputIndex=0
      }
     }
     {
      type=Scalar
      expressionInput {
       expression=&self:MaterialExpressionNamedRerouteUsage_54
       inputName=None
       mask=0
       maskA=0
       maskB=0
       maskG=0
       maskR=0
       outputIndex=0
      }
     }
     {
      type=Scalar
      expressionInput {
       expression=&self:MaterialExpressionNamedRerouteUsage_55
       inputName=None
       mask=0
       maskA=0
       maskB=0
       maskG=0
       maskR=0
       outputIndex=0
      }
     }
     {
      type=Scalar
      expressionInput {
       expression=&self:MaterialExpressionNamedRerouteUsage_52
       inputName=None
       mask=0
       maskA=0
       maskB=0
       maskG=0
       maskR=0
       outputIndex=0
      }
     }
    ]
    convertOutputs [{type=Vector4, defaultValue=<rgba 0 0 0 1>}]
   }
  }
  MaterialExpressionConvert_2 {
   class=MaterialExpressionConvert
   inputs [
    None {from=MaterialExpressionNamedRerouteUsage_33}
    None {from=MaterialExpressionNamedRerouteUsage_34}
    None {from=MaterialExpressionNamedRerouteUsage_35}
   ]
   properties {
    convertInputs [
     {
      type=Scalar
      expressionInput {
       expression=&self:MaterialExpressionNamedRerouteUsage_33
       inputName=None
       mask=0
       maskA=0
       maskB=0
       maskG=0
       maskR=0
       outputIndex=0
      }
     }
     {
      type=Scalar
      expressionInput {
       expression=&self:MaterialExpressionNamedRerouteUsage_34
       inputName=None
       mask=0
       maskA=0
       maskB=0
       maskG=0
       maskR=0
       outputIndex=0
      }
     }
     {
      type=Scalar
      expressionInput {
       expression=&self:MaterialExpressionNamedRerouteUsage_35
       inputName=None
       mask=0
       maskA=0
       maskB=0
       maskG=0
       maskR=0
       outputIndex=0
      }
     }
    ]
    convertOutputs [{type=Vector3}]
   }
  }
  MaterialExpressionDivide_3 {
   class=MaterialExpressionDivide
   inputs [
    A {from=""}
    B {from=MaterialExpressionScalarParameter_20}
   ]
   properties {constA=10, constB=10}
  }
  MaterialExpressionDivide_4 {
   class=MaterialExpressionDivide
   inputs [
    A {from=""}
    B {from=MaterialExpressionScalarParameter_21}
   ]
   properties {constA=10, constB=10}
  }
  MaterialExpressionDivide_6 {
   class=MaterialExpressionDivide
   inputs [
    A {from=MaterialExpressionAdd_1}
    B {from=""}
   ]
  }
  MaterialExpressionDotProduct_1 {
   class=MaterialExpressionDotProduct
   inputs [
    A {from=MaterialExpressionShaderStageSwitch_0, output=""}
    B {from=MaterialExpressionVertexNormalWS_2, output=""}
   ]
   properties {}
  }
  MaterialExpressionMaterialFunctionCall_1 {
   class=MaterialExpressionMaterialFunctionCall
   inputs []
   properties {materialFunction=&CameraDirectionVector}
  }
  MaterialExpressionMaterialFunctionCall_10 {
   class=MaterialExpressionMaterialFunctionCall
   inputs [
    NormalIn {from=MaterialExpressionStaticSwitch_1}
    NormalStrength {from=MaterialExpressionScalarParameter_25}
    ReFresnelVX {from=MaterialExpressionNamedRerouteUsage_29}
   ]
   properties {materialFunction=/InterchangeAxfAssets/MaterialFunctions/MF_NormalStrength}
  }
  MaterialExpressionMaterialFunctionCall_12 {
   class=MaterialExpressionMaterialFunctionCall
   inputs [
    Coordinates {from=MaterialExpressionNamedRerouteUsage_39}
    TexObject {from=MaterialExpressionTextureObjectParameter_2}
    DDX {from=MaterialExpressionNamedRerouteUsage_23}
    DDY {from=MaterialExpressionNamedRerouteUsage_24}
    Mask {from=""}
    DefaultUV {from=""}
   ]
   properties {materialFunction=/BaseMaterial/Materials/Functions/MF_TexPointSample_NRM}
  }
  MaterialExpressionMaterialFunctionCall_13 {
   class=MaterialExpressionMaterialFunctionCall
   inputs [
    TSN {from=MaterialExpressionMaterialFunctionCall_12, output=RGB}
    Tangent {from=MaterialExpressionNamedRerouteUsage_64}
    Bitangent {from=MaterialExpressionNamedRerouteUsage_65}
    "Rotation (Turns)" {from=MaterialExpressionNamedRerouteUsage_27}
   ]
   properties {materialFunction=/BaseMaterial/Materials/Functions/MF_TangentNormal_OrientWorld}
  }
  MaterialExpressionMaterialFunctionCall_15 {
   class=MaterialExpressionMaterialFunctionCall
   inputs [
    Coordinates {from=MaterialExpressionNamedRerouteUsage_190}
    TexObject {from=MaterialExpressionTextureObjectParameter_30}
    DDX {from=MaterialExpressionNamedRerouteUsage_188}
    DDY {from=MaterialExpressionNamedRerouteUsage_189}
    Mask {from=""}
    DefaultUV {from=""}
   ]
   properties {materialFunction=/BaseMaterial/Materials/Functions/MF_TexPointSample_RGB}
  }
  MaterialExpressionMaterialFunctionCall_16 {
   class=MaterialExpressionMaterialFunctionCall
   inputs [SheenRoughness {from=MaterialExpressionStaticSwitchParameter_20}]
   properties {materialFunction=/InterchangeAxfAssets/MaterialFunctions/MF_SheenRoughness}
  }
  MaterialExpressionMaterialFunctionCall_18 {
   class=MaterialExpressionMaterialFunctionCall
   inputs [
    Coordinates {from=MaterialExpressionNamedRerouteUsage_197}
    TexObject {from=MaterialExpressionTextureObjectParameter_31}
    DDX {from=MaterialExpressionNamedRerouteUsage_195}
    DDY {from=MaterialExpressionNamedRerouteUsage_196}
    Mask {from=""}
    DefaultUV {from=""}
   ]
   properties {materialFunction=&MF_TexPointSample_RGB}
  }
  MaterialExpressionMaterialFunctionCall_21 {
   class=MaterialExpressionMaterialFunctionCall
   inputs [
    Coordinates {from=MaterialExpressionNamedRerouteUsage_114}
    TexObject {from=MaterialExpressionTextureObjectParameter_8}
    DDX {from=MaterialExpressionNamedRerouteUsage_110}
    DDY {from=MaterialExpressionNamedRerouteUsage_111}
    Mask {from=""}
    DefaultUV {from=""}
   ]
   properties {materialFunction=&MF_TexPointSample_RGB}
  }
  MaterialExpressionMaterialFunctionCall_23 {
   class=MaterialExpressionMaterialFunctionCall
   inputs [
    Coordinates {from=MaterialExpressionNamedRerouteUsage_119}
    TexObject {from=MaterialExpressionTextureObjectParameter_7}
    DDX {from=MaterialExpressionNamedRerouteUsage_112}
    DDY {from=MaterialExpressionNamedRerouteUsage_113}
    Mask {from=""}
    DefaultUV {from=""}
   ]
   properties {materialFunction=&MF_TexPointSample_RGB}
  }
  MaterialExpressionMaterialFunctionCall_29 {
   class=MaterialExpressionMaterialFunctionCall
   inputs [
    Coordinates {from=MaterialExpressionNamedRerouteUsage_167}
    TexObject {from=MaterialExpressionTextureObjectParameter_26}
    DDX {from=MaterialExpressionNamedRerouteUsage_163}
    DDY {from=MaterialExpressionNamedRerouteUsage_164}
    Mask {from=""}
    DefaultUV {from=""}
   ]
   properties {materialFunction=&MF_TexPointSample_RGB}
  }
  MaterialExpressionMaterialFunctionCall_30 {
   class=MaterialExpressionMaterialFunctionCall
   inputs [
    Coordinates {from=MaterialExpressionNamedRerouteUsage_170}
    TexObject {from=MaterialExpressionTextureObjectParameter_18}
    DDX {from=MaterialExpressionNamedRerouteUsage_165}
    DDY {from=MaterialExpressionNamedRerouteUsage_166}
    Mask {from=""}
    DefaultUV {from=""}
   ]
   properties {materialFunction=&MF_TexPointSample_RGB}
  }
  MaterialExpressionMaterialFunctionCall_31 {
   class=MaterialExpressionMaterialFunctionCall
   inputs [ClearCoatNormalV {from=MaterialExpressionNamedRerouteUsage_173}]
   properties {materialFunction=/InterchangeAxfAssets/MaterialFunctions/MF_ClearCoatTerm}
  }
  MaterialExpressionMaterialFunctionCall_32 {
   class=MaterialExpressionMaterialFunctionCall
   inputs [
    Coordinates {from=MaterialExpressionNamedRerouteUsage_176}
    TexObject {from=MaterialExpressionTextureObjectParameter_28}
    DDX {from=MaterialExpressionNamedRerouteUsage_174}
    DDY {from=MaterialExpressionNamedRerouteUsage_175}
    Mask {from=""}
    DefaultUV {from=""}
   ]
   properties {materialFunction=&MF_TexPointSample_NRM}
  }
  MaterialExpressionMaterialFunctionCall_33 {
   class=MaterialExpressionMaterialFunctionCall
   inputs [
    NormalIn {from=MaterialExpressionStaticSwitch_6}
    NormalStrength {from=MaterialExpressionScalarParameter_29}
    ReFresnelVX {from=MaterialExpressionNamedRerouteUsage_179}
   ]
   properties {materialFunction=&MF_NormalStrength}
  }
  MaterialExpressionMaterialFunctionCall_34 {
   class=MaterialExpressionMaterialFunctionCall
   inputs [
    TSN {from=MaterialExpressionMaterialFunctionCall_32, output=RGB}
    Tangent {from=MaterialExpressionNamedRerouteUsage_181}
    Bitangent {from=MaterialExpressionNamedRerouteUsage_180}
    "Rotation (Turns)" {from=MaterialExpressionNamedRerouteUsage_182}
   ]
   properties {materialFunction=&MF_TangentNormal_OrientWorld}
  }
  MaterialExpressionMaterialFunctionCall_36 {
   class=MaterialExpressionMaterialFunctionCall
   inputs [Color {from=MaterialExpressionStaticSwitchParameter_35}]
   properties {materialFunction=&SplitComponents}
  }
  MaterialExpressionMaterialFunctionCall_37 {
   class=MaterialExpressionMaterialFunctionCall
   inputs [
    X {from=MaterialExpressionTextureSample_7, output=R}
    Y {from=MaterialExpressionTextureSample_7, output=R}
   ]
   properties {materialFunction=&MakeFloat2}
  }
  MaterialExpressionMaterialFunctionCall_38 {
   class=MaterialExpressionMaterialFunctionCall
   inputs [
    X {from=MaterialExpressionMaterialFunctionCall_18, output=R}
    Y {from=MaterialExpressionMaterialFunctionCall_18, output=R}
   ]
   properties {materialFunction=&MakeFloat2}
  }
  MaterialExpressionMaterialFunctionCall_39 {
   class=MaterialExpressionMaterialFunctionCall
   inputs [
    AnisotropicRotation {from=MaterialExpressionStaticSwitchParameter_36}
    AlphaX {from=MaterialExpressionNamedRerouteUsage_193}
    AlphaY {from=MaterialExpressionNamedRerouteUsage_194}
   ]
   properties {materialFunction=/InterchangeAxfAssets/MaterialFunctions/MF_AnisotropyComponent}
  }
  MaterialExpressionMaterialFunctionCall_40 {
   class=MaterialExpressionMaterialFunctionCall
   inputs [
    X {from=MaterialExpressionVectorParameter_30, output=R}
    Y {from=MaterialExpressionVectorParameter_30, output=R}
   ]
   properties {materialFunction=&MakeFloat2}
  }
  MaterialExpressionMaterialFunctionCall_41 {
   class=MaterialExpressionMaterialFunctionCall
   inputs [
    Coordinates {from=MaterialExpressionNamedRerouteUsage_212}
    TexObject {from=MaterialExpressionTextureObjectParameter_27}
    DDX {from=MaterialExpressionNamedRerouteUsage_213}
    DDY {from=MaterialExpressionNamedRerouteUsage_214}
    Mask {from=""}
    DefaultUV {from=""}
   ]
   properties {materialFunction=&MF_TexPointSample_RGB}
  }
  MaterialExpressionMaterialFunctionCall_6 {
   class=MaterialExpressionMaterialFunctionCall
   inputs [
    Coordinates {from=MaterialExpressionNamedRerouteUsage_19}
    TexObject {from=MaterialExpressionTextureObjectParameter_0}
    DDX {from=MaterialExpressionNamedRerouteUsage_17}
    DDY {from=MaterialExpressionNamedRerouteUsage_18}
    Mask {from=""}
    DefaultUV {from=""}
   ]
   properties {materialFunction=&MF_TexPointSample_RGB}
  }
  MaterialExpressionMaterialFunctionCall_7 {
   class=MaterialExpressionMaterialFunctionCall
   inputs [
    AxF_Scale_X {from=MaterialExpressionNamedRerouteUsage_97}
    AxF_Scale_Y {from=MaterialExpressionNamedRerouteUsage_98}
    AxF_UnitScale {from=MaterialExpressionNamedRerouteUsage_99}
    Tile U {from=MaterialExpressionNamedRerouteUsage_93}
    Tile V {from=MaterialExpressionNamedRerouteUsage_94}
    Offset U {from=MaterialExpressionNamedRerouteUsage_95}
    Offset V {from=MaterialExpressionNamedRerouteUsage_96}
    Rotation {from=MaterialExpressionNamedRerouteUsage_92}
    UV Index {from=MaterialExpressionScalarParameter_23}
   ]
   properties {materialFunction=/InterchangeAxfAssets/MaterialFunctions/MF_AxFMakeCoordinatesUV}
  }
  MaterialExpressionMaterialFunctionCall_8 {
   class=MaterialExpressionMaterialFunctionCall
   inputs [
    AxF_Scale_X {from=MaterialExpressionNamedRerouteUsage_56}
    AxF_Scale_Y {from=MaterialExpressionNamedRerouteUsage_59}
    AxF_UnitScale {from=MaterialExpressionNamedRerouteUsage_60}
    Tile XYZ {from=MaterialExpressionNamedRerouteUsage_61}
    Offset XYZ {from=MaterialExpressionNamedRerouteUsage_63}
    Transition Softness {from=MaterialExpressionNamedRerouteUsage_100}
    Coordinate Space {from=MaterialExpressionStaticSwitchParameter_2}
    Rotation {from=MaterialExpressionNamedRerouteUsage_62}
   ]
   properties {materialFunction=/InterchangeAxfAssets/MaterialFunctions/MF_AxFMakeCoordinatesTP}
  }
  MaterialExpressionMaterialFunctionCall_9 {
   class=MaterialExpressionMaterialFunctionCall
   inputs [
    NormalTexOBJ {from=MaterialExpressionTextureObjectParameter_2}
    UV_TC {from=MaterialExpressionNamedRerouteUsage_103}
    Rotation_TC {from=MaterialExpressionNamedRerouteUsage_28}
   ]
   properties {materialFunction=/InterchangeAxfAssets/MaterialFunctions/MF_TexCoodUVNormalSampler}
  }
  MaterialExpressionMultiply_2 {
   class=MaterialExpressionMultiply
   inputs [
    A {from=MaterialExpressionStaticSwitchParameter_94}
    B {from=MaterialExpressionOneMinus_0}
   ]
  }
  MaterialExpressionMultiply_5 {
   class=MaterialExpressionMultiply
   inputs [
    A {from=MaterialExpressionScalarParameter_22}
    B {from=""}
   ]
   properties {constB=10}
  }
  MaterialExpressionNamedRerouteDeclaration_10 {
   class=MaterialExpressionNamedRerouteDeclaration
   inputs [None {from=MaterialExpressionMaterialFunctionCall_8, output=Output UV}]
   properties {name=TP_UV, nodeColor=<rgba 0 0.1094 1 1>}
  }
  MaterialExpressionNamedRerouteDeclaration_11 {
   class=MaterialExpressionNamedRerouteDeclaration
   inputs [None {from=MaterialExpressionSaturate_1}]
   properties {name=ReFresnelVX, nodeColor=<rgba 0.2656 0 1 1>}
  }
  MaterialExpressionNamedRerouteDeclaration_12 {
   class=MaterialExpressionNamedRerouteDeclaration
   inputs [None {from=MaterialExpressionScalarParameter_13}]
   properties {name=UV_Scale, nodeColor=<rgba 0.7578 1 0 1>}
  }
  MaterialExpressionNamedRerouteDeclaration_139 {
   class=MaterialExpressionNamedRerouteDeclaration
   inputs [None {from=MaterialExpressionShadowReplace_0}]
   properties {name=Alpha_O, nodeColor=<rgba 0.3594 0 1 1>}
  }
  MaterialExpressionNamedRerouteDeclaration_14 {
   class=MaterialExpressionNamedRerouteDeclaration
   inputs [None {from=MaterialExpressionScalarParameter_8}]
   properties {name=UV_Rot, nodeColor=<rgba 0.7578 1 0 1>}
  }
  MaterialExpressionNamedRerouteDeclaration_15 {
   class=MaterialExpressionNamedRerouteDeclaration
   inputs [None {from=MaterialExpressionScalarParameter_14}]
   properties {name=UV_X, nodeColor=<rgba 0 1 0.6016 1>}
  }
  MaterialExpressionNamedRerouteDeclaration_17 {
   class=MaterialExpressionNamedRerouteDeclaration
   inputs [None {from=MaterialExpressionScalarParameter_15}]
   properties {name=UV_Y, nodeColor=<rgba 0 1 0.6016 1>}
  }
  MaterialExpressionNamedRerouteDeclaration_18 {
   class=MaterialExpressionNamedRerouteDeclaration
   inputs [None {from=MaterialExpressionScalarParameter_18}]
   properties {name=UV_Z, nodeColor=<rgba 0 1 0.6016 1>}
  }
  MaterialExpressionNamedRerouteDeclaration_19 {
   class=MaterialExpressionNamedRerouteDeclaration
   inputs [None {from=MaterialExpressionScalarParameter_17}]
   properties {name=OffsetX, nodeColor=<rgba 0.3594 1 0 1>}
  }
  MaterialExpressionNamedRerouteDeclaration_20 {
   class=MaterialExpressionNamedRerouteDeclaration
   inputs [None {from=MaterialExpressionScalarParameter_16}]
   properties {name=OffsetY, nodeColor=<rgba 0.3594 1 0 1>}
  }
  MaterialExpressionNamedRerouteDeclaration_21 {
   class=MaterialExpressionNamedRerouteDeclaration
   inputs [None {from=MaterialExpressionScalarParameter_19}]
   properties {name=OffsetZ, nodeColor=<rgba 0.3594 1 0 1>}
  }
  MaterialExpressionNamedRerouteDeclaration_26 {
   class=MaterialExpressionNamedRerouteDeclaration
   inputs [None {from=MaterialExpressionDivide_3}]
   properties {name=AxF_ScaleX, nodeColor=<rgba 0.7605 0 0.0009 1>}
  }
  MaterialExpressionNamedRerouteDeclaration_28 {
   class=MaterialExpressionNamedRerouteDeclaration
   inputs [None {from=MaterialExpressionDivide_4}]
   properties {name=AxF_ScaleY, nodeColor=<rgba 0.7605 0 0.0009 1>}
  }
  MaterialExpressionNamedRerouteDeclaration_29 {
   class=MaterialExpressionNamedRerouteDeclaration
   inputs [None {from=MaterialExpressionMultiply_5}]
   properties {name=AxF_UnitScale_MM, nodeColor=<rgba 0.7605 0 0.0009 1>}
  }
  MaterialExpressionNamedRerouteDeclaration_30 {
   class=MaterialExpressionNamedRerouteDeclaration
   inputs [None {from=MaterialExpressionConvert_1}]
   properties {name=TP_TileXYZ, nodeColor=<rgba 1 0.8438 0 1>}
  }
  MaterialExpressionNamedRerouteDeclaration_31 {
   class=MaterialExpressionNamedRerouteDeclaration
   inputs [None {from=MaterialExpressionConvert_2}]
   properties {name=TP_OffsetXYZ, nodeColor=<rgba 0 1 0.5078 1>}
  }
  MaterialExpressionNamedRerouteDeclaration_32 {
   class=MaterialExpressionNamedRerouteDeclaration
   inputs [None {from=MaterialExpressionMaterialFunctionCall_8, output=T}]
   properties {name=TP_Tangent, nodeColor=<rgba 1 0 0.0938 1>}
  }
  MaterialExpressionNamedRerouteDeclaration_33 {
   class=MaterialExpressionNamedRerouteDeclaration
   inputs [None {from=MaterialExpressionMaterialFunctionCall_8, output=B}]
   properties {name=TP_Bitangent, nodeColor=<rgba 0 0.4141 1 1>}
  }
  MaterialExpressionNamedRerouteDeclaration_34 {
   class=MaterialExpressionNamedRerouteDeclaration
   inputs [None {from=MaterialExpressionMaterialFunctionCall_7, output=Output UV}]
   properties {name=TC_UV, nodeColor=<rgba 0 1 0.25 1>}
  }
  MaterialExpressionNamedRerouteDeclaration_36 {
   class=MaterialExpressionNamedRerouteDeclaration
   inputs [None {from=MaterialExpressionStaticBoolParameter_1}]
   properties {name=UseGeoUVs, nodeColor=<rgba 0 0.8359 1 1>}
  }
  MaterialExpressionNamedRerouteDeclaration_37 {
   class=MaterialExpressionNamedRerouteDeclaration
   inputs [None {from=MaterialExpressionScalarParameter_24}]
   properties {name=TransitionSoftness, nodeColor=<rgba 0 1 0.2031 1>}
  }
  MaterialExpressionNamedRerouteDeclaration_38 {
   class=MaterialExpressionNamedRerouteDeclaration
   inputs [None {from=MaterialExpressionStaticSwitchParameter_0}]
   properties {name=DiffuseAlbedoTerm_O, nodeColor=<rgba 0 1 0.0391 1>}
  }
  MaterialExpressionNamedRerouteDeclaration_39 {
   class=MaterialExpressionNamedRerouteDeclaration
   inputs [None {from=MaterialExpressionStaticSwitchParameter_3}]
   properties {name=Normal_O, nodeColor=<rgba 1 0.9844 0 1>}
  }
  MaterialExpressionNamedRerouteDeclaration_6 {
   class=MaterialExpressionNamedRerouteDeclaration
   inputs [None {from=MaterialExpressionMaterialFunctionCall_8, output=Turns}]
   properties {name=TP_Rotation, nodeColor=<rgba 0.9453 1 0 1>}
  }
  MaterialExpressionNamedRerouteDeclaration_62 {
   class=MaterialExpressionNamedRerouteDeclaration
   inputs [None {from=MaterialExpressionStaticSwitchParameter_16}]
   properties {name=FuzzRoughness_O, nodeColor=<rgba 0.0547 1 0 1>}
  }
  MaterialExpressionNamedRerouteDeclaration_63 {
   class=MaterialExpressionNamedRerouteDeclaration
   inputs [None {from=MaterialExpressionStaticSwitchParameter_17}]
   properties {name=FuzzAmount_O, nodeColor=<rgba 0.0547 1 0 1>}
  }
  MaterialExpressionNamedRerouteDeclaration_64 {
   class=MaterialExpressionNamedRerouteDeclaration
   inputs [None {from=MaterialExpressionStaticSwitchParameter_18}]
   properties {name=FuzzColor_O, nodeColor=<rgba 0.0547 1 0 1>}
  }
  MaterialExpressionNamedRerouteDeclaration_65 {
   class=MaterialExpressionNamedRerouteDeclaration
   inputs [None {from=MaterialExpressionMaterialFunctionCall_16, output=IsotropicRoughness}]
   properties {name=SheenRoughness, nodeColor=<rgba 1 0 0.1641 1>}
  }
  MaterialExpressionNamedRerouteDeclaration_74 {
   class=MaterialExpressionNamedRerouteDeclaration
   inputs [None {from=MaterialExpressionStaticSwitchParameter_31}]
   properties {name=F0_O, nodeColor=<rgba 1 0 0.2812 1>}
  }
  MaterialExpressionNamedRerouteDeclaration_75 {
   class=MaterialExpressionNamedRerouteDeclaration
   inputs [None {from=MaterialExpressionStaticSwitchParameter_32}]
   properties {name=F90_O, nodeColor=<rgba 1 0 0.2812 1>}
  }
  MaterialExpressionNamedRerouteDeclaration_76 {
   class=MaterialExpressionNamedRerouteDeclaration
   inputs [None {from=MaterialExpressionMaterialFunctionCall_31, output=Top Layer Slab}]
   properties {name=ClearCoatTerm_O, nodeColor=<rgba 1 0.6797 0 1>}
  }
  MaterialExpressionNamedRerouteDeclaration_77 {
   class=MaterialExpressionNamedRerouteDeclaration
   inputs [None {from=MaterialExpressionStaticSwitchParameter_33}]
   properties {name=CCNormal, nodeColor=<rgba 0.5469 0 1 1>}
  }
  MaterialExpressionNamedRerouteDeclaration_8 {
   class=MaterialExpressionNamedRerouteDeclaration
   inputs [None {from=MaterialExpressionConvert_0}]
   properties {name=TP_DDY, nodeColor=<rgba 0 0.7891 1 1>}
  }
  MaterialExpressionNamedRerouteDeclaration_81 {
   class=MaterialExpressionNamedRerouteDeclaration
   inputs [None {from=MaterialExpressionSquareRoot_1}]
   properties {name=Roughness_O, nodeColor=<rgba 0.9219 1 0 1>}
  }
  MaterialExpressionNamedRerouteDeclaration_82 {
   class=MaterialExpressionNamedRerouteDeclaration
   inputs [None {from=MaterialExpressionMaterialFunctionCall_36, output=R}]
   properties {name=AlphaX, nodeColor=<rgba 0.1484 1 0 1>}
  }
  MaterialExpressionNamedRerouteDeclaration_83 {
   class=MaterialExpressionNamedRerouteDeclaration
   inputs [None {from=MaterialExpressionMaterialFunctionCall_36, output=G}]
   properties {name=AlphaY, nodeColor=<rgba 0.1484 1 0 1>}
  }
  MaterialExpressionNamedRerouteDeclaration_84 {
   class=MaterialExpressionNamedRerouteDeclaration
   inputs [None {from=MaterialExpressionMaterialFunctionCall_39, output=Anisotropy}]
   properties {name=Anisotropy_O, nodeColor=<rgba 0.9219 1 0 1>}
  }
  MaterialExpressionNamedRerouteDeclaration_85 {
   class=MaterialExpressionNamedRerouteDeclaration
   inputs [None {from=MaterialExpressionMaterialFunctionCall_39, output=Tangent}]
   properties {name=Tangent_O, nodeColor=<rgba 0.9219 1 0 1>}
  }
  MaterialExpressionNamedRerouteDeclaration_9 {
   class=MaterialExpressionNamedRerouteDeclaration
   inputs [None {from=MaterialExpressionConvert_0}]
   properties {name=TP_DDX, nodeColor=<rgba 0.5703 0 1 1>}
  }
  MaterialExpressionNamedRerouteUsage_100 {class=MaterialExpressionNamedRerouteUsage}
  MaterialExpressionNamedRerouteUsage_102 {class=MaterialExpressionNamedRerouteUsage}
  MaterialExpressionNamedRerouteUsage_103 {class=MaterialExpressionNamedRerouteUsage}
  MaterialExpressionNamedRerouteUsage_107 {class=MaterialExpressionNamedRerouteUsage}
  MaterialExpressionNamedRerouteUsage_108 {class=MaterialExpressionNamedRerouteUsage}
  MaterialExpressionNamedRerouteUsage_109 {class=MaterialExpressionNamedRerouteUsage}
  MaterialExpressionNamedRerouteUsage_110 {class=MaterialExpressionNamedRerouteUsage}
  MaterialExpressionNamedRerouteUsage_111 {class=MaterialExpressionNamedRerouteUsage}
  MaterialExpressionNamedRerouteUsage_112 {class=MaterialExpressionNamedRerouteUsage}
  MaterialExpressionNamedRerouteUsage_113 {class=MaterialExpressionNamedRerouteUsage}
  MaterialExpressionNamedRerouteUsage_114 {class=MaterialExpressionNamedRerouteUsage}
  MaterialExpressionNamedRerouteUsage_115 {class=MaterialExpressionNamedRerouteUsage}
  MaterialExpressionNamedRerouteUsage_118 {class=MaterialExpressionNamedRerouteUsage}
  MaterialExpressionNamedRerouteUsage_119 {class=MaterialExpressionNamedRerouteUsage}
  MaterialExpressionNamedRerouteUsage_120 {class=MaterialExpressionNamedRerouteUsage}
  MaterialExpressionNamedRerouteUsage_121 {class=MaterialExpressionNamedRerouteUsage}
  MaterialExpressionNamedRerouteUsage_163 {class=MaterialExpressionNamedRerouteUsage}
  MaterialExpressionNamedRerouteUsage_164 {class=MaterialExpressionNamedRerouteUsage}
  MaterialExpressionNamedRerouteUsage_165 {class=MaterialExpressionNamedRerouteUsage}
  MaterialExpressionNamedRerouteUsage_166 {class=MaterialExpressionNamedRerouteUsage}
  MaterialExpressionNamedRerouteUsage_167 {class=MaterialExpressionNamedRerouteUsage}
  MaterialExpressionNamedRerouteUsage_168 {class=MaterialExpressionNamedRerouteUsage}
  MaterialExpressionNamedRerouteUsage_169 {class=MaterialExpressionNamedRerouteUsage}
  MaterialExpressionNamedRerouteUsage_17 {class=MaterialExpressionNamedRerouteUsage}
  MaterialExpressionNamedRerouteUsage_170 {class=MaterialExpressionNamedRerouteUsage}
  MaterialExpressionNamedRerouteUsage_171 {class=MaterialExpressionNamedRerouteUsage}
  MaterialExpressionNamedRerouteUsage_172 {class=MaterialExpressionNamedRerouteUsage}
  MaterialExpressionNamedRerouteUsage_173 {class=MaterialExpressionNamedRerouteUsage}
  MaterialExpressionNamedRerouteUsage_174 {class=MaterialExpressionNamedRerouteUsage}
  MaterialExpressionNamedRerouteUsage_175 {class=MaterialExpressionNamedRerouteUsage}
  MaterialExpressionNamedRerouteUsage_176 {class=MaterialExpressionNamedRerouteUsage}
  MaterialExpressionNamedRerouteUsage_177 {class=MaterialExpressionNamedRerouteUsage}
  MaterialExpressionNamedRerouteUsage_178 {class=MaterialExpressionNamedRerouteUsage}
  MaterialExpressionNamedRerouteUsage_179 {class=MaterialExpressionNamedRerouteUsage}
  MaterialExpressionNamedRerouteUsage_18 {class=MaterialExpressionNamedRerouteUsage}
  MaterialExpressionNamedRerouteUsage_180 {class=MaterialExpressionNamedRerouteUsage}
  MaterialExpressionNamedRerouteUsage_181 {class=MaterialExpressionNamedRerouteUsage}
  MaterialExpressionNamedRerouteUsage_182 {class=MaterialExpressionNamedRerouteUsage}
  MaterialExpressionNamedRerouteUsage_188 {class=MaterialExpressionNamedRerouteUsage}
  MaterialExpressionNamedRerouteUsage_189 {class=MaterialExpressionNamedRerouteUsage}
  MaterialExpressionNamedRerouteUsage_19 {class=MaterialExpressionNamedRerouteUsage}
  MaterialExpressionNamedRerouteUsage_190 {class=MaterialExpressionNamedRerouteUsage}
  MaterialExpressionNamedRerouteUsage_191 {class=MaterialExpressionNamedRerouteUsage}
  MaterialExpressionNamedRerouteUsage_192 {class=MaterialExpressionNamedRerouteUsage}
  MaterialExpressionNamedRerouteUsage_193 {class=MaterialExpressionNamedRerouteUsage}
  MaterialExpressionNamedRerouteUsage_194 {class=MaterialExpressionNamedRerouteUsage}
  MaterialExpressionNamedRerouteUsage_195 {class=MaterialExpressionNamedRerouteUsage}
  MaterialExpressionNamedRerouteUsage_196 {class=MaterialExpressionNamedRerouteUsage}
  MaterialExpressionNamedRerouteUsage_197 {class=MaterialExpressionNamedRerouteUsage}
  MaterialExpressionNamedRerouteUsage_198 {class=MaterialExpressionNamedRerouteUsage}
  MaterialExpressionNamedRerouteUsage_199 {class=MaterialExpressionNamedRerouteUsage}
  MaterialExpressionNamedRerouteUsage_200 {class=MaterialExpressionNamedRerouteUsage}
  MaterialExpressionNamedRerouteUsage_201 {class=MaterialExpressionNamedRerouteUsage}
  MaterialExpressionNamedRerouteUsage_202 {class=MaterialExpressionNamedRerouteUsage}
  MaterialExpressionNamedRerouteUsage_203 {class=MaterialExpressionNamedRerouteUsage}
  MaterialExpressionNamedRerouteUsage_204 {class=MaterialExpressionNamedRerouteUsage}
  MaterialExpressionNamedRerouteUsage_205 {class=MaterialExpressionNamedRerouteUsage}
  MaterialExpressionNamedRerouteUsage_206 {class=MaterialExpressionNamedRerouteUsage}
  MaterialExpressionNamedRerouteUsage_207 {class=MaterialExpressionNamedRerouteUsage}
  MaterialExpressionNamedRerouteUsage_208 {class=MaterialExpressionNamedRerouteUsage}
  MaterialExpressionNamedRerouteUsage_209 {class=MaterialExpressionNamedRerouteUsage}
  MaterialExpressionNamedRerouteUsage_210 {class=MaterialExpressionNamedRerouteUsage}
  MaterialExpressionNamedRerouteUsage_211 {class=MaterialExpressionNamedRerouteUsage}
  MaterialExpressionNamedRerouteUsage_212 {class=MaterialExpressionNamedRerouteUsage}
  MaterialExpressionNamedRerouteUsage_213 {class=MaterialExpressionNamedRerouteUsage}
  MaterialExpressionNamedRerouteUsage_214 {class=MaterialExpressionNamedRerouteUsage}
  MaterialExpressionNamedRerouteUsage_215 {class=MaterialExpressionNamedRerouteUsage}
  MaterialExpressionNamedRerouteUsage_23 {class=MaterialExpressionNamedRerouteUsage}
  MaterialExpressionNamedRerouteUsage_24 {class=MaterialExpressionNamedRerouteUsage}
  MaterialExpressionNamedRerouteUsage_27 {class=MaterialExpressionNamedRerouteUsage}
  MaterialExpressionNamedRerouteUsage_28 {class=MaterialExpressionNamedRerouteUsage}
  MaterialExpressionNamedRerouteUsage_289 {class=MaterialExpressionNamedRerouteUsage}
  MaterialExpressionNamedRerouteUsage_29 {class=MaterialExpressionNamedRerouteUsage}
  MaterialExpressionNamedRerouteUsage_33 {class=MaterialExpressionNamedRerouteUsage}
  MaterialExpressionNamedRerouteUsage_34 {class=MaterialExpressionNamedRerouteUsage}
  MaterialExpressionNamedRerouteUsage_35 {class=MaterialExpressionNamedRerouteUsage}
  MaterialExpressionNamedRerouteUsage_39 {class=MaterialExpressionNamedRerouteUsage}
  MaterialExpressionNamedRerouteUsage_52 {class=MaterialExpressionNamedRerouteUsage}
  MaterialExpressionNamedRerouteUsage_53 {class=MaterialExpressionNamedRerouteUsage}
  MaterialExpressionNamedRerouteUsage_54 {class=MaterialExpressionNamedRerouteUsage}
  MaterialExpressionNamedRerouteUsage_55 {class=MaterialExpressionNamedRerouteUsage}
  MaterialExpressionNamedRerouteUsage_56 {class=MaterialExpressionNamedRerouteUsage}
  MaterialExpressionNamedRerouteUsage_59 {class=MaterialExpressionNamedRerouteUsage}
  MaterialExpressionNamedRerouteUsage_60 {class=MaterialExpressionNamedRerouteUsage}
  MaterialExpressionNamedRerouteUsage_61 {class=MaterialExpressionNamedRerouteUsage}
  MaterialExpressionNamedRerouteUsage_62 {class=MaterialExpressionNamedRerouteUsage}
  MaterialExpressionNamedRerouteUsage_63 {class=MaterialExpressionNamedRerouteUsage}
  MaterialExpressionNamedRerouteUsage_64 {class=MaterialExpressionNamedRerouteUsage}
  MaterialExpressionNamedRerouteUsage_65 {class=MaterialExpressionNamedRerouteUsage}
  MaterialExpressionNamedRerouteUsage_92 {class=MaterialExpressionNamedRerouteUsage}
  MaterialExpressionNamedRerouteUsage_93 {class=MaterialExpressionNamedRerouteUsage}
  MaterialExpressionNamedRerouteUsage_94 {class=MaterialExpressionNamedRerouteUsage}
  MaterialExpressionNamedRerouteUsage_95 {class=MaterialExpressionNamedRerouteUsage}
  MaterialExpressionNamedRerouteUsage_96 {class=MaterialExpressionNamedRerouteUsage}
  MaterialExpressionNamedRerouteUsage_97 {class=MaterialExpressionNamedRerouteUsage}
  MaterialExpressionNamedRerouteUsage_98 {class=MaterialExpressionNamedRerouteUsage}
  MaterialExpressionNamedRerouteUsage_99 {class=MaterialExpressionNamedRerouteUsage}
  MaterialExpressionOneMinus_0 {class=MaterialExpressionOneMinus, inputs [None {from=MaterialExpressionScalarParameter_26}]}
  MaterialExpressionSaturate_1 {class=MaterialExpressionSaturate, inputs [None {from=MaterialExpressionAbs_1}]}
  MaterialExpressionScalarParameter_13 {
   class=MaterialExpressionScalarParameter
   properties {defaultValue=1, group=UV Transforms, parameterName=Scale, sliderMax=10000, sliderMin=0.001, sortPriority=10}
  }
  MaterialExpressionScalarParameter_14 {
   class=MaterialExpressionScalarParameter
   properties {defaultValue=1, group=UV Transforms, parameterName=StretchOnX, sliderMax=100, sliderMin=0.001, sortPriority=100}
  }
  MaterialExpressionScalarParameter_15 {
   class=MaterialExpressionScalarParameter
   properties {defaultValue=1, group=UV Transforms, parameterName=StretchOnY, sliderMax=1000, sliderMin=0.001, sortPriority=110}
  }
  MaterialExpressionScalarParameter_16 {class=MaterialExpressionScalarParameter, properties {group=UV Transforms, parameterName=OffsetOnY, sortPriority=140}}
  MaterialExpressionScalarParameter_17 {class=MaterialExpressionScalarParameter, properties {group=UV Transforms, parameterName=OffsetOnX, sortPriority=130}}
  MaterialExpressionScalarParameter_18 {
   class=MaterialExpressionScalarParameter
   properties {defaultValue=1, group=UV Transforms, parameterName=StretchOnZ, sliderMax=1000, sliderMin=0.001, sortPriority=120}
  }
  MaterialExpressionScalarParameter_19 {class=MaterialExpressionScalarParameter, properties {group=UV Transforms, parameterName=OffsetOnZ, sortPriority=150}}
  MaterialExpressionScalarParameter_20 {class=MaterialExpressionScalarParameter, properties {defaultValue=100, group=MillimeterSettings, parameterName=WidthMM, sortPriority=10}}
  MaterialExpressionScalarParameter_21 {class=MaterialExpressionScalarParameter, properties {defaultValue=100, group=MillimeterSettings, parameterName=HeightMM, sortPriority=11}}
  MaterialExpressionScalarParameter_22 {class=MaterialExpressionScalarParameter, properties {defaultValue=1, group=MillimeterSettings, parameterName=UnitScaleCM, sortPriority=12}}
  MaterialExpressionScalarParameter_23 {class=MaterialExpressionScalarParameter, properties {group=UV Transforms, parameterName=UV Index}}
  MaterialExpressionScalarParameter_24 {
   class=MaterialExpressionScalarParameter
   properties {
    defaultValue=0.5
    desc="Controls the contrast of the blend mask. Internally applied as pow(mask, 8 × BlendFactor). A default of 0.5 corresponds to pow4 contrast."
    group=UV Transforms
    parameterName=TransitionSoftness
    sliderMax=1
    sortPriority=3
   }
  }
  MaterialExpressionScalarParameter_25 {
   class=MaterialExpressionScalarParameter
   properties {defaultValue=1, group=Normal, parameterName=NormalStrength, sliderMax=2, sliderMin=-1, sortPriority=12}
  }
  MaterialExpressionScalarParameter_26 {class=MaterialExpressionScalarParameter, properties {group=Alpha, parameterName=ShadowCompensation, sliderMax=1, sortPriority=24}}
  MaterialExpressionScalarParameter_29 {
   class=MaterialExpressionScalarParameter
   properties {defaultValue=1, group=ClearCoat, parameterName=ClearCoatNormalStrength, sliderMax=2, sliderMin=-1}
  }
  MaterialExpressionScalarParameter_50 {class=MaterialExpressionScalarParameter, properties {defaultValue=1, group=Alpha, parameterName=Alpha}}
  MaterialExpressionScalarParameter_8 {class=MaterialExpressionScalarParameter, properties {group=UV Transforms, parameterName="Rotation (degree)", sliderMax=360, sortPriority=30}}
  MaterialExpressionShaderStageSwitch_0 {
   class=MaterialExpressionShaderStageSwitch
   inputs [
    PixelShader {from=MaterialExpressionCameraVectorWS_1, output=""}
    VertexShader {from=MaterialExpressionMaterialFunctionCall_1, output=Result}
   ]
   properties {}
  }
  MaterialExpressionShadowReplace_0 {
   class=MaterialExpressionShadowReplace
   inputs [
    Default {from=MaterialExpressionStaticSwitchParameter_94, output=""}
    Shadow {from=MaterialExpressionClamp_0, output=""}
   ]
   properties {}
  }
  MaterialExpressionSquareRoot_1 {class=MaterialExpressionSquareRoot, inputs [None {from=MaterialExpressionDivide_6, output=""}], properties {}}
  MaterialExpressionStaticBoolParameter_1 {
   class=MaterialExpressionStaticBoolParameter
   inputs []
   properties {defaultValue=true, group=UV Transforms, parameterName=Use Geo UVs, sortPriority=0}
  }
  MaterialExpressionStaticSwitchParameter_0 {
   class=MaterialExpressionStaticSwitchParameter
   inputs [
    True {from=MaterialExpressionStaticSwitch_0, output=""}
    False {from=MaterialExpressionVectorParameter_0, output=RGB}
   ]
   properties {group=Diffuse, parameterName=HasDiffuseColorTexture}
  }
  MaterialExpressionStaticSwitchParameter_16 {
   class=MaterialExpressionStaticSwitchParameter
   inputs [
    True {from=MaterialExpressionNamedRerouteUsage_107, output=""}
    False {from=MaterialExpressionConstant_14, output=""}
   ]
   properties {group=Sheen, parameterName=HasSheen, sortPriority=30}
  }
  MaterialExpressionStaticSwitchParameter_17 {
   class=MaterialExpressionStaticSwitchParameter
   inputs [
    True {from=MaterialExpressionConstant_19, output=""}
    False {from=MaterialExpressionConstant_12, output=""}
   ]
   properties {group=Sheen, parameterName=HasSheen, sortPriority=30}
  }
  MaterialExpressionStaticSwitchParameter_18 {
   class=MaterialExpressionStaticSwitchParameter
   inputs [
    True {from=MaterialExpressionStaticSwitchParameter_19, output=""}
    False {from=MaterialExpressionConstant_13, output=""}
   ]
   properties {group=Sheen, parameterName=HasSheen, sortPriority=30}
  }
  MaterialExpressionStaticSwitchParameter_19 {
   class=MaterialExpressionStaticSwitchParameter
   inputs [
    True {from=MaterialExpressionStaticSwitch_3, output=""}
    False {from=MaterialExpressionVectorParameter_6, output=RGB}
   ]
   properties {group=Sheen, parameterName=HasSheenColorTexture}
  }
  MaterialExpressionStaticSwitchParameter_2 {
   class=MaterialExpressionStaticSwitchParameter
   inputs [
    True {from=MaterialExpressionConstant_2, output=""}
    False {from=MaterialExpressionConstant_3, output=""}
   ]
   properties {defaultValue=true, group=UV Transforms, parameterName=UsingLocalSpace, sortPriority=2}
  }
  MaterialExpressionStaticSwitchParameter_20 {
   class=MaterialExpressionStaticSwitchParameter
   inputs [
    True {from=MaterialExpressionStaticSwitch_2, output=""}
    False {from=MaterialExpressionVectorParameter_7, output=RGB}
   ]
   properties {group=Sheen, parameterName=HasSheenRoughnessTexture}
  }
  MaterialExpressionStaticSwitchParameter_3 {
   class=MaterialExpressionStaticSwitchParameter
   inputs [
    True {from=MaterialExpressionMaterialFunctionCall_10, output=NormalOut}
    False {from=MaterialExpressionComponentMask_0, output=""}
   ]
   properties {group=Normal, parameterName=HasNormalTexture, sortPriority=30}
  }
  MaterialExpressionStaticSwitchParameter_31 {
   class=MaterialExpressionStaticSwitchParameter
   inputs [
    True {from=MaterialExpressionStaticSwitch_4, output=""}
    False {from=MaterialExpressionVectorParameter_17, output=RGB}
   ]
   properties {group=Fresnel, parameterName=HasF0Texture}
  }
  MaterialExpressionStaticSwitchParameter_32 {
   class=MaterialExpressionStaticSwitchParameter
   inputs [
    True {from=MaterialExpressionStaticSwitch_5, output=""}
    False {from=MaterialExpressionVectorParameter_26, output=RGB}
   ]
   properties {group=Fresnel, parameterName=HasF90Texture}
  }
  MaterialExpressionStaticSwitchParameter_33 {
   class=MaterialExpressionStaticSwitchParameter
   inputs [
    True {from=MaterialExpressionMaterialFunctionCall_33, output=NormalOut}
    False {from=MaterialExpressionComponentMask_2, output=""}
   ]
   properties {group=ClearCoat, parameterName=HasClearCoatNormalTexture, sortPriority=30}
  }
  MaterialExpressionStaticSwitchParameter_35 {
   class=MaterialExpressionStaticSwitchParameter
   inputs [
    True {from=MaterialExpressionStaticSwitch_7, output=""}
    False {from=MaterialExpressionVectorParameter_29, output=RGB}
   ]
   properties {group=Roughness, parameterName=HasRoughnessTexture, sortPriority=30}
  }
  MaterialExpressionStaticSwitchParameter_36 {
   class=MaterialExpressionStaticSwitchParameter
   inputs [
    True {from=MaterialExpressionStaticSwitch_8, output=""}
    False {from=MaterialExpressionMaterialFunctionCall_40, output=Result}
   ]
   properties {group=Anisotropy, parameterName=HasAnisotropicRotationTexture}
  }
  MaterialExpressionStaticSwitchParameter_66 {
   class=MaterialExpressionStaticSwitchParameter
   inputs [
    True {from=MaterialExpressionSubstrateVerticalLayering_2, output=""}
    False {from=MaterialExpressionSubstrateSlabBSDF_0, output=""}
   ]
   properties {group=ClearCoat, parameterName=HasClearCoat, sortPriority=30}
  }
  MaterialExpressionStaticSwitchParameter_94 {
   class=MaterialExpressionStaticSwitchParameter
   inputs [
    True {from=MaterialExpressionStaticSwitch_9, output=""}
    False {from=MaterialExpressionScalarParameter_50, output=""}
   ]
   properties {group=Alpha, parameterName=HasAlphaTexture}
  }
  MaterialExpressionStaticSwitch_0 {
   class=MaterialExpressionStaticSwitch
   inputs [
    True {from=MaterialExpressionTextureSample_0, output=RGB}
    False {from=MaterialExpressionMaterialFunctionCall_6, output=RGB}
    Value {from=MaterialExpressionNamedRerouteUsage_108, output=""}
   ]
   properties {}
  }
  MaterialExpressionStaticSwitch_1 {
   class=MaterialExpressionStaticSwitch
   inputs [
    True {from=MaterialExpressionMaterialFunctionCall_9, output=Result}
    False {from=MaterialExpressionMaterialFunctionCall_13, output=TSN}
    Value {from=MaterialExpressionNamedRerouteUsage_109, output=""}
   ]
   properties {}
  }
  MaterialExpressionStaticSwitch_2 {
   class=MaterialExpressionStaticSwitch
   inputs [
    True {from=MaterialExpressionTextureSample_4, output=RGB}
    False {from=MaterialExpressionMaterialFunctionCall_21, output=RGB}
    Value {from=MaterialExpressionNamedRerouteUsage_120, output=""}
   ]
   properties {}
  }
  MaterialExpressionStaticSwitch_3 {
   class=MaterialExpressionStaticSwitch
   inputs [
    True {from=MaterialExpressionTextureSample_5, output=RGB}
    False {from=MaterialExpressionMaterialFunctionCall_23, output=RGB}
    Value {from=MaterialExpressionNamedRerouteUsage_121, output=""}
   ]
   properties {}
  }
  MaterialExpressionStaticSwitch_4 {
   class=MaterialExpressionStaticSwitch
   inputs [
    True {from=MaterialExpressionTextureSample_2, output=RGB}
    False {from=MaterialExpressionMaterialFunctionCall_29, output=RGB}
    Value {from=MaterialExpressionNamedRerouteUsage_171, output=""}
   ]
   properties {}
  }
  MaterialExpressionStaticSwitch_5 {
   class=MaterialExpressionStaticSwitch
   inputs [
    True {from=MaterialExpressionTextureSample_3, output=RGB}
    False {from=MaterialExpressionMaterialFunctionCall_30, output=RGB}
    Value {from=MaterialExpressionNamedRerouteUsage_172, output=""}
   ]
   properties {}
  }
  MaterialExpressionStaticSwitch_6 {
   class=MaterialExpressionStaticSwitch
   inputs [
    True {from=MaterialExpressionTextureSample_1, output=RGB}
    False {from=MaterialExpressionMaterialFunctionCall_34, output=TSN}
    Value {from=MaterialExpressionNamedRerouteUsage_178, output=""}
   ]
   properties {}
  }
  MaterialExpressionStaticSwitch_7 {
   class=MaterialExpressionStaticSwitch
   inputs [
    True {from=MaterialExpressionTextureSample_8, output=RGB}
    False {from=MaterialExpressionMaterialFunctionCall_15, output=RGB}
    Value {from=MaterialExpressionNamedRerouteUsage_192, output=""}
   ]
   properties {}
  }
  MaterialExpressionStaticSwitch_8 {
   class=MaterialExpressionStaticSwitch
   inputs [
    True {from=MaterialExpressionMaterialFunctionCall_37, output=Result}
    False {from=MaterialExpressionMaterialFunctionCall_38, output=Result}
    Value {from=MaterialExpressionNamedRerouteUsage_199, output=""}
   ]
   properties {}
  }
  MaterialExpressionStaticSwitch_9 {
   class=MaterialExpressionStaticSwitch
   inputs [
    True {from=MaterialExpressionTextureSample_22, output=G}
    False {from=MaterialExpressionMaterialFunctionCall_41, output=G}
    Value {from=MaterialExpressionNamedRerouteUsage_215, output=""}
   ]
   properties {}
  }
  MaterialExpressionSubstrateSlabBSDF_0 {
   class=MaterialExpressionSubstrateSlabBSDF
   inputs [
    Diffuse Albedo {from=MaterialExpressionNamedRerouteUsage_200, output=""}
    F0 {from=MaterialExpressionNamedRerouteUsage_201, output=""}
    F90 {from=MaterialExpressionNamedRerouteUsage_209, output=""}
    Roughness {from=MaterialExpressionNamedRerouteUsage_202, output=""}
    Anisotropy {from=MaterialExpressionNamedRerouteUsage_203, output=""}
    Normal {from=MaterialExpressionNamedRerouteUsage_208, output=""}
    Tangent {from=MaterialExpressionNamedRerouteUsage_204, output=""}
    SSS MFP {from="", output=""}
    SSS MFP Scale {from="", output=""}
    SSS Phase Anisotropy {from="", output=""}
    Emissive Color {from="", output=""}
    Second Roughness {from="", output=""}
    Second Roughness Weight {from="", output=""}
    Fuzz Roughness {from=MaterialExpressionNamedRerouteUsage_205, output=""}
    Fuzz Amount {from=MaterialExpressionNamedRerouteUsage_206, output=""}
    Fuzz Color {from=MaterialExpressionNamedRerouteUsage_207, output=""}
    "Glint Density (Disabled)" {from="", output=""}
    "Glint UVs (Disabled)" {from="", output=""}
   ]
   properties {}
  }
  MaterialExpressionSubstrateVerticalLayering_2 {
   class=MaterialExpressionSubstrateVerticalLayering
   inputs [
    Top {from=MaterialExpressionNamedRerouteUsage_210, output=""}
    Bottom {from=MaterialExpressionSubstrateSlabBSDF_0, output=""}
    Top Thickness {from="", output=""}
   ]
   properties {}
  }
  MaterialExpressionTextureObjectParameter_0 {
   class=MaterialExpressionTextureObjectParameter
   inputs [
    UVs {from="", output=""}
    Apply View MipBias {from="", output=""}
   ]
   properties {
    group=Diffuse
    parameterName=DiffuseColorTexture
    samplerType=SAMPLERTYPE_LinearColor
    sortPriority=1
    texture=/Engine/Functions/Engine_MaterialFunctions02/PivotPainter2/Black_1x1_EXR_Texture
   }
  }
  MaterialExpressionTextureObjectParameter_18 {
   class=MaterialExpressionTextureObjectParameter
   inputs [
    UVs {from="", output=""}
    Apply View MipBias {from="", output=""}
   ]
   properties {
    group=Fresnel
    parameterName=F90Texture
    samplerType=SAMPLERTYPE_LinearColor
    sortPriority=2
    texture=&Black_1x1_EXR_Texture
   }
  }
  MaterialExpressionTextureObjectParameter_2 {
   class=MaterialExpressionTextureObjectParameter
   inputs [
    UVs {from="", output=""}
    Apply View MipBias {from="", output=""}
   ]
   properties {group=Normal, parameterName=NormalTexture, samplerType=SAMPLERTYPE_Normal, sortPriority=1, texture=/Engine/EngineMaterials/DefaultNormal}
  }
  MaterialExpressionTextureObjectParameter_26 {
   class=MaterialExpressionTextureObjectParameter
   inputs [
    UVs {from="", output=""}
    Apply View MipBias {from="", output=""}
   ]
   properties {
    group=Fresnel
    parameterName=F0Texture
    samplerType=SAMPLERTYPE_LinearColor
    sortPriority=1
    texture=&Black_1x1_EXR_Texture
   }
  }
  MaterialExpressionTextureObjectParameter_27 {
   class=MaterialExpressionTextureObjectParameter
   inputs [
    UVs {from="", output=""}
    Apply View MipBias {from="", output=""}
   ]
   properties {
    group=Alpha
    parameterName=AlphaTexture
    samplerType=SAMPLERTYPE_LinearColor
    sortPriority=1
    texture=/InterchangeAssets/gltf/Textures/T_White_Linear
   }
  }
  MaterialExpressionTextureObjectParameter_28 {
   class=MaterialExpressionTextureObjectParameter
   inputs [
    UVs {from="", output=""}
    Apply View MipBias {from="", output=""}
   ]
   properties {
    group=ClearCoat
    parameterName=ClearCoatNormalTexture
    samplerType=SAMPLERTYPE_Normal
    sortPriority=1
    texture=&DefaultNormal
   }
  }
  MaterialExpressionTextureObjectParameter_30 {
   class=MaterialExpressionTextureObjectParameter
   inputs [
    UVs {from="", output=""}
    Apply View MipBias {from="", output=""}
   ]
   properties {
    group=Roughness
    parameterName=RoughnessTexture
    samplerType=SAMPLERTYPE_LinearColor
    sortPriority=1
    texture=&Black_1x1_EXR_Texture
   }
  }
  MaterialExpressionTextureObjectParameter_31 {
   class=MaterialExpressionTextureObjectParameter
   inputs [
    UVs {from="", output=""}
    Apply View MipBias {from="", output=""}
   ]
   properties {
    group=Anisotropy
    parameterName=AnisotropicRotationTexture
    samplerType=SAMPLERTYPE_LinearColor
    sortPriority=1
    texture=&Black_1x1_EXR_Texture
   }
  }
  MaterialExpressionTextureObjectParameter_7 {
   class=MaterialExpressionTextureObjectParameter
   inputs [
    UVs {from="", output=""}
    Apply View MipBias {from="", output=""}
   ]
   properties {
    group=Sheen
    parameterName=SheenColorTexture
    samplerType=SAMPLERTYPE_LinearColor
    sortPriority=2
    texture=&Black_1x1_EXR_Texture
   }
  }
  MaterialExpressionTextureObjectParameter_8 {
   class=MaterialExpressionTextureObjectParameter
   inputs [
    UVs {from="", output=""}
    Apply View MipBias {from="", output=""}
   ]
   properties {
    group=Sheen
    parameterName=SheenRoughnessTexture
    samplerType=SAMPLERTYPE_LinearColor
    sortPriority=1
    texture=&Black_1x1_EXR_Texture
   }
  }
  MaterialExpressionTextureSample_0 {
   class=MaterialExpressionTextureSample
   inputs [
    UVs {from=MaterialExpressionNamedRerouteUsage_102}
    Tex {from=MaterialExpressionTextureObjectParameter_0}
    Apply View MipBias {from=""}
   ]
   properties {samplerType=SAMPLERTYPE_LinearColor}
  }
  MaterialExpressionTextureSample_1 {
   class=MaterialExpressionTextureSample
   inputs [
    UVs {from=MaterialExpressionNamedRerouteUsage_177}
    Tex {from=MaterialExpressionTextureObjectParameter_28}
    Apply View MipBias {from=""}
   ]
  }
  MaterialExpressionTextureSample_2 {
   class=MaterialExpressionTextureSample
   inputs [
    UVs {from=MaterialExpressionNamedRerouteUsage_168}
    Tex {from=MaterialExpressionTextureObjectParameter_26}
    Apply View MipBias {from=""}
   ]
  }
  MaterialExpressionTextureSample_22 {
   class=MaterialExpressionTextureSample
   inputs [
    UVs {from=MaterialExpressionNamedRerouteUsage_211}
    Tex {from=MaterialExpressionTextureObjectParameter_27}
    Apply View MipBias {from=""}
   ]
   properties {samplerType=SAMPLERTYPE_LinearColor}
  }
  MaterialExpressionTextureSample_3 {
   class=MaterialExpressionTextureSample
   inputs [
    UVs {from=MaterialExpressionNamedRerouteUsage_169}
    Tex {from=MaterialExpressionTextureObjectParameter_18}
    Apply View MipBias {from=""}
   ]
  }
  MaterialExpressionTextureSample_4 {
   class=MaterialExpressionTextureSample
   inputs [
    UVs {from=MaterialExpressionNamedRerouteUsage_115}
    Tex {from=MaterialExpressionTextureObjectParameter_8}
    Apply View MipBias {from=""}
   ]
  }
  MaterialExpressionTextureSample_5 {
   class=MaterialExpressionTextureSample
   inputs [
    UVs {from=MaterialExpressionNamedRerouteUsage_118}
    Tex {from=MaterialExpressionTextureObjectParameter_7}
    Apply View MipBias {from=""}
   ]
  }
  MaterialExpressionTextureSample_7 {
   class=MaterialExpressionTextureSample
   inputs [
    UVs {from=MaterialExpressionNamedRerouteUsage_198}
    Tex {from=MaterialExpressionTextureObjectParameter_31}
    Apply View MipBias {from=""}
   ]
  }
  MaterialExpressionTextureSample_8 {
   class=MaterialExpressionTextureSample
   inputs [
    UVs {from=MaterialExpressionNamedRerouteUsage_191}
    Tex {from=MaterialExpressionTextureObjectParameter_30}
    Apply View MipBias {from=""}
   ]
  }
  MaterialExpressionVectorParameter_0 {class=MaterialExpressionVectorParameter, properties {group=Diffuse, parameterName=DiffuseColor, defaultValue=<rgba 0 0 0 1>}}
  MaterialExpressionVectorParameter_1 {class=MaterialExpressionVectorParameter, properties {group=Normal, parameterName=Normal, defaultValue=<rgba 0 0 1 1>}}
  MaterialExpressionVectorParameter_17 {class=MaterialExpressionVectorParameter, properties {group=Fresnel, parameterName=F0, defaultValue=<rgba 0.04 0.04 0.04 1>}}
  MaterialExpressionVectorParameter_26 {class=MaterialExpressionVectorParameter, properties {group=Fresnel, parameterName=F90, defaultValue=<rgba 1 1 1 0>}}
  MaterialExpressionVectorParameter_27 {class=MaterialExpressionVectorParameter, properties {group=ClearCoat, parameterName=ClearCoatNormal, defaultValue=<rgba 0 0 1 1>}}
  MaterialExpressionVectorParameter_29 {
   class=MaterialExpressionVectorParameter
   properties {group=Roughness, parameterName=Roughness, sortPriority=30, defaultValue=<rgba 0.2122 0.2122 0 1>}
  }
  MaterialExpressionVectorParameter_30 {class=MaterialExpressionVectorParameter, properties {group=Anisotropy, parameterName=AnisotropicRotation, defaultValue=<rgba 0.5 0.5 0 0>}}
  MaterialExpressionVectorParameter_6 {class=MaterialExpressionVectorParameter, properties {group=Sheen, parameterName=SheenColor, defaultValue=<rgba 0 0 0 1>}}
  MaterialExpressionVectorParameter_7 {class=MaterialExpressionVectorParameter, properties {group=Sheen, parameterName=SheenRoughness, defaultValue=<rgba 0.1 0.1 0 1>}}
  MaterialExpressionVertexNormalWS_2 {class=MaterialExpressionVertexNormalWS, inputs [], properties {}}
 }
 properties {
  bUsedWithNanite=true
  bUsedWithStaticLighting=true
  blendMode=BLEND_Masked
  ditherOpacityMask=true
  shadingModel=MSM_SubsurfaceProfile
  translucencyLightingMode=TLM_SurfacePerPixelLighting
 }
}
