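# Base material for AxF SVBRDF surfaces with alpha and parallax occlusion mapping, using texture and parameter inputs.
# Millimeter-based scaling and UV transforms feed two POM coordinate functions, geometry UV or triplanar, which drive texture sampling through named reroutes.
# Samples and switches build diffuse, F0/F90, anisotropic roughness, fuzz, normal and clear coat slab terms.
# View-angle terms feed normal strength, and a shadow-pass switch supplies separately compensated opacity.
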Material /InterchangeAxfAssets/BaseMaterials/SVBRDF_Alpha_POM
graph {
 expressions {
  MaterialExpressionAbs_1 {class=MaterialExpressionAbs, inputs [None {from=MaterialExpressionDotProduct_1, output=""}], properties {}}
  MaterialExpressionAbs_2 {class=MaterialExpressionAbs, inputs [None {from=MaterialExpressionDotProduct_2, output=""}], properties {}}
  MaterialExpressionAdd_1 {
   class=MaterialExpressionAdd
   inputs [
    A {from=MaterialExpressionMaterialFunctionCall_35, output=R}
    B {from=MaterialExpressionMaterialFunctionCall_35, output=R}
   ]
  }
  MaterialExpressionCameraVectorWS_1 {class=MaterialExpressionCameraVectorWS, inputs [], properties {}}
  MaterialExpressionClamp_0 {
   class=MaterialExpressionClamp
   inputs [
    None {from=MaterialExpressionMultiply_4, output=""}
    Min {from="", output=""}
    Max {from="", output=""}
   ]
   properties {}
  }
  MaterialExpressionComponentMask_4 {
   class=MaterialExpressionComponentMask
   inputs [None {from=MaterialExpressionVectorParameter_10, output=RGB}]
   properties {b=true, g=true, r=true}
  }
  MaterialExpressionComponentMask_5 {
   class=MaterialExpressionComponentMask
   inputs [None {from=MaterialExpressionVectorParameter_26, output=RGB}]
   properties {b=true, g=true, r=true}
  }
  MaterialExpressionConstant_1 {class=MaterialExpressionConstant, properties {desc="0 = world"}}
  MaterialExpressionConstant_12 {class=MaterialExpressionConstant}
  MaterialExpressionConstant_13 {class=MaterialExpressionConstant}
  MaterialExpressionConstant_14 {class=MaterialExpressionConstant}
  MaterialExpressionConstant_2 {class=MaterialExpressionConstant, properties {desc="1 = local", r=1}}
  MaterialExpressionConstant_20 {class=MaterialExpressionConstant, properties {r=1}}
  MaterialExpressionConvert_0 {
   class=MaterialExpressionConvert
   inputs [None {from=MaterialExpressionMaterialFunctionCall_9, output=DXY}]
   properties {
    convertInputs [{
      type=Vector4
      defaultValue=<rgba 0 0 0 1>
      expressionInput {
       expression=&self:MaterialExpressionMaterialFunctionCall_9
       inputName=None
       mask=0
       maskA=0
       maskB=0
       maskG=0
       maskR=0
       outputIndex=4
      }
     }]
    convertOutputs [{type=Vector2}, {type=Vector2}]
   }
  }
  MaterialExpressionConvert_1 {
   class=MaterialExpressionConvert
   inputs [
    None {from=MaterialExpressionNamedRerouteUsage_21}
    None {from=MaterialExpressionNamedRerouteUsage_22}
    None {from=MaterialExpressionNamedRerouteUsage_26}
    None {from=MaterialExpressionNamedRerouteUsage_11}
   ]
   properties {
    convertInputs [
     {
      type=Scalar
      expressionInput {
       expression=&self:MaterialExpressionNamedRerouteUsage_21
       inputName=None
       mask=0
       maskA=0
       maskB=0
       maskG=0
       maskR=0
       outputIndex=0
      }
     }
     {
      type=Scalar
      expressionInput {
       expression=&self:MaterialExpressionNamedRerouteUsage_22
       inputName=None
       mask=0
       maskA=0
       maskB=0
       maskG=0
       maskR=0
       outputIndex=0
      }
     }
     {
      type=Scalar
      expressionInput {
       expression=&self:MaterialExpressionNamedRerouteUsage_26
       inputName=None
       mask=0
       maskA=0
       maskB=0
       maskG=0
       maskR=0
       outputIndex=0
      }
     }
     {
      type=Scalar
      expressionInput {
       expression=&self:MaterialExpressionNamedRerouteUsage_11
       inputName=None
       mask=0
       maskA=0
       maskB=0
       maskG=0
       maskR=0
       outputIndex=0
      }
     }
    ]
    convertOutputs [{type=Vector4, defaultValue=<rgba 0 0 0 1>}]
   }
  }
  MaterialExpressionConvert_2 {
   class=MaterialExpressionConvert
   inputs [
    None {from=MaterialExpressionNamedRerouteUsage_52}
    None {from=MaterialExpressionNamedRerouteUsage_53}
    None {from=MaterialExpressionNamedRerouteUsage_56}
   ]
   properties {
    convertInputs [
     {
      type=Scalar
      expressionInput {
       expression=&self:MaterialExpressionNamedRerouteUsage_52
       inputName=None
       mask=0
       maskA=0
       maskB=0
       maskG=0
       maskR=0
       outputIndex=0
      }
     }
     {
      type=Scalar
      expressionInput {
       expression=&self:MaterialExpressionNamedRerouteUsage_53
       inputName=None
       mask=0
       maskA=0
       maskB=0
       maskG=0
       maskR=0
       outputIndex=0
      }
     }
     {
      type=Scalar
      expressionInput {
       expression=&self:MaterialExpressionNamedRerouteUsage_56
       inputName=None
       mask=0
       maskA=0
       maskB=0
       maskG=0
       maskR=0
       outputIndex=0
      }
     }
    ]
    convertOutputs [{type=Vector3}]
   }
  }
  MaterialExpressionDivide_3 {
   class=MaterialExpressionDivide
   inputs [
    A {from=""}
    B {from=MaterialExpressionScalarParameter_33}
   ]
   properties {constA=10, constB=10}
  }
  MaterialExpressionDivide_4 {
   class=MaterialExpressionDivide
   inputs [
    A {from=""}
    B {from=MaterialExpressionScalarParameter_34}
   ]
   properties {constA=10, constB=10}
  }
  MaterialExpressionDivide_6 {
   class=MaterialExpressionDivide
   inputs [
    A {from=MaterialExpressionAdd_1}
    B {from=""}
   ]
  }
  MaterialExpressionDotProduct_1 {
   class=MaterialExpressionDotProduct
   inputs [
    A {from=MaterialExpressionShaderStageSwitch_0, output=""}
    B {from=MaterialExpressionVertexNormalWS_2, output=""}
   ]
   properties {}
  }
  MaterialExpressionDotProduct_2 {
   class=MaterialExpressionDotProduct
   inputs [
    A {from=MaterialExpressionShaderStageSwitch_0, output=""}
    B {from=MaterialExpressionPixelNormalWS_0, output=""}
   ]
   properties {}
  }
  MaterialExpressionMaterialFunctionCall_10 {
   class=MaterialExpressionMaterialFunctionCall
   inputs [
    Coordinates {from=MaterialExpressionNamedRerouteUsage_109}
    TexObject {from=MaterialExpressionTextureObjectParameter_4}
    DDX {from=MaterialExpressionNamedRerouteUsage_113}
    DDY {from=MaterialExpressionNamedRerouteUsage_114}
    Mask {from=""}
    DefaultUV {from=""}
   ]
   properties {materialFunction=/BaseMaterial/Materials/Functions/MF_TexPointSample_RGB}
  }
  MaterialExpressionMaterialFunctionCall_12 {
   class=MaterialExpressionMaterialFunctionCall
   inputs [
    Coordinates {from=MaterialExpressionNamedRerouteUsage_228}
    TexObject {from=MaterialExpressionTextureObjectParameter_14}
    DDX {from=MaterialExpressionNamedRerouteUsage_209}
    DDY {from=MaterialExpressionNamedRerouteUsage_210}
    Mask {from=""}
    DefaultUV {from=""}
   ]
   properties {materialFunction=&MF_TexPointSample_RGB}
  }
  MaterialExpressionMaterialFunctionCall_13 {
   class=MaterialExpressionMaterialFunctionCall
   inputs [
    Coordinates {from=MaterialExpressionNamedRerouteUsage_229}
    TexObject {from=MaterialExpressionTextureObjectParameter_18}
    DDX {from=MaterialExpressionNamedRerouteUsage_215}
    DDY {from=MaterialExpressionNamedRerouteUsage_216}
    Mask {from=""}
    DefaultUV {from=""}
   ]
   properties {materialFunction=/BaseMaterial/Materials/Functions/MF_TexPointSample_NRM}
  }
  MaterialExpressionMaterialFunctionCall_15 {
   class=MaterialExpressionMaterialFunctionCall
   inputs [
    Coordinates {from=MaterialExpressionNamedRerouteUsage_59}
    TexObject {from=MaterialExpressionTextureObjectParameter_12}
    DDX {from=MaterialExpressionNamedRerouteUsage_211}
    DDY {from=MaterialExpressionNamedRerouteUsage_212}
    Mask {from=""}
    DefaultUV {from=""}
   ]
   properties {materialFunction=&MF_TexPointSample_RGB}
  }
  MaterialExpressionMaterialFunctionCall_18 {
   class=MaterialExpressionMaterialFunctionCall
   inputs [
    Coordinates {from=MaterialExpressionNamedRerouteUsage_64}
    TexObject {from=MaterialExpressionTextureObjectParameter_13}
    DDX {from=MaterialExpressionNamedRerouteUsage_213}
    DDY {from=MaterialExpressionNamedRerouteUsage_214}
    Mask {from=""}
    DefaultUV {from=""}
   ]
   properties {materialFunction=&MF_TexPointSample_RGB}
  }
  MaterialExpressionMaterialFunctionCall_21 {
   class=MaterialExpressionMaterialFunctionCall
   inputs [
    Coordinates {from=MaterialExpressionNamedRerouteUsage_221}
    TexObject {from=MaterialExpressionTextureObjectParameter_17}
    DDX {from=MaterialExpressionNamedRerouteUsage_203}
    DDY {from=MaterialExpressionNamedRerouteUsage_204}
    Mask {from=""}
    DefaultUV {from=""}
   ]
   properties {materialFunction=&MF_TexPointSample_RGB}
  }
  MaterialExpressionMaterialFunctionCall_23 {
   class=MaterialExpressionMaterialFunctionCall
   inputs [
    Coordinates {from=MaterialExpressionNamedRerouteUsage_224}
    TexObject {from=MaterialExpressionTextureObjectParameter_16}
    DDX {from=MaterialExpressionNamedRerouteUsage_205}
    DDY {from=MaterialExpressionNamedRerouteUsage_206}
    Mask {from=""}
    DefaultUV {from=""}
   ]
   properties {materialFunction=&MF_TexPointSample_RGB}
  }
  MaterialExpressionMaterialFunctionCall_34 {
   class=MaterialExpressionMaterialFunctionCall
   inputs [ClearCoatNormalV {from=MaterialExpressionNamedRerouteUsage_198}]
   properties {materialFunction=/InterchangeAxfAssets/MaterialFunctions/MF_ClearCoatTerm}
  }
  MaterialExpressionMaterialFunctionCall_35 {
   class=MaterialExpressionMaterialFunctionCall
   inputs [Color {from=MaterialExpressionStaticSwitchParameter_21}]
   properties {materialFunction=&SplitComponents}
  }
  MaterialExpressionMaterialFunctionCall_36 {
   class=MaterialExpressionMaterialFunctionCall
   inputs [
    X {from=MaterialExpressionTextureSample_8, output=R}
    Y {from=MaterialExpressionTextureSample_8, output=R}
   ]
   properties {materialFunction=&MakeFloat2}
  }
  MaterialExpressionMaterialFunctionCall_37 {
   class=MaterialExpressionMaterialFunctionCall
   inputs [
    X {from=MaterialExpressionMaterialFunctionCall_18, output=R}
    Y {from=MaterialExpressionMaterialFunctionCall_18, output=R}
   ]
   properties {materialFunction=&MakeFloat2}
  }
  MaterialExpressionMaterialFunctionCall_38 {
   class=MaterialExpressionMaterialFunctionCall
   inputs [
    AnisotropicRotation {from=MaterialExpressionStaticSwitchParameter_22}
    AlphaX {from=MaterialExpressionNamedRerouteUsage_195}
    AlphaY {from=MaterialExpressionNamedRerouteUsage_196}
   ]
   properties {materialFunction=/InterchangeAxfAssets/MaterialFunctions/MF_AnisotropyComponent}
  }
  MaterialExpressionMaterialFunctionCall_39 {
   class=MaterialExpressionMaterialFunctionCall
   inputs [SheenRoughness {from=MaterialExpressionStaticSwitchParameter_32}]
   properties {materialFunction=/InterchangeAxfAssets/MaterialFunctions/MF_SheenRoughness}
  }
  MaterialExpressionMaterialFunctionCall_40 {
   class=MaterialExpressionMaterialFunctionCall
   inputs [
    X {from=MaterialExpressionVectorParameter_17, output=R}
    Y {from=MaterialExpressionVectorParameter_17, output=R}
   ]
   properties {materialFunction=&MakeFloat2}
  }
  MaterialExpressionMaterialFunctionCall_41 {
   class=MaterialExpressionMaterialFunctionCall
   inputs [
    Coordinates {from=MaterialExpressionNamedRerouteUsage_217}
    TexObject {from=MaterialExpressionTextureObjectParameter_2}
    DDX {from=MaterialExpressionNamedRerouteUsage_199}
    DDY {from=MaterialExpressionNamedRerouteUsage_200}
    Mask {from=""}
    DefaultUV {from=""}
   ]
   properties {materialFunction=&MF_TexPointSample_RGB}
  }
  MaterialExpressionMaterialFunctionCall_42 {
   class=MaterialExpressionMaterialFunctionCall
   inputs [
    Coordinates {from=MaterialExpressionNamedRerouteUsage_219}
    TexObject {from=MaterialExpressionTextureObjectParameter_3}
    DDX {from=MaterialExpressionNamedRerouteUsage_201}
    DDY {from=MaterialExpressionNamedRerouteUsage_202}
    Mask {from=""}
    DefaultUV {from=""}
   ]
   properties {materialFunction=&MF_TexPointSample_NRM}
  }
  MaterialExpressionMaterialFunctionCall_43 {
   class=MaterialExpressionMaterialFunctionCall
   inputs [
    Coordinates {from=MaterialExpressionNamedRerouteUsage_225}
    TexObject {from=MaterialExpressionTextureObjectParameter_15}
    DDX {from=MaterialExpressionNamedRerouteUsage_207}
    DDY {from=MaterialExpressionNamedRerouteUsage_208}
    Mask {from=""}
    DefaultUV {from=""}
   ]
   properties {materialFunction=&MF_TexPointSample_RGB}
  }
  MaterialExpressionMaterialFunctionCall_44 {
   class=MaterialExpressionMaterialFunctionCall
   inputs [
    NormalIn {from=MaterialExpressionStaticSwitch_6}
    NormalStrength {from=MaterialExpressionScalarParameter_38}
    ReFresnelVX {from=MaterialExpressionNamedRerouteUsage_298}
   ]
   properties {materialFunction=/InterchangeAxfAssets/MaterialFunctions/MF_NormalStrength}
  }
  MaterialExpressionMaterialFunctionCall_45 {
   class=MaterialExpressionMaterialFunctionCall
   inputs [
    NormalIn {from=MaterialExpressionStaticSwitch_0}
    NormalStrength {from=MaterialExpressionScalarParameter_37}
    ReFresnelVX {from=MaterialExpressionNamedRerouteUsage_299}
   ]
   properties {materialFunction=&MF_NormalStrength}
  }
  MaterialExpressionMaterialFunctionCall_46 {
   class=MaterialExpressionMaterialFunctionCall
   inputs [
    TSN {from=MaterialExpressionMaterialFunctionCall_42, output=RGB}
    Tangent {from=MaterialExpressionNamedRerouteUsage_300}
    Bitangent {from=MaterialExpressionNamedRerouteUsage_301}
    "Rotation (Turns)" {from=MaterialExpressionNamedRerouteUsage_302}
   ]
   properties {materialFunction=/BaseMaterial/Materials/Functions/MF_TangentNormal_OrientWorld}
  }
  MaterialExpressionMaterialFunctionCall_47 {
   class=MaterialExpressionMaterialFunctionCall
   inputs [
    TSN {from=MaterialExpressionMaterialFunctionCall_13, output=RGB}
    Tangent {from=MaterialExpressionNamedRerouteUsage_304}
    Bitangent {from=MaterialExpressionNamedRerouteUsage_303}
    "Rotation (Turns)" {from=MaterialExpressionNamedRerouteUsage_305}
   ]
   properties {materialFunction=&MF_TangentNormal_OrientWorld}
  }
  MaterialExpressionMaterialFunctionCall_6 {
   class=MaterialExpressionMaterialFunctionCall
   inputs []
   properties {materialFunction=&CameraDirectionVector}
  }
  MaterialExpressionMaterialFunctionCall_8 {
   class=MaterialExpressionMaterialFunctionCall
   inputs [
    AxF_Scale_X {from=MaterialExpressionNamedRerouteUsage_40}
    AxF_Scale_Y {from=MaterialExpressionNamedRerouteUsage_50}
    AxF_UnitScale {from=MaterialExpressionNamedRerouteUsage_51}
    Tile U {from=MaterialExpressionNamedRerouteUsage_30}
    Tile V {from=MaterialExpressionNamedRerouteUsage_31}
    Offset U {from=MaterialExpressionNamedRerouteUsage_32}
    Offset V {from=MaterialExpressionNamedRerouteUsage_36}
    Rotation {from=MaterialExpressionNamedRerouteUsage_12}
    UV Index {from=MaterialExpressionScalarParameter_36}
    Height Texture {from=MaterialExpressionNamedRerouteUsage_69}
    POM Intensity {from=MaterialExpressionNamedRerouteUsage_74}
    POM Max Steps {from=MaterialExpressionNamedRerouteUsage_103}
    POM RefPlane {from=MaterialExpressionNamedRerouteUsage_75}
   ]
   properties {materialFunction=/InterchangeAxfAssets/MaterialFunctions/MF_AxFMakeCoordinatesUV_POM}
  }
  MaterialExpressionMaterialFunctionCall_9 {
   class=MaterialExpressionMaterialFunctionCall
   inputs [
    AxF_Scale_X {from=MaterialExpressionNamedRerouteUsage_37}
    AxF_Scale_Y {from=MaterialExpressionNamedRerouteUsage_38}
    AxF_UnitScale {from=MaterialExpressionNamedRerouteUsage_39}
    Tile XYZ {from=MaterialExpressionNamedRerouteUsage_60}
    Offset XYZ {from=MaterialExpressionNamedRerouteUsage_57}
    Transition Softness {from=MaterialExpressionNamedRerouteUsage_10}
    Coordinate Space {from=MaterialExpressionStaticSwitchParameter_2}
    Rotation {from=MaterialExpressionNamedRerouteUsage_58}
    Height Texture {from=MaterialExpressionNamedRerouteUsage_130}
    POM Intensity {from=MaterialExpressionNamedRerouteUsage_131}
    POM MaxSteps {from=MaterialExpressionNamedRerouteUsage_101}
    POM RefPlane {from=MaterialExpressionNamedRerouteUsage_133}
   ]
   properties {materialFunction=/InterchangeAxfAssets/MaterialFunctions/MF_AxFMakeCoordinatesTP_POM}
  }
  MaterialExpressionMultiply_3 {
   class=MaterialExpressionMultiply
   inputs [
    A {from=MaterialExpressionScalarParameter_35}
    B {from=""}
   ]
   properties {constB=10}
  }
  MaterialExpressionMultiply_4 {
   class=MaterialExpressionMultiply
   inputs [
    A {from=MaterialExpressionStaticSwitchParameter_1}
    B {from=MaterialExpressionOneMinus_0}
   ]
  }
  MaterialExpressionNamedRerouteDeclaration_100 {
   class=MaterialExpressionNamedRerouteDeclaration
   inputs [None {from=MaterialExpressionMaterialFunctionCall_39, output=IsotropicRoughness}]
   properties {name=SheenRoughness, nodeColor=<rgba 1 0 0.1641 1>}
  }
  MaterialExpressionNamedRerouteDeclaration_101 {
   class=MaterialExpressionNamedRerouteDeclaration
   inputs [None {from=MaterialExpressionStaticSwitchParameter_33}]
   properties {name=CCNormal, nodeColor=<rgba 0.5469 0 1 1>}
  }
  MaterialExpressionNamedRerouteDeclaration_11 {
   class=MaterialExpressionNamedRerouteDeclaration
   inputs [None {from=MaterialExpressionScalarParameter_15}]
   properties {name=POM_MaxSteps, nodeColor=<rgba 0 1 0.8125 1>}
  }
  MaterialExpressionNamedRerouteDeclaration_12 {
   class=MaterialExpressionNamedRerouteDeclaration
   inputs [None {from=MaterialExpressionScalarParameter_14}]
   properties {name=POM_RefPlane, nodeColor=<rgba 0 1 0.8125 1>}
  }
  MaterialExpressionNamedRerouteDeclaration_14 {
   class=MaterialExpressionNamedRerouteDeclaration
   inputs [None {from=MaterialExpressionMaterialFunctionCall_9, output=Turns}]
   properties {name=TP_Rotation, nodeColor=<rgba 0.9453 1 0 1>}
  }
  MaterialExpressionNamedRerouteDeclaration_15 {
   class=MaterialExpressionNamedRerouteDeclaration
   inputs [None {from=MaterialExpressionMaterialFunctionCall_9, output=T}]
   properties {name=TP_Tangent, nodeColor=<rgba 1 0 0.8672 1>}
  }
  MaterialExpressionNamedRerouteDeclaration_17 {
   class=MaterialExpressionNamedRerouteDeclaration
   inputs [None {from=MaterialExpressionMaterialFunctionCall_9, output=B}]
   properties {name=TP_Bitangent, nodeColor=<rgba 0 1 0.8125 1>}
  }
  MaterialExpressionNamedRerouteDeclaration_18 {
   class=MaterialExpressionNamedRerouteDeclaration
   inputs [None {from=MaterialExpressionConvert_0}]
   properties {name=TP_DDY, nodeColor=<rgba 0 0.7891 1 1>}
  }
  MaterialExpressionNamedRerouteDeclaration_19 {
   class=MaterialExpressionNamedRerouteDeclaration
   inputs [None {from=MaterialExpressionConvert_0}]
   properties {name=TP_DDX, nodeColor=<rgba 0.5703 0 1 1>}
  }
  MaterialExpressionNamedRerouteDeclaration_20 {
   class=MaterialExpressionNamedRerouteDeclaration
   inputs [None {from=MaterialExpressionMaterialFunctionCall_9, output=Output UV}]
   properties {name=TP_UV, nodeColor=<rgba 0 0.1094 1 1>}
  }
  MaterialExpressionNamedRerouteDeclaration_21 {
   class=MaterialExpressionNamedRerouteDeclaration
   inputs [None {from=MaterialExpressionMaterialFunctionCall_8, output=Output UV}]
   properties {name=TC_UV, nodeColor=<rgba 0 1 0.25 1>}
  }
  MaterialExpressionNamedRerouteDeclaration_28 {
   class=MaterialExpressionNamedRerouteDeclaration
   inputs [None {from=MaterialExpressionMaterialFunctionCall_8, output=Turns}]
   properties {name=TC_Rotation, nodeColor=<rgba 1 0.0703 0 1>}
  }
  MaterialExpressionNamedRerouteDeclaration_29 {
   class=MaterialExpressionNamedRerouteDeclaration
   inputs [None {from=MaterialExpressionSaturate_1}]
   properties {name=ReFresnelVX, nodeColor=<rgba 0.2656 0 1 1>}
  }
  MaterialExpressionNamedRerouteDeclaration_30 {
   class=MaterialExpressionNamedRerouteDeclaration
   inputs [None {from=MaterialExpressionScalarParameter_32}]
   properties {name=TransitionSoftness, nodeColor=<rgba 0 1 0.2031 1>}
  }
  MaterialExpressionNamedRerouteDeclaration_38 {
   class=MaterialExpressionNamedRerouteDeclaration
   inputs [None {from=MaterialExpressionScalarParameter_17}]
   properties {name=UV_Scale, nodeColor=<rgba 0.7578 1 0 1>}
  }
  MaterialExpressionNamedRerouteDeclaration_39 {
   class=MaterialExpressionNamedRerouteDeclaration
   inputs [None {from=MaterialExpressionScalarParameter_16}]
   properties {name=UV_Rot, nodeColor=<rgba 0.7578 1 0 1>}
  }
  MaterialExpressionNamedRerouteDeclaration_40 {
   class=MaterialExpressionNamedRerouteDeclaration
   inputs [None {from=MaterialExpressionScalarParameter_18}]
   properties {name=UV_X, nodeColor=<rgba 0 1 0.6016 1>}
  }
  MaterialExpressionNamedRerouteDeclaration_41 {
   class=MaterialExpressionNamedRerouteDeclaration
   inputs [None {from=MaterialExpressionScalarParameter_19}]
   properties {name=UV_Y, nodeColor=<rgba 0 1 0.6016 1>}
  }
  MaterialExpressionNamedRerouteDeclaration_47 {
   class=MaterialExpressionNamedRerouteDeclaration
   inputs [None {from=MaterialExpressionScalarParameter_30}]
   properties {name=UV_Z, nodeColor=<rgba 0 1 0.6016 1>}
  }
  MaterialExpressionNamedRerouteDeclaration_48 {
   class=MaterialExpressionNamedRerouteDeclaration
   inputs [None {from=MaterialExpressionScalarParameter_29}]
   properties {name=OffsetX, nodeColor=<rgba 0.3594 1 0 1>}
  }
  MaterialExpressionNamedRerouteDeclaration_58 {
   class=MaterialExpressionNamedRerouteDeclaration
   inputs [None {from=MaterialExpressionScalarParameter_26}]
   properties {name=OffsetY, nodeColor=<rgba 0.3594 1 0 1>}
  }
  MaterialExpressionNamedRerouteDeclaration_61 {
   class=MaterialExpressionNamedRerouteDeclaration
   inputs [None {from=MaterialExpressionScalarParameter_31}]
   properties {name=OffsetZ, nodeColor=<rgba 0.3594 1 0 1>}
  }
  MaterialExpressionNamedRerouteDeclaration_69 {
   class=MaterialExpressionNamedRerouteDeclaration
   inputs [None {from=MaterialExpressionSaturate_2}]
   properties {name=ReFresnelPX, nodeColor=<rgba 0.2656 0 1 1>}
  }
  MaterialExpressionNamedRerouteDeclaration_70 {
   class=MaterialExpressionNamedRerouteDeclaration
   inputs [None {from=MaterialExpressionDivide_3}]
   properties {name=AxF_ScaleX, nodeColor=<rgba 0.7605 0 0.0009 1>}
  }
  MaterialExpressionNamedRerouteDeclaration_71 {
   class=MaterialExpressionNamedRerouteDeclaration
   inputs [None {from=MaterialExpressionDivide_4}]
   properties {name=AxF_ScaleY, nodeColor=<rgba 0.7605 0 0.0009 1>}
  }
  MaterialExpressionNamedRerouteDeclaration_72 {
   class=MaterialExpressionNamedRerouteDeclaration
   inputs [None {from=MaterialExpressionMultiply_3}]
   properties {name=AxF_UnitScale_MM, nodeColor=<rgba 0.7605 0 0.0009 1>}
  }
  MaterialExpressionNamedRerouteDeclaration_73 {
   class=MaterialExpressionNamedRerouteDeclaration
   inputs [None {from=MaterialExpressionStaticBoolParameter_1}]
   properties {name=UseGeoUVs, nodeColor=<rgba 0 0.8359 1 1>}
  }
  MaterialExpressionNamedRerouteDeclaration_74 {
   class=MaterialExpressionNamedRerouteDeclaration
   inputs [None {from=MaterialExpressionConvert_1}]
   properties {name=TP_TileXYZ, nodeColor=<rgba 1 0.8438 0 1>}
  }
  MaterialExpressionNamedRerouteDeclaration_75 {
   class=MaterialExpressionNamedRerouteDeclaration
   inputs [None {from=MaterialExpressionConvert_2}]
   properties {name=TP_OffsetXYZ, nodeColor=<rgba 0.5234 0 1 1>}
  }
  MaterialExpressionNamedRerouteDeclaration_76 {
   class=MaterialExpressionNamedRerouteDeclaration
   inputs [None {from=MaterialExpressionShadowReplace_1}]
   properties {name=Alpha_O, nodeColor=<rgba 0.3594 0 1 1>}
  }
  MaterialExpressionNamedRerouteDeclaration_8 {
   class=MaterialExpressionNamedRerouteDeclaration
   inputs [None {from=MaterialExpressionTextureObjectParameter_0}]
   properties {name=HeightTex, nodeColor=<rgba 0 1 0.8125 1>}
  }
  MaterialExpressionNamedRerouteDeclaration_87 {
   class=MaterialExpressionNamedRerouteDeclaration
   inputs [None {from=MaterialExpressionMaterialFunctionCall_34, output=Top Layer Slab}]
   properties {name=ClearCoatTerm_O, nodeColor=<rgba 1 0.6797 0 1>}
  }
  MaterialExpressionNamedRerouteDeclaration_88 {
   class=MaterialExpressionNamedRerouteDeclaration
   inputs [None {from=MaterialExpressionStaticSwitchParameter_19}]
   properties {name=DiffuseAlbedoTerm_O, nodeColor=<rgba 0 1 0.0391 1>}
  }
  MaterialExpressionNamedRerouteDeclaration_89 {
   class=MaterialExpressionNamedRerouteDeclaration
   inputs [None {from=MaterialExpressionStaticSwitchParameter_23}]
   properties {name=F0_O, nodeColor=<rgba 1 0 0.2812 1>}
  }
  MaterialExpressionNamedRerouteDeclaration_9 {
   class=MaterialExpressionNamedRerouteDeclaration
   inputs [None {from=MaterialExpressionScalarParameter_8}]
   properties {name=HeightStrength, nodeColor=<rgba 0 1 0.8125 1>}
  }
  MaterialExpressionNamedRerouteDeclaration_90 {
   class=MaterialExpressionNamedRerouteDeclaration
   inputs [None {from=MaterialExpressionStaticSwitchParameter_24}]
   properties {name=F90_O, nodeColor=<rgba 1 0 0.2812 1>}
  }
  MaterialExpressionNamedRerouteDeclaration_91 {
   class=MaterialExpressionNamedRerouteDeclaration
   inputs [None {from=MaterialExpressionSquareRoot_1}]
   properties {name=Roughness_O, nodeColor=<rgba 0.9219 1 0 1>}
  }
  MaterialExpressionNamedRerouteDeclaration_92 {
   class=MaterialExpressionNamedRerouteDeclaration
   inputs [None {from=MaterialExpressionMaterialFunctionCall_38, output=Anisotropy}]
   properties {name=Anisotropy_O, nodeColor=<rgba 0.9219 1 0 1>}
  }
  MaterialExpressionNamedRerouteDeclaration_93 {
   class=MaterialExpressionNamedRerouteDeclaration
   inputs [None {from=MaterialExpressionMaterialFunctionCall_38, output=Tangent}]
   properties {name=Tangent_O, nodeColor=<rgba 0.9219 1 0 1>}
  }
  MaterialExpressionNamedRerouteDeclaration_94 {
   class=MaterialExpressionNamedRerouteDeclaration
   inputs [None {from=MaterialExpressionStaticSwitchParameter_25}]
   properties {name=FuzzRoughness_O, nodeColor=<rgba 0.0547 1 0 1>}
  }
  MaterialExpressionNamedRerouteDeclaration_95 {
   class=MaterialExpressionNamedRerouteDeclaration
   inputs [None {from=MaterialExpressionStaticSwitchParameter_26}]
   properties {name=FuzzAmount_O, nodeColor=<rgba 0.0547 1 0 1>}
  }
  MaterialExpressionNamedRerouteDeclaration_96 {
   class=MaterialExpressionNamedRerouteDeclaration
   inputs [None {from=MaterialExpressionStaticSwitchParameter_27}]
   properties {name=FuzzColor_O, nodeColor=<rgba 0.0547 1 0 1>}
  }
  MaterialExpressionNamedRerouteDeclaration_97 {
   class=MaterialExpressionNamedRerouteDeclaration
   inputs [None {from=MaterialExpressionStaticSwitchParameter_20}]
   properties {name=Normal_O, nodeColor=<rgba 1 0.9844 0 1>}
  }
  MaterialExpressionNamedRerouteDeclaration_98 {
   class=MaterialExpressionNamedRerouteDeclaration
   inputs [None {from=MaterialExpressionMaterialFunctionCall_35, output=R}]
   properties {name=AlphaX, nodeColor=<rgba 0.1484 1 0 1>}
  }
  MaterialExpressionNamedRerouteDeclaration_99 {
   class=MaterialExpressionNamedRerouteDeclaration
   inputs [None {from=MaterialExpressionMaterialFunctionCall_35, output=G}]
   properties {name=AlphaY, nodeColor=<rgba 0.1484 1 0 1>}
  }
  MaterialExpressionNamedRerouteUsage_10 {class=MaterialExpressionNamedRerouteUsage}
  MaterialExpressionNamedRerouteUsage_101 {class=MaterialExpressionNamedRerouteUsage}
  MaterialExpressionNamedRerouteUsage_103 {class=MaterialExpressionNamedRerouteUsage}
  MaterialExpressionNamedRerouteUsage_104 {class=MaterialExpressionNamedRerouteUsage}
  MaterialExpressionNamedRerouteUsage_109 {class=MaterialExpressionNamedRerouteUsage}
  MaterialExpressionNamedRerouteUsage_11 {class=MaterialExpressionNamedRerouteUsage}
  MaterialExpressionNamedRerouteUsage_113 {class=MaterialExpressionNamedRerouteUsage}
  MaterialExpressionNamedRerouteUsage_114 {class=MaterialExpressionNamedRerouteUsage}
  MaterialExpressionNamedRerouteUsage_118 {class=MaterialExpressionNamedRerouteUsage}
  MaterialExpressionNamedRerouteUsage_12 {class=MaterialExpressionNamedRerouteUsage}
  MaterialExpressionNamedRerouteUsage_125 {class=MaterialExpressionNamedRerouteUsage}
  MaterialExpressionNamedRerouteUsage_130 {class=MaterialExpressionNamedRerouteUsage}
  MaterialExpressionNamedRerouteUsage_131 {class=MaterialExpressionNamedRerouteUsage}
  MaterialExpressionNamedRerouteUsage_133 {class=MaterialExpressionNamedRerouteUsage}
  MaterialExpressionNamedRerouteUsage_195 {class=MaterialExpressionNamedRerouteUsage}
  MaterialExpressionNamedRerouteUsage_196 {class=MaterialExpressionNamedRerouteUsage}
  MaterialExpressionNamedRerouteUsage_197 {class=MaterialExpressionNamedRerouteUsage}
  MaterialExpressionNamedRerouteUsage_198 {class=MaterialExpressionNamedRerouteUsage}
  MaterialExpressionNamedRerouteUsage_199 {class=MaterialExpressionNamedRerouteUsage}
  MaterialExpressionNamedRerouteUsage_200 {class=MaterialExpressionNamedRerouteUsage}
  MaterialExpressionNamedRerouteUsage_201 {class=MaterialExpressionNamedRerouteUsage}
  MaterialExpressionNamedRerouteUsage_202 {class=MaterialExpressionNamedRerouteUsage}
  MaterialExpressionNamedRerouteUsage_203 {class=MaterialExpressionNamedRerouteUsage}
  MaterialExpressionNamedRerouteUsage_204 {class=MaterialExpressionNamedRerouteUsage}
  MaterialExpressionNamedRerouteUsage_205 {class=MaterialExpressionNamedRerouteUsage}
  MaterialExpressionNamedRerouteUsage_206 {class=MaterialExpressionNamedRerouteUsage}
  MaterialExpressionNamedRerouteUsage_207 {class=MaterialExpressionNamedRerouteUsage}
  MaterialExpressionNamedRerouteUsage_208 {class=MaterialExpressionNamedRerouteUsage}
  MaterialExpressionNamedRerouteUsage_209 {class=MaterialExpressionNamedRerouteUsage}
  MaterialExpressionNamedRerouteUsage_21 {class=MaterialExpressionNamedRerouteUsage}
  MaterialExpressionNamedRerouteUsage_210 {class=MaterialExpressionNamedRerouteUsage}
  MaterialExpressionNamedRerouteUsage_211 {class=MaterialExpressionNamedRerouteUsage}
  MaterialExpressionNamedRerouteUsage_212 {class=MaterialExpressionNamedRerouteUsage}
  MaterialExpressionNamedRerouteUsage_213 {class=MaterialExpressionNamedRerouteUsage}
  MaterialExpressionNamedRerouteUsage_214 {class=MaterialExpressionNamedRerouteUsage}
  MaterialExpressionNamedRerouteUsage_215 {class=MaterialExpressionNamedRerouteUsage}
  MaterialExpressionNamedRerouteUsage_216 {class=MaterialExpressionNamedRerouteUsage}
  MaterialExpressionNamedRerouteUsage_217 {class=MaterialExpressionNamedRerouteUsage}
  MaterialExpressionNamedRerouteUsage_218 {class=MaterialExpressionNamedRerouteUsage}
  MaterialExpressionNamedRerouteUsage_219 {class=MaterialExpressionNamedRerouteUsage}
  MaterialExpressionNamedRerouteUsage_22 {class=MaterialExpressionNamedRerouteUsage}
  MaterialExpressionNamedRerouteUsage_220 {class=MaterialExpressionNamedRerouteUsage}
  MaterialExpressionNamedRerouteUsage_221 {class=MaterialExpressionNamedRerouteUsage}
  MaterialExpressionNamedRerouteUsage_222 {class=MaterialExpressionNamedRerouteUsage}
  MaterialExpressionNamedRerouteUsage_223 {class=MaterialExpressionNamedRerouteUsage}
  MaterialExpressionNamedRerouteUsage_224 {class=MaterialExpressionNamedRerouteUsage}
  MaterialExpressionNamedRerouteUsage_225 {class=MaterialExpressionNamedRerouteUsage}
  MaterialExpressionNamedRerouteUsage_226 {class=MaterialExpressionNamedRerouteUsage}
  MaterialExpressionNamedRerouteUsage_227 {class=MaterialExpressionNamedRerouteUsage}
  MaterialExpressionNamedRerouteUsage_228 {class=MaterialExpressionNamedRerouteUsage}
  MaterialExpressionNamedRerouteUsage_229 {class=MaterialExpressionNamedRerouteUsage}
  MaterialExpressionNamedRerouteUsage_230 {class=MaterialExpressionNamedRerouteUsage}
  MaterialExpressionNamedRerouteUsage_26 {class=MaterialExpressionNamedRerouteUsage}
  MaterialExpressionNamedRerouteUsage_290 {class=MaterialExpressionNamedRerouteUsage}
  MaterialExpressionNamedRerouteUsage_291 {class=MaterialExpressionNamedRerouteUsage}
  MaterialExpressionNamedRerouteUsage_292 {class=MaterialExpressionNamedRerouteUsage}
  MaterialExpressionNamedRerouteUsage_293 {class=MaterialExpressionNamedRerouteUsage}
  MaterialExpressionNamedRerouteUsage_294 {class=MaterialExpressionNamedRerouteUsage}
  MaterialExpressionNamedRerouteUsage_295 {class=MaterialExpressionNamedRerouteUsage}
  MaterialExpressionNamedRerouteUsage_296 {class=MaterialExpressionNamedRerouteUsage}
  MaterialExpressionNamedRerouteUsage_297 {class=MaterialExpressionNamedRerouteUsage}
  MaterialExpressionNamedRerouteUsage_298 {class=MaterialExpressionNamedRerouteUsage}
  MaterialExpressionNamedRerouteUsage_299 {class=MaterialExpressionNamedRerouteUsage}
  MaterialExpressionNamedRerouteUsage_30 {class=MaterialExpressionNamedRerouteUsage}
  MaterialExpressionNamedRerouteUsage_300 {class=MaterialExpressionNamedRerouteUsage}
  MaterialExpressionNamedRerouteUsage_301 {class=MaterialExpressionNamedRerouteUsage}
  MaterialExpressionNamedRerouteUsage_302 {class=MaterialExpressionNamedRerouteUsage}
  MaterialExpressionNamedRerouteUsage_303 {class=MaterialExpressionNamedRerouteUsage}
  MaterialExpressionNamedRerouteUsage_304 {class=MaterialExpressionNamedRerouteUsage}
  MaterialExpressionNamedRerouteUsage_305 {class=MaterialExpressionNamedRerouteUsage}
  MaterialExpressionNamedRerouteUsage_31 {class=MaterialExpressionNamedRerouteUsage}
  MaterialExpressionNamedRerouteUsage_317 {class=MaterialExpressionNamedRerouteUsage}
  MaterialExpressionNamedRerouteUsage_318 {class=MaterialExpressionNamedRerouteUsage}
  MaterialExpressionNamedRerouteUsage_319 {class=MaterialExpressionNamedRerouteUsage}
  MaterialExpressionNamedRerouteUsage_32 {class=MaterialExpressionNamedRerouteUsage}
  MaterialExpressionNamedRerouteUsage_320 {class=MaterialExpressionNamedRerouteUsage}
  MaterialExpressionNamedRerouteUsage_321 {class=MaterialExpressionNamedRerouteUsage}
  MaterialExpressionNamedRerouteUsage_322 {class=MaterialExpressionNamedRerouteUsage}
  MaterialExpressionNamedRerouteUsage_323 {class=MaterialExpressionNamedRerouteUsage}
  MaterialExpressionNamedRerouteUsage_324 {class=MaterialExpressionNamedRerouteUsage}
  MaterialExpressionNamedRerouteUsage_325 {class=MaterialExpressionNamedRerouteUsage}
  MaterialExpressionNamedRerouteUsage_326 {class=MaterialExpressionNamedRerouteUsage}
  MaterialExpressionNamedRerouteUsage_327 {class=MaterialExpressionNamedRerouteUsage}
  MaterialExpressionNamedRerouteUsage_36 {class=MaterialExpressionNamedRerouteUsage}
  MaterialExpressionNamedRerouteUsage_37 {class=MaterialExpressionNamedRerouteUsage}
  MaterialExpressionNamedRerouteUsage_38 {class=MaterialExpressionNamedRerouteUsage}
  MaterialExpressionNamedRerouteUsage_39 {class=MaterialExpressionNamedRerouteUsage}
  MaterialExpressionNamedRerouteUsage_40 {class=MaterialExpressionNamedRerouteUsage}
  MaterialExpressionNamedRerouteUsage_50 {class=MaterialExpressionNamedRerouteUsage}
  MaterialExpressionNamedRerouteUsage_51 {class=MaterialExpressionNamedRerouteUsage}
  MaterialExpressionNamedRerouteUsage_52 {class=MaterialExpressionNamedRerouteUsage}
  MaterialExpressionNamedRerouteUsage_53 {class=MaterialExpressionNamedRerouteUsage}
  MaterialExpressionNamedRerouteUsage_56 {class=MaterialExpressionNamedRerouteUsage}
  MaterialExpressionNamedRerouteUsage_57 {class=MaterialExpressionNamedRerouteUsage}
  MaterialExpressionNamedRerouteUsage_58 {class=MaterialExpressionNamedRerouteUsage}
  MaterialExpressionNamedRerouteUsage_59 {class=MaterialExpressionNamedRerouteUsage}
  MaterialExpressionNamedRerouteUsage_60 {class=MaterialExpressionNamedRerouteUsage}
  MaterialExpressionNamedRerouteUsage_63 {class=MaterialExpressionNamedRerouteUsage}
  MaterialExpressionNamedRerouteUsage_64 {class=MaterialExpressionNamedRerouteUsage}
  MaterialExpressionNamedRerouteUsage_65 {class=MaterialExpressionNamedRerouteUsage}
  MaterialExpressionNamedRerouteUsage_66 {class=MaterialExpressionNamedRerouteUsage}
  MaterialExpressionNamedRerouteUsage_69 {class=MaterialExpressionNamedRerouteUsage}
  MaterialExpressionNamedRerouteUsage_74 {class=MaterialExpressionNamedRerouteUsage}
  MaterialExpressionNamedRerouteUsage_75 {class=MaterialExpressionNamedRerouteUsage}
  MaterialExpressionOneMinus_0 {class=MaterialExpressionOneMinus, inputs [None {from=MaterialExpressionScalarParameter_40}]}
  MaterialExpressionPixelNormalWS_0 {class=MaterialExpressionPixelNormalWS, inputs [], properties {}}
  MaterialExpressionSaturate_1 {class=MaterialExpressionSaturate, inputs [None {from=MaterialExpressionAbs_1}]}
  MaterialExpressionSaturate_2 {class=MaterialExpressionSaturate, inputs [None {from=MaterialExpressionAbs_2}]}
  MaterialExpressionScalarParameter_14 {
   class=MaterialExpressionScalarParameter
   properties {desc=0 to 1, group=Parrallax Occlusion, parameterName=HeightReferencePlane, sliderMax=1, sortPriority=3}
  }
  MaterialExpressionScalarParameter_15 {
   class=MaterialExpressionScalarParameter
   properties {defaultValue=32, group=Parrallax Occlusion, parameterName=POM Max Steps, sortPriority=4}
  }
  MaterialExpressionScalarParameter_16 {class=MaterialExpressionScalarParameter, properties {group=UV Transforms, parameterName="Rotation (degree)", sliderMax=360, sortPriority=30}}
  MaterialExpressionScalarParameter_17 {
   class=MaterialExpressionScalarParameter
   properties {defaultValue=1, group=UV Transforms, parameterName=Scale, sliderMax=10000, sliderMin=0.001, sortPriority=10}
  }
  MaterialExpressionScalarParameter_18 {
   class=MaterialExpressionScalarParameter
   properties {defaultValue=1, group=UV Transforms, parameterName=StretchOnX, sliderMax=100, sliderMin=0.001, sortPriority=100}
  }
  MaterialExpressionScalarParameter_19 {
   class=MaterialExpressionScalarParameter
   properties {defaultValue=1, group=UV Transforms, parameterName=StretchOnY, sliderMax=1000, sliderMin=0.001, sortPriority=110}
  }
  MaterialExpressionScalarParameter_26 {class=MaterialExpressionScalarParameter, properties {group=UV Transforms, parameterName=OffsetOnY, sortPriority=140}}
  MaterialExpressionScalarParameter_29 {class=MaterialExpressionScalarParameter, properties {group=UV Transforms, parameterName=OffsetOnX, sortPriority=130}}
  MaterialExpressionScalarParameter_30 {
   class=MaterialExpressionScalarParameter
   properties {defaultValue=1, group=UV Transforms, parameterName=StretchOnZ, sliderMax=1000, sliderMin=0.001, sortPriority=120}
  }
  MaterialExpressionScalarParameter_31 {class=MaterialExpressionScalarParameter, properties {group=UV Transforms, parameterName=OffsetOnZ, sortPriority=150}}
  MaterialExpressionScalarParameter_32 {
   class=MaterialExpressionScalarParameter
   properties {
    defaultValue=0.5
    desc="Controls the contrast of the blend mask. Internally applied as pow(mask, 8 × BlendFactor). A default of 0.5 corresponds to pow4 contrast."
    group=UV Transforms
    parameterName=TransitionSoftness
    sliderMax=1
    sortPriority=3
   }
  }
  MaterialExpressionScalarParameter_33 {class=MaterialExpressionScalarParameter, properties {defaultValue=100, group=MillimeterSettings, parameterName=WidthMM, sortPriority=10}}
  MaterialExpressionScalarParameter_34 {class=MaterialExpressionScalarParameter, properties {defaultValue=100, group=MillimeterSettings, parameterName=HeightMM, sortPriority=11}}
  MaterialExpressionScalarParameter_35 {class=MaterialExpressionScalarParameter, properties {defaultValue=1, group=MillimeterSettings, parameterName=UnitScaleCM}}
  MaterialExpressionScalarParameter_36 {class=MaterialExpressionScalarParameter, properties {group=UV Transforms, parameterName=UV Index}}
  MaterialExpressionScalarParameter_37 {
   class=MaterialExpressionScalarParameter
   properties {defaultValue=1, group=Normal, parameterName=NormalStrength, sliderMax=2, sliderMin=-1, sortPriority=12}
  }
  MaterialExpressionScalarParameter_38 {
   class=MaterialExpressionScalarParameter
   properties {defaultValue=1, group=ClearCoat, parameterName=ClearCoatNormalStrength, sliderMax=2, sliderMin=-1}
  }
  MaterialExpressionScalarParameter_39 {class=MaterialExpressionScalarParameter, properties {defaultValue=1, group=Alpha, parameterName=Alpha}}
  MaterialExpressionScalarParameter_40 {class=MaterialExpressionScalarParameter, properties {group=Alpha, parameterName=ShadowCompensation, sliderMax=1, sortPriority=24}}
  MaterialExpressionScalarParameter_8 {
   class=MaterialExpressionScalarParameter
   properties {defaultValue=1, group=Parrallax Occlusion, parameterName=HeightStrength, sortPriority=2}
  }
  MaterialExpressionShaderStageSwitch_0 {
   class=MaterialExpressionShaderStageSwitch
   inputs [
    PixelShader {from=MaterialExpressionCameraVectorWS_1, output=""}
    VertexShader {from=MaterialExpressionMaterialFunctionCall_6, output=Result}
   ]
   properties {}
  }
  MaterialExpressionShadowReplace_1 {
   class=MaterialExpressionShadowReplace
   inputs [
    Default {from=MaterialExpressionStaticSwitchParameter_1, output=""}
    Shadow {from=MaterialExpressionClamp_0, output=""}
   ]
   properties {}
  }
  MaterialExpressionSquareRoot_1 {class=MaterialExpressionSquareRoot, inputs [None {from=MaterialExpressionDivide_6, output=""}], properties {}}
  MaterialExpressionStaticBoolParameter_1 {
   class=MaterialExpressionStaticBoolParameter
   inputs []
   properties {defaultValue=true, group=UV Transforms, parameterName=Use Geo UVs, sortPriority=0}
  }
  MaterialExpressionStaticSwitchParameter_0 {
   class=MaterialExpressionStaticSwitchParameter
   inputs [
    True {from="", output=""}
    False {from="", output=""}
   ]
   properties {defaultValue=true, group=UV Transforms, parameterName=UsingLocalSpace, sortPriority=2}
  }
  MaterialExpressionStaticSwitchParameter_1 {
   class=MaterialExpressionStaticSwitchParameter
   inputs [
    True {from=MaterialExpressionStaticSwitch_9, output=""}
    False {from=MaterialExpressionScalarParameter_39, output=""}
   ]
   properties {group=Alpha, parameterName=HasAlphaTexture}
  }
  MaterialExpressionStaticSwitchParameter_19 {
   class=MaterialExpressionStaticSwitchParameter
   inputs [
    True {from=MaterialExpressionStaticSwitch_1, output=""}
    False {from=MaterialExpressionVectorParameter_9, output=RGB}
   ]
   properties {group=Diffuse, parameterName=HasDiffuseColorTexture}
  }
  MaterialExpressionStaticSwitchParameter_2 {
   class=MaterialExpressionStaticSwitchParameter
   inputs [
    True {from=MaterialExpressionConstant_2, output=""}
    False {from=MaterialExpressionConstant_1, output=""}
   ]
   properties {defaultValue=true, group=UV Transforms, parameterName=UsingLocalSpace, sortPriority=2}
  }
  MaterialExpressionStaticSwitchParameter_20 {
   class=MaterialExpressionStaticSwitchParameter
   inputs [
    True {from=MaterialExpressionMaterialFunctionCall_45, output=NormalOut}
    False {from=MaterialExpressionComponentMask_4, output=""}
   ]
   properties {group=Normal, parameterName=HasNormalTexture, sortPriority=30}
  }
  MaterialExpressionStaticSwitchParameter_21 {
   class=MaterialExpressionStaticSwitchParameter
   inputs [
    True {from=MaterialExpressionStaticSwitch_7, output=""}
    False {from=MaterialExpressionVectorParameter_12, output=RGB}
   ]
   properties {group=Roughness, parameterName=HasRoughnessTexture, sortPriority=30}
  }
  MaterialExpressionStaticSwitchParameter_22 {
   class=MaterialExpressionStaticSwitchParameter
   inputs [
    True {from=MaterialExpressionStaticSwitch_8, output=""}
    False {from=MaterialExpressionMaterialFunctionCall_40, output=Result}
   ]
   properties {group=Anisotropy, parameterName=HasAnisotropicRotationTexture}
  }
  MaterialExpressionStaticSwitchParameter_23 {
   class=MaterialExpressionStaticSwitchParameter
   inputs [
    True {from=MaterialExpressionStaticSwitch_4, output=""}
    False {from=MaterialExpressionVectorParameter_13, output=RGB}
   ]
   properties {group=Fresnel, parameterName=HasF0Texture}
  }
  MaterialExpressionStaticSwitchParameter_24 {
   class=MaterialExpressionStaticSwitchParameter
   inputs [
    True {from=MaterialExpressionStaticSwitch_5, output=""}
    False {from=MaterialExpressionVectorParameter_14, output=RGB}
   ]
   properties {group=Fresnel, parameterName=HasF90Texture}
  }
  MaterialExpressionStaticSwitchParameter_25 {
   class=MaterialExpressionStaticSwitchParameter
   inputs [
    True {from=MaterialExpressionNamedRerouteUsage_197, output=""}
    False {from=MaterialExpressionConstant_14, output=""}
   ]
   properties {group=Sheen, parameterName=HasSheen, sortPriority=30}
  }
  MaterialExpressionStaticSwitchParameter_26 {
   class=MaterialExpressionStaticSwitchParameter
   inputs [
    True {from=MaterialExpressionConstant_20, output=""}
    False {from=MaterialExpressionConstant_12, output=""}
   ]
   properties {group=Sheen, parameterName=HasSheen, sortPriority=30}
  }
  MaterialExpressionStaticSwitchParameter_27 {
   class=MaterialExpressionStaticSwitchParameter
   inputs [
    True {from=MaterialExpressionStaticSwitchParameter_28, output=""}
    False {from=MaterialExpressionConstant_13, output=""}
   ]
   properties {group=Sheen, parameterName=HasSheen, sortPriority=30}
  }
  MaterialExpressionStaticSwitchParameter_28 {
   class=MaterialExpressionStaticSwitchParameter
   inputs [
    True {from=MaterialExpressionStaticSwitch_3, output=""}
    False {from=MaterialExpressionVectorParameter_15, output=RGB}
   ]
   properties {group=Sheen, parameterName=HasSheenColorTexture}
  }
  MaterialExpressionStaticSwitchParameter_32 {
   class=MaterialExpressionStaticSwitchParameter
   inputs [
    True {from=MaterialExpressionStaticSwitch_2, output=""}
    False {from=MaterialExpressionVectorParameter_16, output=RGB}
   ]
   properties {group=Sheen, parameterName=HasSheenRoughnessTexture}
  }
  MaterialExpressionStaticSwitchParameter_33 {
   class=MaterialExpressionStaticSwitchParameter
   inputs [
    True {from=MaterialExpressionMaterialFunctionCall_44, output=NormalOut}
    False {from=MaterialExpressionComponentMask_5, output=""}
   ]
   properties {group=ClearCoat, parameterName=HasClearCoatNormalTexture, sortPriority=30}
  }
  MaterialExpressionStaticSwitchParameter_35 {
   class=MaterialExpressionStaticSwitchParameter
   inputs [
    True {from=MaterialExpressionSubstrateVerticalLayering_1, output=""}
    False {from=MaterialExpressionSubstrateSlabBSDF_1, output=""}
   ]
   properties {group=ClearCoat, parameterName=HasClearCoat, sortPriority=30}
  }
  MaterialExpressionStaticSwitch_0 {
   class=MaterialExpressionStaticSwitch
   inputs [
    True {from=MaterialExpressionTextureSample_7, output=RGB}
    False {from=MaterialExpressionMaterialFunctionCall_46, output=TSN}
    Value {from=MaterialExpressionNamedRerouteUsage_66, output=""}
   ]
   properties {}
  }
  MaterialExpressionStaticSwitch_1 {
   class=MaterialExpressionStaticSwitch
   inputs [
    True {from=MaterialExpressionTextureSample_0, output=RGB}
    False {from=MaterialExpressionMaterialFunctionCall_41, output=RGB}
    Value {from=MaterialExpressionNamedRerouteUsage_290, output=""}
   ]
   properties {}
  }
  MaterialExpressionStaticSwitch_2 {
   class=MaterialExpressionStaticSwitch
   inputs [
    True {from=MaterialExpressionTextureSample_5, output=RGB}
    False {from=MaterialExpressionMaterialFunctionCall_21, output=RGB}
    Value {from=MaterialExpressionNamedRerouteUsage_291, output=""}
   ]
   properties {}
  }
  MaterialExpressionStaticSwitch_3 {
   class=MaterialExpressionStaticSwitch
   inputs [
    True {from=MaterialExpressionTextureSample_6, output=RGB}
    False {from=MaterialExpressionMaterialFunctionCall_23, output=RGB}
    Value {from=MaterialExpressionNamedRerouteUsage_292, output=""}
   ]
   properties {}
  }
  MaterialExpressionStaticSwitch_4 {
   class=MaterialExpressionStaticSwitch
   inputs [
    True {from=MaterialExpressionTextureSample_3, output=RGB}
    False {from=MaterialExpressionMaterialFunctionCall_43, output=RGB}
    Value {from=MaterialExpressionNamedRerouteUsage_293, output=""}
   ]
   properties {}
  }
  MaterialExpressionStaticSwitch_5 {
   class=MaterialExpressionStaticSwitch
   inputs [
    True {from=MaterialExpressionTextureSample_4, output=RGB}
    False {from=MaterialExpressionMaterialFunctionCall_12, output=RGB}
    Value {from=MaterialExpressionNamedRerouteUsage_294, output=""}
   ]
   properties {}
  }
  MaterialExpressionStaticSwitch_6 {
   class=MaterialExpressionStaticSwitch
   inputs [
    True {from=MaterialExpressionTextureSample_1, output=RGB}
    False {from=MaterialExpressionMaterialFunctionCall_47, output=TSN}
    Value {from=MaterialExpressionNamedRerouteUsage_295, output=""}
   ]
   properties {}
  }
  MaterialExpressionStaticSwitch_7 {
   class=MaterialExpressionStaticSwitch
   inputs [
    True {from=MaterialExpressionTextureSample_9, output=RGB}
    False {from=MaterialExpressionMaterialFunctionCall_15, output=RGB}
    Value {from=MaterialExpressionNamedRerouteUsage_296, output=""}
   ]
   properties {}
  }
  MaterialExpressionStaticSwitch_8 {
   class=MaterialExpressionStaticSwitch
   inputs [
    True {from=MaterialExpressionMaterialFunctionCall_36, output=Result}
    False {from=MaterialExpressionMaterialFunctionCall_37, output=Result}
    Value {from=MaterialExpressionNamedRerouteUsage_297, output=""}
   ]
   properties {}
  }
  MaterialExpressionStaticSwitch_9 {
   class=MaterialExpressionStaticSwitch
   inputs [
    True {from=MaterialExpressionTextureSample_10, output=G}
    False {from=MaterialExpressionMaterialFunctionCall_10, output=G}
    Value {from=MaterialExpressionNamedRerouteUsage_118, output=""}
   ]
   properties {}
  }
  MaterialExpressionSubstrateSlabBSDF_1 {
   class=MaterialExpressionSubstrateSlabBSDF
   inputs [
    Diffuse Albedo {from=MaterialExpressionNamedRerouteUsage_318, output=""}
    F0 {from=MaterialExpressionNamedRerouteUsage_319, output=""}
    F90 {from=MaterialExpressionNamedRerouteUsage_327, output=""}
    Roughness {from=MaterialExpressionNamedRerouteUsage_320, output=""}
    Anisotropy {from=MaterialExpressionNamedRerouteUsage_321, output=""}
    Normal {from=MaterialExpressionNamedRerouteUsage_326, output=""}
    Tangent {from=MaterialExpressionNamedRerouteUsage_322, output=""}
    SSS MFP {from="", output=""}
    SSS MFP Scale {from="", output=""}
    SSS Phase Anisotropy {from="", output=""}
    Emissive Color {from="", output=""}
    Second Roughness {from="", output=""}
    Second Roughness Weight {from="", output=""}
    Fuzz Roughness {from=MaterialExpressionNamedRerouteUsage_323, output=""}
    Fuzz Amount {from=MaterialExpressionNamedRerouteUsage_324, output=""}
    Fuzz Color {from=MaterialExpressionNamedRerouteUsage_325, output=""}
    "Glint Density (Disabled)" {from="", output=""}
    "Glint UVs (Disabled)" {from="", output=""}
   ]
   properties {}
  }
  MaterialExpressionSubstrateVerticalLayering_1 {
   class=MaterialExpressionSubstrateVerticalLayering
   inputs [
    Top {from=MaterialExpressionNamedRerouteUsage_317, output=""}
    Bottom {from=MaterialExpressionSubstrateSlabBSDF_1, output=""}
    Top Thickness {from="", output=""}
   ]
   properties {}
  }
  MaterialExpressionTextureObjectParameter_0 {
   class=MaterialExpressionTextureObjectParameter
   inputs [
    UVs {from="", output=""}
    Apply View MipBias {from="", output=""}
   ]
   properties {
    group=Parrallax Occlusion
    parameterName=HeightMap
    samplerType=SAMPLERTYPE_LinearColor
    sortPriority=1
    texture=/InterchangeAssets/gltf/Textures/T_White_Linear
   }
  }
  MaterialExpressionTextureObjectParameter_12 {
   class=MaterialExpressionTextureObjectParameter
   inputs [
    UVs {from="", output=""}
    Apply View MipBias {from="", output=""}
   ]
   properties {
    group=Roughness
    parameterName=RoughnessTexture
    samplerType=SAMPLERTYPE_LinearColor
    sortPriority=1
    texture=/Engine/Functions/Engine_MaterialFunctions02/PivotPainter2/Black_1x1_EXR_Texture
   }
  }
  MaterialExpressionTextureObjectParameter_13 {
   class=MaterialExpressionTextureObjectParameter
   inputs [
    UVs {from="", output=""}
    Apply View MipBias {from="", output=""}
   ]
   properties {
    group=Anisotropy
    parameterName=AnisotropicRotationTexture
    samplerType=SAMPLERTYPE_LinearColor
    sortPriority=1
    texture=&Black_1x1_EXR_Texture
   }
  }
  MaterialExpressionTextureObjectParameter_14 {
   class=MaterialExpressionTextureObjectParameter
   inputs [
    UVs {from="", output=""}
    Apply View MipBias {from="", output=""}
   ]
   properties {
    group=Fresnel
    parameterName=F90Texture
    samplerType=SAMPLERTYPE_LinearColor
    sortPriority=2
    texture=&Black_1x1_EXR_Texture
   }
  }
  MaterialExpressionTextureObjectParameter_15 {
   class=MaterialExpressionTextureObjectParameter
   inputs [
    UVs {from="", output=""}
    Apply View MipBias {from="", output=""}
   ]
   properties {
    group=Fresnel
    parameterName=F0Texture
    samplerType=SAMPLERTYPE_LinearColor
    sortPriority=1
    texture=&Black_1x1_EXR_Texture
   }
  }
  MaterialExpressionTextureObjectParameter_16 {
   class=MaterialExpressionTextureObjectParameter
   inputs [
    UVs {from="", output=""}
    Apply View MipBias {from="", output=""}
   ]
   properties {
    group=Sheen
    parameterName=SheenColorTexture
    samplerType=SAMPLERTYPE_LinearColor
    sortPriority=2
    texture=&Black_1x1_EXR_Texture
   }
  }
  MaterialExpressionTextureObjectParameter_17 {
   class=MaterialExpressionTextureObjectParameter
   inputs [
    UVs {from="", output=""}
    Apply View MipBias {from="", output=""}
   ]
   properties {
    group=Sheen
    parameterName=SheenRoughnessTexture
    samplerType=SAMPLERTYPE_LinearColor
    sortPriority=1
    texture=&Black_1x1_EXR_Texture
   }
  }
  MaterialExpressionTextureObjectParameter_18 {
   class=MaterialExpressionTextureObjectParameter
   inputs [
    UVs {from="", output=""}
    Apply View MipBias {from="", output=""}
   ]
   properties {
    group=ClearCoat
    parameterName=ClearCoatNormalTexture
    samplerType=SAMPLERTYPE_Normal
    sortPriority=1
    texture=/Engine/EngineMaterials/DefaultNormal
   }
  }
  MaterialExpressionTextureObjectParameter_2 {
   class=MaterialExpressionTextureObjectParameter
   inputs [
    UVs {from="", output=""}
    Apply View MipBias {from="", output=""}
   ]
   properties {
    group=Diffuse
    parameterName=DiffuseColorTexture
    samplerType=SAMPLERTYPE_LinearColor
    sortPriority=1
    texture=&Black_1x1_EXR_Texture
   }
  }
  MaterialExpressionTextureObjectParameter_3 {
   class=MaterialExpressionTextureObjectParameter
   inputs [
    UVs {from="", output=""}
    Apply View MipBias {from="", output=""}
   ]
   properties {group=Normal, parameterName=NormalTexture, samplerType=SAMPLERTYPE_Normal, sortPriority=1, texture=&DefaultNormal}
  }
  MaterialExpressionTextureObjectParameter_4 {
   class=MaterialExpressionTextureObjectParameter
   inputs [
    UVs {from="", output=""}
    Apply View MipBias {from="", output=""}
   ]
   properties {
    group=Alpha
    parameterName=AlphaTexture
    samplerType=SAMPLERTYPE_LinearColor
    sortPriority=1
    texture=&T_White_Linear
   }
  }
  MaterialExpressionTextureSample_0 {
   class=MaterialExpressionTextureSample
   inputs [
    UVs {from=MaterialExpressionNamedRerouteUsage_218}
    Tex {from=MaterialExpressionTextureObjectParameter_2}
    Apply View MipBias {from=""}
   ]
   properties {samplerType=SAMPLERTYPE_LinearColor}
  }
  MaterialExpressionTextureSample_1 {
   class=MaterialExpressionTextureSample
   inputs [
    UVs {from=MaterialExpressionNamedRerouteUsage_230}
    Tex {from=MaterialExpressionTextureObjectParameter_18}
    Apply View MipBias {from=""}
   ]
  }
  MaterialExpressionTextureSample_10 {
   class=MaterialExpressionTextureSample
   inputs [
    UVs {from=MaterialExpressionNamedRerouteUsage_104}
    Tex {from=MaterialExpressionTextureObjectParameter_4}
    Apply View MipBias {from=""}
   ]
   properties {samplerType=SAMPLERTYPE_LinearColor}
  }
  MaterialExpressionTextureSample_3 {
   class=MaterialExpressionTextureSample
   inputs [
    UVs {from=MaterialExpressionNamedRerouteUsage_226}
    Tex {from=MaterialExpressionTextureObjectParameter_15}
    Apply View MipBias {from=""}
   ]
  }
  MaterialExpressionTextureSample_4 {
   class=MaterialExpressionTextureSample
   inputs [
    UVs {from=MaterialExpressionNamedRerouteUsage_227}
    Tex {from=MaterialExpressionTextureObjectParameter_14}
    Apply View MipBias {from=""}
   ]
  }
  MaterialExpressionTextureSample_5 {
   class=MaterialExpressionTextureSample
   inputs [
    UVs {from=MaterialExpressionNamedRerouteUsage_222}
    Tex {from=MaterialExpressionTextureObjectParameter_17}
    Apply View MipBias {from=""}
   ]
  }
  MaterialExpressionTextureSample_6 {
   class=MaterialExpressionTextureSample
   inputs [
    UVs {from=MaterialExpressionNamedRerouteUsage_223}
    Tex {from=MaterialExpressionTextureObjectParameter_16}
    Apply View MipBias {from=""}
   ]
  }
  MaterialExpressionTextureSample_7 {
   class=MaterialExpressionTextureSample
   inputs [
    UVs {from=MaterialExpressionNamedRerouteUsage_220}
    Tex {from=MaterialExpressionTextureObjectParameter_3}
    Apply View MipBias {from=""}
   ]
  }
  MaterialExpressionTextureSample_8 {
   class=MaterialExpressionTextureSample
   inputs [
    UVs {from=MaterialExpressionNamedRerouteUsage_65}
    Tex {from=MaterialExpressionTextureObjectParameter_13}
    Apply View MipBias {from=""}
   ]
  }
  MaterialExpressionTextureSample_9 {
   class=MaterialExpressionTextureSample
   inputs [
    UVs {from=MaterialExpressionNamedRerouteUsage_63}
    Tex {from=MaterialExpressionTextureObjectParameter_12}
    Apply View MipBias {from=""}
   ]
  }
  MaterialExpressionVectorParameter_10 {class=MaterialExpressionVectorParameter, properties {group=Normal, parameterName=Normal, defaultValue=<rgba 0 0 1 1>}}
  MaterialExpressionVectorParameter_12 {
   class=MaterialExpressionVectorParameter
   properties {group=Roughness, parameterName=Roughness, sortPriority=30, defaultValue=<rgba 0.2122 0.2122 0 1>}
  }
  MaterialExpressionVectorParameter_13 {class=MaterialExpressionVectorParameter, properties {group=Fresnel, parameterName=F0, defaultValue=<rgba 0.04 0.04 0.04 1>}}
  MaterialExpressionVectorParameter_14 {class=MaterialExpressionVectorParameter, properties {group=Fresnel, parameterName=F90, defaultValue=<rgba 1 1 1 0>}}
  MaterialExpressionVectorParameter_15 {class=MaterialExpressionVectorParameter, properties {group=Sheen, parameterName=SheenColor, defaultValue=<rgba 0 0 0 1>}}
  MaterialExpressionVectorParameter_16 {class=MaterialExpressionVectorParameter, properties {group=Sheen, parameterName=SheenRoughness, defaultValue=<rgba 0.1 0.1 0 1>}}
  MaterialExpressionVectorParameter_17 {class=MaterialExpressionVectorParameter, properties {group=Anisotropy, parameterName=AnisotropicRotation, defaultValue=<rgba 0.5 0.5 0 0>}}
  MaterialExpressionVectorParameter_26 {class=MaterialExpressionVectorParameter, properties {group=ClearCoat, parameterName=ClearCoatNormal, defaultValue=<rgba 0 0 1 1>}}
  MaterialExpressionVectorParameter_9 {class=MaterialExpressionVectorParameter, properties {group=Diffuse, parameterName=DiffuseColor, defaultValue=<rgba 0 0 0 1>}}
  MaterialExpressionVertexNormalWS_2 {class=MaterialExpressionVertexNormalWS, inputs [], properties {}}
 }
 properties {
  bUsedWithNanite=true
  bUsedWithStaticLighting=true
  blendMode=BLEND_Masked
  shadingModel=MSM_SubsurfaceProfile
  translucencyLightingMode=TLM_SurfacePerPixelLighting
 }
}
